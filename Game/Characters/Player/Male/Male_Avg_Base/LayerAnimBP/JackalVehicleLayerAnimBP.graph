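# Animation layer Blueprint that poses the Fortnite player riding the Jackal vehicle (hoverboard).
# The event graph copies previous-frame aim from the player anim instance; JackalVehicleInput flags drive many blends and transitions.
# The full-body layer blends the incoming pose with hoverboard state machines, sprint additives and leans, with variants for carrying downed (DBNO) players.
# The final-pose layer adjusts the root bone for emotes and female riders.
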AnimBlueprint /Game/Characters/Player/Male/Male_Avg_Base/LayerAnimBP/JackalVehicleLayerAnimBP
target_skeleton=&Fortnite_M_Avg_Player_Skeleton
graphs [
 AnimGraph {nodes [Output Pose() {layerGroup=None}]}
 EventGraph {nodes [
   Event BlueprintUpdateAnimation()
   > Sequence()
   n1 = TryGetPawnOwner()
   > n2 = Cast To FortPlayerAnimInstance(Object=GetAnimInstance(self=GetComponentByClass!(self:Actor Object Reference=@n1, ComponentClass:Actor Component Class Reference=/Script/Engine.SkeletalMeshComponent) -> (ReturnValue:Skeletal Mesh Component Object Reference)))
   n3 = GetPreviousFrameAimPitchAndYaw!(self:Fort Anim Instance Object Reference=@n2) -> (OutPitch:float, OutYaw:float)
   > Set AimYaw(AimYaw:float=@n3.OutYaw) -> (Output_Get:float)
   > Set AimPitch(AimPitch:float=@n3) -> (Output_Get:float)
  ]}
 VehicleFinalPoseOverride {nodes [
   n1 = Break Fort Anim Input Jackal Vehicle!(FortAnimInput_JackalVehicle:"Fort Anim Input Jackal Vehicle Structure (by ref)"=Get JackalVehicleInput() -> (JackalVehicleInput:Fort Anim Input Jackal Vehicle Structure)) -> (
    bIsUsingJackalVehicle:bool=~
    bIsSteeringLeft:bool=~
    bIsSteeringRight:bool=~
    SteerAngleDeadZoneDegrees:float=~
    SteerAngle:float=~
    LeanYaw:float=~
    QuantizedSteerAngle:float=~
    SteerAlpha:float=~
    RunForwardAlpha:float=~
    SlopePitchDegreeAngle:float=~
    SlopeRollDegreeAngle:float=~
    DistanceFromGround:float=~
    PivotPlayRate:float=~
    bIsReversing:bool=~
    bIsBraking:bool=~
    bIsMoving:bool=~
    bIsMovingForward:bool=~
    bIsSprinting:bool=~
    bInAir:bool=~
    bIsFalling:bool=~
    bIsJumping:bool=~
    bIsRelaxed:bool=~
    bIsBoosting:bool=~
    bHasReachedJumpChargeStartThreshold:bool=~
    bHasReachedJumpChargeMidThreshold:bool=~
    bHasReachedJumpChargeFullThreshold:bool=~
    bAdjustRootForFemaleRider:bool=~
    bIsPlayingEmoteOnHoverboard:bool=~
    bShouldApplyAdditive:bool=~
    bPlayPivotOnGroundAndNotBoosting:bool=~
    bIsOnLowerHill:bool=~
    bLowerBodyIdleToLoopTransition:bool=~
    bInterruptHoverboardFullybody:bool=~
    bEnterFullBodyHoverboardState:bool=~
    bDefaultToJumpStartTransition:bool=~
    bLocomotionPoseToJumpTransition:bool=~
    bBoostingToBoostingJumpTransition:bool=~
    bJumpToLocomotionPoseTransition:bool=~
    bJumpApexToJumpFallTransition:bool=~
    bIdleToMovementStartTransition:bool=~
    bIdleToMovementLoopTransition:bool=~
    bMovementLoopToMovementStopTransition:bool=~
    bMovementLoopToPivotTransition:bool=~
    bMovementLoopToIdleTransition:bool=~
    bIdleAdditiveToCollisionNTransition:bool=~
    bSplitBodyToHoverboardBRTransition:bool=~
    bHoverboardBRToSplitBodyTransition:bool=~
    bHoverboardBRMovementToJumpChargeTransition:bool=~
    bIdlesToJackalVehicleTransition:bool=~
    bPlayAdditiveLeans:bool=~
    bPlayBalloonLeans:bool=~
    bPlayJumpTrickVertical:bool=~
    bPlayJumpTrick:bool=~
    bPlayMovingFast:bool=~
    bPlayHipAdjustmentAdditive:bool=~
    bPlayDriveSouth:bool=~
    bPlayHeadAimOffset:bool=~
    bPlaySlopeAimOffset:bool=~
    JumpCombatAdditiveWeight:float=~
    MeleeTwistIdle:float=~
    MeleeTwistLocomotionLoop:float=~
    VelocityImpact:Fort Anim Input Velocity Impact Structure
    ImpactDisplacement:Vector=~
    SteerYaw:float=~
    EmoteHoverboardPosition:Vector=~
    EmoteHoverboardRotation:Rotator=~
    bShouldAttachFeetToHoverboard:bool=~
    FootLeftLocationOffset:Vector=~
    FootLeftRotationOffset:Rotator=~
    FootRightLocationOffset:Vector=~
    FootRightRotationOffset:Rotator=~
    StoppedThreshold:float=~
    MovingForwardThreshold:float=~
    MovingFowardFastThreshold:float=~
    DefaultToJumpStartTransitionThreshold:float=~
    JumpTrickAngularVelocityThreshold:float=~
    JumpDistanceFromGroundThreshold:float=~
    VelocityStartThreshold:float=~
    MovingFastThreshold:float=~
    RelaxedSpeedThreshold:float=~
    JumpChargeStartThreshold:float=~
    JumpChargeMidThreshold:float=~
    JumpChargeFullThreshold:float=~
    RotatingAngularVelocityThreshold:float=~
    IdleToLoopTransitionSpeedThreshold:float=~
    IdleToMovementLoopTransitionThreshold:float=~
    MeleeTwistIdleMultiplier:float=~
    MeleeTwistLocomotionLoopMultiplier:float=~
    LeanYawForMaxPivotPlayRate:float=~
    SteerAlphaForMaxPivotPlayRate:float=~
    PrePivotCardinalDirection:EFortCardinalDirection Enum
   ) {is=title K2Node_BreakStruct}
   n2 = "Transform (Modify) Bone - Bone: Root"!(
    ComponentPose=Local To Component(LocalPose=Use cached pose 'HoverboardAdjustmentsInputPose'())
    Alpha=1
    Translation=@n1.EmoteHoverboardPosition
   ) -> (Pose) {translationMode=BMM_Replace, translationSpace=BCS_WorldSpace, alphaBoolBlend {blendOption=Linear}, boneToModify {boneName=Root}}
   n3 = "Transform (Modify) Bone - Bone: Root"!(ComponentPose=@n2, Alpha=0.33, Rotation:Rotator=@n1.EmoteHoverboardRotation) -> (Pose) {
    rotationMode=BMM_Additive
    rotationSpace=BCS_BoneSpace
    translationSpace=BCS_WorldSpace
    alphaBoolBlend {blendOption=Linear}
    alphaScaleBiasClamp {outRange {max=0.5}}
    boneToModify {boneName=Root}
   }
   n4 = Component To Local(ComponentPose=@n3)
   n5 = Use cached pose 'HoverboardAdjustmentsInputPose'()
   n6 = Break Fort Anim Input Jackal Vehicle(FortAnimInput_JackalVehicle=Get JackalVehicleInput()) -> (
    bIsUsingJackalVehicle=~
    bIsSteeringLeft=~
    bIsSteeringRight=~
    SteerAngleDeadZoneDegrees=~
    SteerAngle=~
    LeanYaw=~
    QuantizedSteerAngle=~
    SteerAlpha=~
    RunForwardAlpha=~
    SlopePitchDegreeAngle=~
    SlopeRollDegreeAngle=~
    DistanceFromGround=~
    PivotPlayRate=~
    bIsReversing=~
    bIsBraking=~
    bIsMoving=~
    bIsMovingForward=~
    bIsSprinting=~
    bInAir=~
    bIsFalling=~
    bIsJumping=~
    bIsRelaxed=~
    bIsBoosting=~
    bHasReachedJumpChargeStartThreshold=~
    bHasReachedJumpChargeMidThreshold=~
    bHasReachedJumpChargeFullThreshold=~
    bAdjustRootForFemaleRider=~
    bIsPlayingEmoteOnHoverboard=~
    bShouldApplyAdditive=~
    bPlayPivotOnGroundAndNotBoosting=~
    bIsOnLowerHill=~
    bLowerBodyIdleToLoopTransition=~
    bInterruptHoverboardFullybody=~
    bEnterFullBodyHoverboardState=~
    bDefaultToJumpStartTransition=~
    bLocomotionPoseToJumpTransition=~
    bBoostingToBoostingJumpTransition=~
    bJumpToLocomotionPoseTransition=~
    bJumpApexToJumpFallTransition=~
    bIdleToMovementStartTransition=~
    bIdleToMovementLoopTransition=~
    bMovementLoopToMovementStopTransition=~
    bMovementLoopToPivotTransition=~
    bMovementLoopToIdleTransition=~
    bIdleAdditiveToCollisionNTransition=~
    bSplitBodyToHoverboardBRTransition=~
    bHoverboardBRToSplitBodyTransition=~
    bHoverboardBRMovementToJumpChargeTransition=~
    bIdlesToJackalVehicleTransition=~
    bPlayAdditiveLeans=~
    bPlayBalloonLeans=~
    bPlayJumpTrickVertical=~
    bPlayJumpTrick=~
    bPlayMovingFast=~
    bPlayHipAdjustmentAdditive=~
    bPlayDriveSouth=~
    bPlayHeadAimOffset=~
    bPlaySlopeAimOffset=~
    JumpCombatAdditiveWeight=~
    MeleeTwistIdle=~
    MeleeTwistLocomotionLoop=~
    ImpactDisplacement=~
    SteerYaw=~
    EmoteHoverboardPosition=~
    EmoteHoverboardRotation=~
    bShouldAttachFeetToHoverboard=~
    FootLeftLocationOffset=~
    FootLeftRotationOffset=~
    FootRightLocationOffset=~
    FootRightRotationOffset=~
    StoppedThreshold=~
    MovingForwardThreshold=~
    MovingFowardFastThreshold=~
    DefaultToJumpStartTransitionThreshold=~
    JumpTrickAngularVelocityThreshold=~
    JumpDistanceFromGroundThreshold=~
    VelocityStartThreshold=~
    MovingFastThreshold=~
    RelaxedSpeedThreshold=~
    JumpChargeStartThreshold=~
    JumpChargeMidThreshold=~
    JumpChargeFullThreshold=~
    RotatingAngularVelocityThreshold=~
    IdleToLoopTransitionSpeedThreshold=~
    IdleToMovementLoopTransitionThreshold=~
    MeleeTwistIdleMultiplier=~
    MeleeTwistLocomotionLoopMultiplier=~
    LeanYawForMaxPivotPlayRate=~
    SteerAlphaForMaxPivotPlayRate=~
   )
   n7 = "Transform (Modify) Bone - Bone: Root"!(
    ComponentPose=Local To Component(LocalPose=Blend Poses by bool(BlendPose_0=@n4, BlendPose_1=@n5, BlendTime_0=0.25, BlendTime_1=0.25, bActiveValue=@n6.bIsPlayingEmoteOnHoverboard) {blendType=Linear})
    bAlphaBoolEnabled:bool=Break Fort Anim Input Jackal Vehicle(FortAnimInput_JackalVehicle=Get JackalVehicleInput()) -> (
     bIsUsingJackalVehicle=~
     bIsSteeringLeft=~
     bIsSteeringRight=~
     SteerAngleDeadZoneDegrees=~
     SteerAngle=~
     LeanYaw=~
     QuantizedSteerAngle=~
     SteerAlpha=~
     RunForwardAlpha=~
     SlopePitchDegreeAngle=~
     SlopeRollDegreeAngle=~
     DistanceFromGround=~
     PivotPlayRate=~
     bIsReversing=~
     bIsBraking=~
     bIsMoving=~
     bIsMovingForward=~
     bIsSprinting=~
     bInAir=~
     bIsFalling=~
     bIsJumping=~
     bIsRelaxed=~
     bIsBoosting=~
     bHasReachedJumpChargeStartThreshold=~
     bHasReachedJumpChargeMidThreshold=~
     bHasReachedJumpChargeFullThreshold=~
     bAdjustRootForFemaleRider=~
     bIsPlayingEmoteOnHoverboard=~
     bShouldApplyAdditive=~
     bPlayPivotOnGroundAndNotBoosting=~
     bIsOnLowerHill=~
     bLowerBodyIdleToLoopTransition=~
     bInterruptHoverboardFullybody=~
     bEnterFullBodyHoverboardState=~
     bDefaultToJumpStartTransition=~
     bLocomotionPoseToJumpTransition=~
     bBoostingToBoostingJumpTransition=~
     bJumpToLocomotionPoseTransition=~
     bJumpApexToJumpFallTransition=~
     bIdleToMovementStartTransition=~
     bIdleToMovementLoopTransition=~
     bMovementLoopToMovementStopTransition=~
     bMovementLoopToPivotTransition=~
     bMovementLoopToIdleTransition=~
     bIdleAdditiveToCollisionNTransition=~
     bSplitBodyToHoverboardBRTransition=~
     bHoverboardBRToSplitBodyTransition=~
     bHoverboardBRMovementToJumpChargeTransition=~
     bIdlesToJackalVehicleTransition=~
     bPlayAdditiveLeans=~
     bPlayBalloonLeans=~
     bPlayJumpTrickVertical=~
     bPlayJumpTrick=~
     bPlayMovingFast=~
     bPlayHipAdjustmentAdditive=~
     bPlayDriveSouth=~
     bPlayHeadAimOffset=~
     bPlaySlopeAimOffset=~
     JumpCombatAdditiveWeight=~
     MeleeTwistIdle=~
     MeleeTwistLocomotionLoop=~
     ImpactDisplacement=~
     SteerYaw=~
     EmoteHoverboardPosition=~
     EmoteHoverboardRotation=~
     bShouldAttachFeetToHoverboard=~
     FootLeftLocationOffset=~
     FootLeftRotationOffset=~
     FootRightLocationOffset=~
     FootRightRotationOffset=~
     StoppedThreshold=~
     MovingForwardThreshold=~
     MovingFowardFastThreshold=~
     DefaultToJumpStartTransitionThreshold=~
     JumpTrickAngularVelocityThreshold=~
     JumpDistanceFromGroundThreshold=~
     VelocityStartThreshold=~
     MovingFastThreshold=~
     RelaxedSpeedThreshold=~
     JumpChargeStartThreshold=~
     JumpChargeMidThreshold=~
     JumpChargeFullThreshold=~
     RotatingAngularVelocityThreshold=~
     IdleToLoopTransitionSpeedThreshold=~
     IdleToMovementLoopTransitionThreshold=~
     MeleeTwistIdleMultiplier=~
     MeleeTwistLocomotionLoopMultiplier=~
     LeanYawForMaxPivotPlayRate=~
     SteerAlphaForMaxPivotPlayRate=~
    ).bAdjustRootForFemaleRider
    Translation:Vector="0,0,-1.5"
   ) -> (Pose) {
    alphaInputType=Bool
    translationMode=BMM_Additive
    translationSpace=BCS_WorldSpace
    alphaBoolBlend {blendOption=Linear}
    boneToModify {boneName=Root}
    translation {z=-1.5}
   }
   Output Pose(Result=Component To Local(ComponentPose=@n7)) {layerGroup=VehicleOverride}
   Save cached pose 'HoverboardAdjustmentsInputPose'(Pose=Input Pose() {name=InFinalPose}) {cacheName=HoverboardAdjustmentsInputPose}
  ]}
 VehicleFullBodyOverride {nodes [
   AnimGraphNode_StateMachine_1 = HoverboardBRFullBody() {is=title AnimGraphNode_StateMachine}
   AnimGraphNode_StateMachine_2 = HoverboardAdditive() {is=title AnimGraphNode_StateMachine}
   n1 = Break Fort Anim Input Jackal Vehicle(FortAnimInput_JackalVehicle=Get JackalVehicleInput()) -> (
    bIsUsingJackalVehicle=~
    bIsSteeringLeft=~
    bIsSteeringRight=~
    SteerAngleDeadZoneDegrees=~
    SteerAngle=~
    LeanYaw=~
    QuantizedSteerAngle=~
    SteerAlpha=~
    RunForwardAlpha=~
    SlopePitchDegreeAngle=~
    SlopeRollDegreeAngle=~
    DistanceFromGround=~
    PivotPlayRate=~
    bIsReversing=~
    bIsBraking=~
    bIsMoving=~
    bIsMovingForward=~
    bIsSprinting=~
    bInAir=~
    bIsFalling=~
    bIsJumping=~
    bIsRelaxed=~
    bIsBoosting=~
    bHasReachedJumpChargeStartThreshold=~
    bHasReachedJumpChargeMidThreshold=~
    bHasReachedJumpChargeFullThreshold=~
    bAdjustRootForFemaleRider=~
    bIsPlayingEmoteOnHoverboard=~
    bShouldApplyAdditive=~
    bPlayPivotOnGroundAndNotBoosting=~
    bIsOnLowerHill=~
    bLowerBodyIdleToLoopTransition=~
    bInterruptHoverboardFullybody=~
    bEnterFullBodyHoverboardState=~
    bDefaultToJumpStartTransition=~
    bLocomotionPoseToJumpTransition=~
    bBoostingToBoostingJumpTransition=~
    bJumpToLocomotionPoseTransition=~
    bJumpApexToJumpFallTransition=~
    bIdleToMovementStartTransition=~
    bIdleToMovementLoopTransition=~
    bMovementLoopToMovementStopTransition=~
    bMovementLoopToPivotTransition=~
    bMovementLoopToIdleTransition=~
    bIdleAdditiveToCollisionNTransition=~
    bSplitBodyToHoverboardBRTransition=~
    bHoverboardBRToSplitBodyTransition=~
    bHoverboardBRMovementToJumpChargeTransition=~
    bIdlesToJackalVehicleTransition=~
    bPlayAdditiveLeans=~
    bPlayBalloonLeans=~
    bPlayJumpTrickVertical=~
    bPlayJumpTrick=~
    bPlayMovingFast=~
    bPlayHipAdjustmentAdditive=~
    bPlayDriveSouth=~
    bPlayHeadAimOffset=~
    bPlaySlopeAimOffset=~
    JumpCombatAdditiveWeight=~
    MeleeTwistIdle=~
    MeleeTwistLocomotionLoop=~
    ImpactDisplacement=~
    SteerYaw=~
    EmoteHoverboardPosition=~
    EmoteHoverboardRotation=~
    bShouldAttachFeetToHoverboard=~
    FootLeftLocationOffset=~
    FootLeftRotationOffset=~
    FootRightLocationOffset=~
    FootRightRotationOffset=~
    StoppedThreshold=~
    MovingForwardThreshold=~
    MovingFowardFastThreshold=~
    DefaultToJumpStartTransitionThreshold=~
    JumpTrickAngularVelocityThreshold=~
    JumpDistanceFromGroundThreshold=~
    VelocityStartThreshold=~
    MovingFastThreshold=~
    RelaxedSpeedThreshold=~
    JumpChargeStartThreshold=~
    JumpChargeMidThreshold=~
    JumpChargeFullThreshold=~
    RotatingAngularVelocityThreshold=~
    IdleToLoopTransitionSpeedThreshold=~
    IdleToMovementLoopTransitionThreshold=~
    MeleeTwistIdleMultiplier=~
    MeleeTwistLocomotionLoopMultiplier=~
    LeanYawForMaxPivotPlayRate=~
    SteerAlphaForMaxPivotPlayRate=~
   )
   n2 = Apply Additive!(
    Base=Apply Additive(Base=@AnimGraphNode_StateMachine_1, Additive=@AnimGraphNode_StateMachine_2, bAlphaBoolEnabled=@n1.bIsSprinting) {alphaInputType=Bool, alphaBoolBlend {blendInTime=0.15, blendOption=Linear, blendOutTime=0.15}}
    Additive=Blend Poses by bool(
     BlendPose_0=Blendspace Player 'DBNOCarry_JackalVehicle_Leans_BS'!(X=@n1.LeanYaw) -> (Pose) {blendSpace=/Game/Animation/Game/MainPlayer/Vehicles/JackalVehicle/DBNO/DBNOCarry_JackalVehicle_Leans_BS}
     BlendPose_1=Blendspace Player 'JackalVehicle_Leans_BS'!(X=@n1.LeanYaw) -> (Pose) {blendSpace=/Game/Animation/Game/MainPlayer/Vehicles/JackalVehicle/JackalVehicle_Leans_BS}
     bActiveValue=Get bIsDBNOCarrying() -> (bIsDBNOCarrying:bool)
    ) {blendType=Linear}
    bAlphaBoolEnabled:bool=@n1.bPlayAdditiveLeans
   ) -> (Pose) {alphaInputType=Bool, alphaBoolBlend {blendInTime=0.15, blendOption=Linear, blendOutTime=0.15}}
   n3 = Use cached pose 'FullBodyInputPoseFromMain'()
   n4 = Break Fort Anim Input Jackal Vehicle(FortAnimInput_JackalVehicle=Get JackalVehicleInput()) -> (
    bIsUsingJackalVehicle=~
    bIsSteeringLeft=~
    bIsSteeringRight=~
    SteerAngleDeadZoneDegrees=~
    SteerAngle=~
    LeanYaw=~
    QuantizedSteerAngle=~
    SteerAlpha=~
    RunForwardAlpha=~
    SlopePitchDegreeAngle=~
    SlopeRollDegreeAngle=~
    DistanceFromGround=~
    PivotPlayRate=~
    bIsReversing=~
    bIsBraking=~
    bIsMoving=~
    bIsMovingForward=~
    bIsSprinting=~
    bInAir=~
    bIsFalling=~
    bIsJumping=~
    bIsRelaxed=~
    bIsBoosting=~
    bHasReachedJumpChargeStartThreshold=~
    bHasReachedJumpChargeMidThreshold=~
    bHasReachedJumpChargeFullThreshold=~
    bAdjustRootForFemaleRider=~
    bIsPlayingEmoteOnHoverboard=~
    bShouldApplyAdditive=~
    bPlayPivotOnGroundAndNotBoosting=~
    bIsOnLowerHill=~
    bLowerBodyIdleToLoopTransition=~
    bInterruptHoverboardFullybody=~
    bEnterFullBodyHoverboardState=~
    bDefaultToJumpStartTransition=~
    bLocomotionPoseToJumpTransition=~
    bBoostingToBoostingJumpTransition=~
    bJumpToLocomotionPoseTransition=~
    bJumpApexToJumpFallTransition=~
    bIdleToMovementStartTransition=~
    bIdleToMovementLoopTransition=~
    bMovementLoopToMovementStopTransition=~
    bMovementLoopToPivotTransition=~
    bMovementLoopToIdleTransition=~
    bIdleAdditiveToCollisionNTransition=~
    bSplitBodyToHoverboardBRTransition=~
    bHoverboardBRToSplitBodyTransition=~
    bHoverboardBRMovementToJumpChargeTransition=~
    bIdlesToJackalVehicleTransition=~
    bPlayAdditiveLeans=~
    bPlayBalloonLeans=~
    bPlayJumpTrickVertical=~
    bPlayJumpTrick=~
    bPlayMovingFast=~
    bPlayHipAdjustmentAdditive=~
    bPlayDriveSouth=~
    bPlayHeadAimOffset=~
    bPlaySlopeAimOffset=~
    JumpCombatAdditiveWeight=~
    MeleeTwistIdle=~
    MeleeTwistLocomotionLoop=~
    ImpactDisplacement=~
    SteerYaw=~
    EmoteHoverboardPosition=~
    EmoteHoverboardRotation=~
    bShouldAttachFeetToHoverboard=~
    FootLeftLocationOffset=~
    FootLeftRotationOffset=~
    FootRightLocationOffset=~
    FootRightRotationOffset=~
    StoppedThreshold=~
    MovingForwardThreshold=~
    MovingFowardFastThreshold=~
    DefaultToJumpStartTransitionThreshold=~
    JumpTrickAngularVelocityThreshold=~
    JumpDistanceFromGroundThreshold=~
    VelocityStartThreshold=~
    MovingFastThreshold=~
    RelaxedSpeedThreshold=~
    JumpChargeStartThreshold=~
    JumpChargeMidThreshold=~
    JumpChargeFullThreshold=~
    RotatingAngularVelocityThreshold=~
    IdleToLoopTransitionSpeedThreshold=~
    IdleToMovementLoopTransitionThreshold=~
    MeleeTwistIdleMultiplier=~
    MeleeTwistLocomotionLoopMultiplier=~
    LeanYawForMaxPivotPlayRate=~
    SteerAlphaForMaxPivotPlayRate=~
   )
   Output Pose(Result=Blend Poses by bool(
     BlendPose_0=Use cached pose 'FullBodyInputPoseFromMain'()
     BlendPose_1=Blend Poses by bool(BlendPose_0=@n3, BlendPose_1=@n2, BlendTime_0=0, bActiveValue=@n4.bInterruptHoverboardFullybody) {blendType=Linear}
     BlendTime_0=0.2
     bActiveValue=@n4.bHoverboardBRToSplitBodyTransition
    ) {blendType=Linear}) {layerGroup=VehicleOverride}
   AnimGraphNode_StateMachine_0 = HoverboardBR() {is=title AnimGraphNode_StateMachine}
   Save cached pose 'HoverboardBRPose'(Pose=@AnimGraphNode_StateMachine_0) {cacheName=HoverboardBRPose, cachePoseName=HoverboardBRPose}
   AnimGraphNode_SubInput_0 = Input Pose() {name=InPoseFullBody}
   Save cached pose 'FullBodyInputPoseFromMain'(Pose=@AnimGraphNode_SubInput_0) {cacheName=FullBodyInputPoseFromMain}
  ]}
 HoverboardBR {
  entry=Idle
  path=VehicleFullBodyOverride.AnimGraphNode_StateMachine_0.HoverboardBR
  ids=id
  nodes [
   AnimStateNode_3 = MovementStart() {is=title AnimStateNode}
   AnimStateNode_4 = MovementLoop() {is=title AnimStateNode}
   AnimStateNode_5 = MovementStop() {is=title AnimStateNode}
   AnimStateNode_6 = Pivot() {is=title AnimStateNode}
   AnimStateNode_8 = Idle()
   AnimStateTransitionNode_10 = AnimStateTransitionNode()
   AnimStateTransitionNode_16 = AnimStateTransitionNode()
   AnimStateTransitionNode_25 = AnimStateTransitionNode() {bAutomaticRuleBasedOnSequencePlayerInState=true, crossfadeDuration=0.4}
   AnimStateTransitionNode_32 = AnimStateTransitionNode() {crossfadeDuration=0.45}
   AnimStateTransitionNode_4 = AnimStateTransitionNode() {priorityOrder=2}
   AnimStateTransitionNode_5 = AnimStateTransitionNode() {bAutomaticRuleBasedOnSequencePlayerInState=true, crossfadeDuration=0.4, priorityOrder=2}
   AnimStateTransitionNode_7 = AnimStateTransitionNode() {bAutomaticRuleBasedOnSequencePlayerInState=true, crossfadeDuration=0.35}
   AnimStateTransitionNode_9 = AnimStateTransitionNode()
  ]
  links [
   {
    dst=MovementLoop
    rule=VehicleFullBodyOverride.AnimGraphNode_StateMachine_0.HoverboardBR.AnimStateTransitionNode_9.Transition
    src=Idle
    via=AnimStateTransitionNode_9
   }
   {
    dst=MovementStart
    rule=VehicleFullBodyOverride.AnimGraphNode_StateMachine_0.HoverboardBR.AnimStateTransitionNode_4.AnimationTransitionGraph_1
    src=Idle
    via=AnimStateTransitionNode_4
   }
   {
    dst=Idle
    rule=VehicleFullBodyOverride.AnimGraphNode_StateMachine_0.HoverboardBR.AnimStateTransitionNode_10.Transition
    src=MovementLoop
    via=AnimStateTransitionNode_10
   }
   {
    dst=MovementStop
    rule=VehicleFullBodyOverride.AnimGraphNode_StateMachine_0.HoverboardBR.AnimStateTransitionNode_32.Transition
    src=MovementLoop
    via=AnimStateTransitionNode_32
   }
   {
    dst=Pivot
    rule=VehicleFullBodyOverride.AnimGraphNode_StateMachine_0.HoverboardBR.AnimStateTransitionNode_16.Transition
    src=MovementLoop
    via=AnimStateTransitionNode_16
   }
   {
    dst=MovementLoop
    rule=VehicleFullBodyOverride.AnimGraphNode_StateMachine_0.HoverboardBR.AnimStateTransitionNode_5.AnimationTransitionGraph_1
    src=MovementStart
    via=AnimStateTransitionNode_5
   }
   {
    dst=Idle
    rule=VehicleFullBodyOverride.AnimGraphNode_StateMachine_0.HoverboardBR.AnimStateTransitionNode_7.AnimationTransitionGraph_1
    src=MovementStop
    via=AnimStateTransitionNode_7
   }
   {
    dst=MovementLoop
    rule=VehicleFullBodyOverride.AnimGraphNode_StateMachine_0.HoverboardBR.AnimStateTransitionNode_25.Transition
    src=Pivot
    via=AnimStateTransitionNode_25
   }
  ]
 }
 MovementStart {
  path=VehicleFullBodyOverride.AnimGraphNode_StateMachine_0.HoverboardBR.AnimStateNode_3.MovementStart
  nodes [
   n1 = Get LocomotionCardinalDirection() -> (LocomotionCardinalDirection:EFortCardinalDirection Enum=North)
   n2 = "Blend Poses (EFortCardinalDirection)"!(
    BlendPose_0:Pose Link Structure=Sequence Player 'DBNOCarry_JackalVehicle_Drive_N_Start'() {bLoopAnimation=false, sequence=/Game/Animation/Game/MainPlayer/Vehicles/JackalVehicle/DBNO/DBNOCarry_JackalVehicle_Drive_N_Start}
    BlendPose_1:Pose Link Structure=Sequence Player 'DBNOCarry_JackalVehicle_Drive_S_Start'() {bLoopAnimation=false, sequence=/Game/Animation/Game/MainPlayer/Vehicles/JackalVehicle/DBNO/DBNOCarry_JackalVehicle_Drive_S_Start}
    BlendTime_0:float=0.2
    BlendTime_1:float=0.2
    ActiveEnumValue:Byte=@n1
   ) -> (Pose:Pose Link Structure) {
    is=title AnimGraphNode_BlendListByEnum
    blendType=Linear
    blendPose [{linkId=-1, sourceLinkId=-1}, {linkId=-1, sourceLinkId=-1}]
    blendTime [0.1,0.1]
   }
   n3 = "Blend Poses (EFortCardinalDirection)"(
    BlendPose_0=Sequence Player 'JackalVehicle_Drive_N_Fast_Start'() {bLoopAnimation=false, sequence=/Game/Animation/Game/MainPlayer/Vehicles/JackalVehicle/JackalVehicle_Drive_N_Fast_Start}
    BlendPose_1=Sequence Player 'JackalVehicle_Drive_S_Start'() {bLoopAnimation=false, sequence=/Game/Animation/Game/MainPlayer/Vehicles/JackalVehicle/JackalVehicle_Drive_S_Start}
    BlendTime_0=0.2
    BlendTime_1=0.2
    ActiveEnumValue=@n1
   ) {blendType=Linear, blendPose [{linkId=-1, sourceLinkId=-1}, {linkId=-1, sourceLinkId=-1}], blendTime [0.1,0.1]}
   Output Animation Pose(Result=Blend Poses by bool(BlendPose_0=@n2, BlendPose_1=@n3, bActiveValue=Get bIsDBNOCarrying()) {blendType=Linear})
  ]
 }
 MovementLoop {
  path=VehicleFullBodyOverride.AnimGraphNode_StateMachine_0.HoverboardBR.AnimStateNode_4.MovementLoop
  nodes [
   n1 = Break Fort Anim Input Jackal Vehicle(FortAnimInput_JackalVehicle=Get JackalVehicleInput()) -> (
    bIsUsingJackalVehicle=~
    bIsSteeringLeft=~
    bIsSteeringRight=~
    SteerAngleDeadZoneDegrees=~
    SteerAngle=~
    LeanYaw=~
    QuantizedSteerAngle=~
    SteerAlpha=~
    RunForwardAlpha=~
    SlopePitchDegreeAngle=~
    SlopeRollDegreeAngle=~
    DistanceFromGround=~
    PivotPlayRate=~
    bIsReversing=~
    bIsBraking=~
    bIsMoving=~
    bIsMovingForward=~
    bIsSprinting=~
    bInAir=~
    bIsFalling=~
    bIsJumping=~
    bIsRelaxed=~
    bIsBoosting=~
    bHasReachedJumpChargeStartThreshold=~
    bHasReachedJumpChargeMidThreshold=~
    bHasReachedJumpChargeFullThreshold=~
    bAdjustRootForFemaleRider=~
    bIsPlayingEmoteOnHoverboard=~
    bShouldApplyAdditive=~
    bPlayPivotOnGroundAndNotBoosting=~
    bIsOnLowerHill=~
    bLowerBodyIdleToLoopTransition=~
    bInterruptHoverboardFullybody=~
    bEnterFullBodyHoverboardState=~
    bDefaultToJumpStartTransition=~
    bLocomotionPoseToJumpTransition=~
    bBoostingToBoostingJumpTransition=~
    bJumpToLocomotionPoseTransition=~
    bJumpApexToJumpFallTransition=~
    bIdleToMovementStartTransition=~
    bIdleToMovementLoopTransition=~
    bMovementLoopToMovementStopTransition=~
    bMovementLoopToPivotTransition=~
    bMovementLoopToIdleTransition=~
    bIdleAdditiveToCollisionNTransition=~
    bSplitBodyToHoverboardBRTransition=~
    bHoverboardBRToSplitBodyTransition=~
    bHoverboardBRMovementToJumpChargeTransition=~
    bIdlesToJackalVehicleTransition=~
    bPlayAdditiveLeans=~
    bPlayBalloonLeans=~
    bPlayJumpTrickVertical=~
    bPlayJumpTrick=~
    bPlayMovingFast=~
    bPlayHipAdjustmentAdditive=~
    bPlayDriveSouth=~
    bPlayHeadAimOffset=~
    bPlaySlopeAimOffset=~
    JumpCombatAdditiveWeight=~
    MeleeTwistIdle=~
    MeleeTwistLocomotionLoop=~
    ImpactDisplacement=~
    SteerYaw=~
    EmoteHoverboardPosition=~
    EmoteHoverboardRotation=~
    bShouldAttachFeetToHoverboard=~
    FootLeftLocationOffset=~
    FootLeftRotationOffset=~
    FootRightLocationOffset=~
    FootRightRotationOffset=~
    StoppedThreshold=~
    MovingForwardThreshold=~
    MovingFowardFastThreshold=~
    DefaultToJumpStartTransitionThreshold=~
    JumpTrickAngularVelocityThreshold=~
    JumpDistanceFromGroundThreshold=~
    VelocityStartThreshold=~
    MovingFastThreshold=~
    RelaxedSpeedThreshold=~
    JumpChargeStartThreshold=~
    JumpChargeMidThreshold=~
    JumpChargeFullThreshold=~
    RotatingAngularVelocityThreshold=~
    IdleToLoopTransitionSpeedThreshold=~
    IdleToMovementLoopTransitionThreshold=~
    MeleeTwistIdleMultiplier=~
    MeleeTwistLocomotionLoopMultiplier=~
    LeanYawForMaxPivotPlayRate=~
    SteerAlphaForMaxPivotPlayRate=~
   )
   n2 = Sequence Player 'JackalVehicle_Pivot_WtoE'!(PlayRate:float=@n1.PivotPlayRate) -> (Pose) {
    bLoopAnimation=false
    groupName=SlopeE
    method=SyncGroup
    sequence=/Game/Animation/Game/MainPlayer/Vehicles/JackalVehicle/JackalVehicle_Pivot_WtoE
   }
   n3 = Sequence Player 'JackalVehicle_Pivot_WtoE_Slope'!(PlayRate:float=@n1.PivotPlayRate) -> (Pose) {
    bLoopAnimation=false
    groupName=SlopeE
    method=SyncGroup
    sequence=/Game/Animation/Game/MainPlayer/Vehicles/JackalVehicle/JackalVehicle_Pivot_WtoE_Slope
   }
   n4 = Sequence Player 'JackalVehicle_Pivot_EtoW_Slope'!(PlayRate:float=@n1.PivotPlayRate) -> (Pose) {
    bLoopAnimation=false
    groupName=SlopeW
    method=SyncGroup
    sequence=/Game/Animation/Game/MainPlayer/Vehicles/JackalVehicle/JackalVehicle_Pivot_EtoW_Slope
   }
   n5 = Sequence Player 'JackalVehicle_Pivot_EtoW'!(PlayRate:float=@n1.PivotPlayRate) -> (Pose) {
    bLoopAnimation=false
    groupName=SlopeW
    method=SyncGroup
    sequence=/Game/Animation/Game/MainPlayer/Vehicles/JackalVehicle/JackalVehicle_Pivot_EtoW
   }
   n6 = "Blend Poses (EFortCardinalDirection)"!(
    BlendPose_0=Sequence Player 'JackalVehicle_Drive_N_Fast_CMM'() {sequence=/Game/Animation/Game/MainPlayer/Vehicles/JackalVehicle/JackalVehicle_Drive_N_Fast_CMM}
    BlendPose_1=Blend Poses by bool(
     BlendPose_0=Sequence Player 'JackalVehicle_Drive_S_CMM'() {sequence=/Game/Animation/Game/MainPlayer/Vehicles/JackalVehicle/JackalVehicle_Drive_S_CMM}
     BlendPose_1=Blend Poses by bool(BlendPose_0=@n2, BlendPose_1=@n3, BlendTime_0=0.2, BlendTime_1=0.2, bActiveValue=@n1.bIsOnLowerHill) {blendType=Linear}
     BlendTime_0=0.2
     BlendTime_1=0.2
     bActiveValue=@n1.bPlayDriveSouth
    ) {blendType=Linear}
    BlendPose_2:Pose Link Structure=Sequence Player 'JackalVehicle_Drive_S_CMM'() {sequence=&JackalVehicle_Drive_S_CMM}
    BlendPose_3:Pose Link Structure=Blend Poses by bool(
     BlendPose_0=Sequence Player 'JackalVehicle_Drive_S_CMM'() {sequence=&JackalVehicle_Drive_S_CMM}
     BlendPose_1=Blend Poses by bool(BlendPose_0=@n5, BlendPose_1=@n4, BlendTime_0=0.2, BlendTime_1=0.2, bActiveValue=@n1.bIsOnLowerHill) {blendType=Linear}
     BlendTime_0=0.2
     BlendTime_1=0.2
     bActiveValue=@n1.bPlayDriveSouth
    ) {blendType=Linear}
    BlendTime_0=0.3
    BlendTime_1=0.25
    BlendTime_2:float=0.15
    BlendTime_3:float=0.25
    ActiveEnumValue=Get LocomotionCardinalDirection() -> (LocomotionCardinalDirection=North)
   ) -> (Pose) {
    blendType=Linear
    childUpateMode=ResetChildOnActivate
    blendPose [{linkId=-1, sourceLinkId=-1}, {linkId=-1, sourceLinkId=-1}, {linkId=-1, sourceLinkId=-1}, {linkId=-1, sourceLinkId=-1}]
    blendTime [0.1,0.1,0.1,0.1]
   }
   n7 = Blendspace Player 'JackalVehicle_BodyLeans_BS'(X=@n1.SteerAngle, Y=0) {blendSpace=/Game/Animation/Game/MainPlayer/Vehicles/JackalVehicle/JackalVehicle_BodyLeans_BS}
   Output Animation Pose(Result=Blend Poses by bool(
     BlendPose_0=Sequence Player 'DBNOCarry_JackalVehicle_Drive_N_Fast_CMM'() {sequence=/Game/Animation/Game/MainPlayer/Vehicles/JackalVehicle/DBNO/DBNOCarry_JackalVehicle_Drive_N_Fast_CMM}
     BlendPose_1=Blend Poses by bool(BlendPose_0=@n7, BlendPose_1=@n6, BlendTime_0=0.4, BlendTime_1=0.25, bActiveValue=@n1.bIsRelaxed) {blendType=Linear}
     bActiveValue=Get bIsDBNOCarrying()
    ) {blendType=Linear})
  ]
 }
 MovementStop {
  path=VehicleFullBodyOverride.AnimGraphNode_StateMachine_0.HoverboardBR.AnimStateNode_5.MovementStop
  nodes [
   n1 = Sequence Player 'JackalVehicle_Drive_N_Stop'() {
    bLoopAnimation=false
    groupName=Stop
    method=SyncGroup
    sequence=/Game/Animation/Game/MainPlayer/Vehicles/JackalVehicle/JackalVehicle_Drive_N_Stop
    startPosition=0.3
   }
   n2 = Sequence Player 'DBNOCarry_JackalVehicle_Drive_Stop'() {
    bLoopAnimation=false
    groupName=Stop
    method=SyncGroup
    sequence=/Game/Animation/Game/MainPlayer/Vehicles/JackalVehicle/DBNO/DBNOCarry_JackalVehicle_Drive_Stop
   }
   Output Animation Pose(Result=Blend Poses by bool(BlendPose_0=@n2, BlendPose_1=@n1, bActiveValue=Get bIsDBNOCarrying()) {blendType=Linear})
  ]
 }
 Pivot {
  path=VehicleFullBodyOverride.AnimGraphNode_StateMachine_0.HoverboardBR.AnimStateNode_6.Pivot
  nodes [
   n1 = Break Fort Anim Input Jackal Vehicle(FortAnimInput_JackalVehicle=Get JackalVehicleInput()) -> (
    bIsUsingJackalVehicle=~
    bIsSteeringLeft=~
    bIsSteeringRight=~
    SteerAngleDeadZoneDegrees=~
    SteerAngle=~
    LeanYaw=~
    QuantizedSteerAngle=~
    SteerAlpha=~
    RunForwardAlpha=~
    SlopePitchDegreeAngle=~
    SlopeRollDegreeAngle=~
    DistanceFromGround=~
    PivotPlayRate=~
    bIsReversing=~
    bIsBraking=~
    bIsMoving=~
    bIsMovingForward=~
    bIsSprinting=~
    bInAir=~
    bIsFalling=~
    bIsJumping=~
    bIsRelaxed=~
    bIsBoosting=~
    bHasReachedJumpChargeStartThreshold=~
    bHasReachedJumpChargeMidThreshold=~
    bHasReachedJumpChargeFullThreshold=~
    bAdjustRootForFemaleRider=~
    bIsPlayingEmoteOnHoverboard=~
    bShouldApplyAdditive=~
    bPlayPivotOnGroundAndNotBoosting=~
    bIsOnLowerHill=~
    bLowerBodyIdleToLoopTransition=~
    bInterruptHoverboardFullybody=~
    bEnterFullBodyHoverboardState=~
    bDefaultToJumpStartTransition=~
    bLocomotionPoseToJumpTransition=~
    bBoostingToBoostingJumpTransition=~
    bJumpToLocomotionPoseTransition=~
    bJumpApexToJumpFallTransition=~
    bIdleToMovementStartTransition=~
    bIdleToMovementLoopTransition=~
    bMovementLoopToMovementStopTransition=~
    bMovementLoopToPivotTransition=~
    bMovementLoopToIdleTransition=~
    bIdleAdditiveToCollisionNTransition=~
    bSplitBodyToHoverboardBRTransition=~
    bHoverboardBRToSplitBodyTransition=~
    bHoverboardBRMovementToJumpChargeTransition=~
    bIdlesToJackalVehicleTransition=~
    bPlayAdditiveLeans=~
    bPlayBalloonLeans=~
    bPlayJumpTrickVertical=~
    bPlayJumpTrick=~
    bPlayMovingFast=~
    bPlayHipAdjustmentAdditive=~
    bPlayDriveSouth=~
    bPlayHeadAimOffset=~
    bPlaySlopeAimOffset=~
    JumpCombatAdditiveWeight=~
    MeleeTwistIdle=~
    MeleeTwistLocomotionLoop=~
    ImpactDisplacement=~
    SteerYaw=~
    EmoteHoverboardPosition=~
    EmoteHoverboardRotation=~
    bShouldAttachFeetToHoverboard=~
    FootLeftLocationOffset=~
    FootLeftRotationOffset=~
    FootRightLocationOffset=~
    FootRightRotationOffset=~
    StoppedThreshold=~
    MovingForwardThreshold=~
    MovingFowardFastThreshold=~
    DefaultToJumpStartTransitionThreshold=~
    JumpTrickAngularVelocityThreshold=~
    JumpDistanceFromGroundThreshold=~
    VelocityStartThreshold=~
    MovingFastThreshold=~
    RelaxedSpeedThreshold=~
    JumpChargeStartThreshold=~
    JumpChargeMidThreshold=~
    JumpChargeFullThreshold=~
    RotatingAngularVelocityThreshold=~
    IdleToLoopTransitionSpeedThreshold=~
    IdleToMovementLoopTransitionThreshold=~
    MeleeTwistIdleMultiplier=~
    MeleeTwistLocomotionLoopMultiplier=~
    LeanYawForMaxPivotPlayRate=~
    SteerAlphaForMaxPivotPlayRate=~
   )
   n2 = "Blend Poses (EFortCardinalDirection)"!(
    BlendPose_0=Sequence Player 'DBNOCarry_JackalVehicle_Pivot_StoN'() {bLoopAnimation=false, sequence=/Game/Animation/Game/MainPlayer/Vehicles/JackalVehicle/DBNO/DBNOCarry_JackalVehicle_Pivot_StoN}
    BlendPose_1=Sequence Player 'DBNOCarry_JackalVehicle_Pivot_NtoS'() {bLoopAnimation=false, sequence=/Game/Animation/Game/MainPlayer/Vehicles/JackalVehicle/DBNO/DBNOCarry_JackalVehicle_Pivot_NtoS}
    BlendTime_0=0.1
    BlendTime_1=0.1
    ActiveEnumValue=@n1.PrePivotCardinalDirection
   ) -> (Pose) {blendType=Linear, blendPose [{linkId=-1, sourceLinkId=-1}, {linkId=-1, sourceLinkId=-1}], blendTime [0.1,0.1]}
   n3 = "Blend Poses (EFortCardinalDirection)"(
    BlendPose_0=Sequence Player 'JackalVehicle_Pivot_StoN'() {bLoopAnimation=false, sequence=/Game/Animation/Game/MainPlayer/Vehicles/JackalVehicle/JackalVehicle_Pivot_StoN}
    BlendPose_1=Sequence Player 'JackalVehicle_Pivot_NtoS'() {bLoopAnimation=false, sequence=/Game/Animation/Game/MainPlayer/Vehicles/JackalVehicle/JackalVehicle_Pivot_NtoS}
    BlendTime_0=0.1
    BlendTime_1=0.1
    ActiveEnumValue=@n1.PrePivotCardinalDirection
   ) {blendType=Linear, blendPose [{linkId=-1, sourceLinkId=-1}, {linkId=-1, sourceLinkId=-1}], blendTime [0.1,0.1]}
   Output Animation Pose(Result=Blend Poses by bool(BlendPose_0=@n2, BlendPose_1=@n3, bActiveValue=Get bIsDBNOCarrying()) {blendType=Linear})
  ]
 }
 Idle {
  path=VehicleFullBodyOverride.AnimGraphNode_StateMachine_0.HoverboardBR.AnimStateNode_8.Idle
  nodes [Output Animation Pose(Result=Blend Poses by bool(
     BlendPose_0=Sequence Player 'DBNOCarry_JackalVehicle_Idle_Pose_CMM'() {sequence=/Game/Animation/Game/MainPlayer/Vehicles/JackalVehicle/DBNO/DBNOCarry_JackalVehicle_Idle_Pose_CMM}
     BlendPose_1=Sequence Player 'JackalVehicle_Idle_Pose_CMM'() {sequence=/Game/Animation/Game/MainPlayer/Vehicles/JackalVehicle/JackalVehicle_Idle_Pose_CMM}
     bActiveValue=Get bIsDBNOCarrying()
    ) {blendType=Linear})]
 }
 Transition {
  path=VehicleFullBodyOverride.AnimGraphNode_StateMachine_0.HoverboardBR.AnimStateTransitionNode_10.Transition
  nodes [Result(bCanEnterTransition=Break Fort Anim Input Jackal Vehicle(FortAnimInput_JackalVehicle=Get JackalVehicleInput()) -> (
     bIsUsingJackalVehicle=~
     bIsSteeringLeft=~
     bIsSteeringRight=~
     SteerAngleDeadZoneDegrees=~
     SteerAngle=~
     LeanYaw=~
     QuantizedSteerAngle=~
     SteerAlpha=~
     RunForwardAlpha=~
     SlopePitchDegreeAngle=~
     SlopeRollDegreeAngle=~
     DistanceFromGround=~
     PivotPlayRate=~
     bIsReversing=~
     bIsBraking=~
     bIsMoving=~
     bIsMovingForward=~
     bIsSprinting=~
     bInAir=~
     bIsFalling=~
     bIsJumping=~
     bIsRelaxed=~
     bIsBoosting=~
     bHasReachedJumpChargeStartThreshold=~
     bHasReachedJumpChargeMidThreshold=~
     bHasReachedJumpChargeFullThreshold=~
     bAdjustRootForFemaleRider=~
     bIsPlayingEmoteOnHoverboard=~
     bShouldApplyAdditive=~
     bPlayPivotOnGroundAndNotBoosting=~
     bIsOnLowerHill=~
     bLowerBodyIdleToLoopTransition=~
     bInterruptHoverboardFullybody=~
     bEnterFullBodyHoverboardState=~
     bDefaultToJumpStartTransition=~
     bLocomotionPoseToJumpTransition=~
     bBoostingToBoostingJumpTransition=~
     bJumpToLocomotionPoseTransition=~
     bJumpApexToJumpFallTransition=~
     bIdleToMovementStartTransition=~
     bIdleToMovementLoopTransition=~
     bMovementLoopToMovementStopTransition=~
     bMovementLoopToPivotTransition=~
     bMovementLoopToIdleTransition=~
     bIdleAdditiveToCollisionNTransition=~
     bSplitBodyToHoverboardBRTransition=~
     bHoverboardBRToSplitBodyTransition=~
     bHoverboardBRMovementToJumpChargeTransition=~
     bIdlesToJackalVehicleTransition=~
     bPlayAdditiveLeans=~
     bPlayBalloonLeans=~
     bPlayJumpTrickVertical=~
     bPlayJumpTrick=~
     bPlayMovingFast=~
     bPlayHipAdjustmentAdditive=~
     bPlayDriveSouth=~
     bPlayHeadAimOffset=~
     bPlaySlopeAimOffset=~
     JumpCombatAdditiveWeight=~
     MeleeTwistIdle=~
     MeleeTwistLocomotionLoop=~
     ImpactDisplacement=~
     SteerYaw=~
     EmoteHoverboardPosition=~
     EmoteHoverboardRotation=~
     bShouldAttachFeetToHoverboard=~
     FootLeftLocationOffset=~
     FootLeftRotationOffset=~
     FootRightLocationOffset=~
     FootRightRotationOffset=~
     StoppedThreshold=~
     MovingForwardThreshold=~
     MovingFowardFastThreshold=~
     DefaultToJumpStartTransitionThreshold=~
     JumpTrickAngularVelocityThreshold=~
     JumpDistanceFromGroundThreshold=~
     VelocityStartThreshold=~
     MovingFastThreshold=~
     RelaxedSpeedThreshold=~
     JumpChargeStartThreshold=~
     JumpChargeMidThreshold=~
     JumpChargeFullThreshold=~
     RotatingAngularVelocityThreshold=~
     IdleToLoopTransitionSpeedThreshold=~
     IdleToMovementLoopTransitionThreshold=~
     MeleeTwistIdleMultiplier=~
     MeleeTwistLocomotionLoopMultiplier=~
     LeanYawForMaxPivotPlayRate=~
     SteerAlphaForMaxPivotPlayRate=~
    ).bMovementLoopToIdleTransition)]
 }
 Transition {
  path=VehicleFullBodyOverride.AnimGraphNode_StateMachine_0.HoverboardBR.AnimStateTransitionNode_16.Transition
  nodes [Result(bCanEnterTransition=Get bMovementLoopToPivotTransitionAndNotDBNOCarrying() -> (bMovementLoopToPivotTransitionAndNotDBNOCarrying:bool))]
 }
 Transition {path=VehicleFullBodyOverride.AnimGraphNode_StateMachine_0.HoverboardBR.AnimStateTransitionNode_25.Transition, nodes [Result()]}
 Transition {
  path=VehicleFullBodyOverride.AnimGraphNode_StateMachine_0.HoverboardBR.AnimStateTransitionNode_32.Transition
  nodes [Result(bCanEnterTransition=Break Fort Anim Input Jackal Vehicle(FortAnimInput_JackalVehicle=Get JackalVehicleInput()) -> (
     bIsUsingJackalVehicle=~
     bIsSteeringLeft=~
     bIsSteeringRight=~
     SteerAngleDeadZoneDegrees=~
     SteerAngle=~
     LeanYaw=~
     QuantizedSteerAngle=~
     SteerAlpha=~
     RunForwardAlpha=~
     SlopePitchDegreeAngle=~
     SlopeRollDegreeAngle=~
     DistanceFromGround=~
     PivotPlayRate=~
     bIsReversing=~
     bIsBraking=~
     bIsMoving=~
     bIsMovingForward=~
     bIsSprinting=~
     bInAir=~
     bIsFalling=~
     bIsJumping=~
     bIsRelaxed=~
     bIsBoosting=~
     bHasReachedJumpChargeStartThreshold=~
     bHasReachedJumpChargeMidThreshold=~
     bHasReachedJumpChargeFullThreshold=~
     bAdjustRootForFemaleRider=~
     bIsPlayingEmoteOnHoverboard=~
     bShouldApplyAdditive=~
     bPlayPivotOnGroundAndNotBoosting=~
     bIsOnLowerHill=~
     bLowerBodyIdleToLoopTransition=~
     bInterruptHoverboardFullybody=~
     bEnterFullBodyHoverboardState=~
     bDefaultToJumpStartTransition=~
     bLocomotionPoseToJumpTransition=~
     bBoostingToBoostingJumpTransition=~
     bJumpToLocomotionPoseTransition=~
     bJumpApexToJumpFallTransition=~
     bIdleToMovementStartTransition=~
     bIdleToMovementLoopTransition=~
     bMovementLoopToMovementStopTransition=~
     bMovementLoopToPivotTransition=~
     bMovementLoopToIdleTransition=~
     bIdleAdditiveToCollisionNTransition=~
     bSplitBodyToHoverboardBRTransition=~
     bHoverboardBRToSplitBodyTransition=~
     bHoverboardBRMovementToJumpChargeTransition=~
     bIdlesToJackalVehicleTransition=~
     bPlayAdditiveLeans=~
     bPlayBalloonLeans=~
     bPlayJumpTrickVertical=~
     bPlayJumpTrick=~
     bPlayMovingFast=~
     bPlayHipAdjustmentAdditive=~
     bPlayDriveSouth=~
     bPlayHeadAimOffset=~
     bPlaySlopeAimOffset=~
     JumpCombatAdditiveWeight=~
     MeleeTwistIdle=~
     MeleeTwistLocomotionLoop=~
     ImpactDisplacement=~
     SteerYaw=~
     EmoteHoverboardPosition=~
     EmoteHoverboardRotation=~
     bShouldAttachFeetToHoverboard=~
     FootLeftLocationOffset=~
     FootLeftRotationOffset=~
     FootRightLocationOffset=~
     FootRightRotationOffset=~
     StoppedThreshold=~
     MovingForwardThreshold=~
     MovingFowardFastThreshold=~
     DefaultToJumpStartTransitionThreshold=~
     JumpTrickAngularVelocityThreshold=~
     JumpDistanceFromGroundThreshold=~
     VelocityStartThreshold=~
     MovingFastThreshold=~
     RelaxedSpeedThreshold=~
     JumpChargeStartThreshold=~
     JumpChargeMidThreshold=~
     JumpChargeFullThreshold=~
     RotatingAngularVelocityThreshold=~
     IdleToLoopTransitionSpeedThreshold=~
     IdleToMovementLoopTransitionThreshold=~
     MeleeTwistIdleMultiplier=~
     MeleeTwistLocomotionLoopMultiplier=~
     LeanYawForMaxPivotPlayRate=~
     SteerAlphaForMaxPivotPlayRate=~
    ).bMovementLoopToMovementStopTransition)]
 }
 AnimationTransitionGraph_1 {
  path=VehicleFullBodyOverride.AnimGraphNode_StateMachine_0.HoverboardBR.AnimStateTransitionNode_4.AnimationTransitionGraph_1
  nodes [Result(bCanEnterTransition=Break Fort Anim Input Jackal Vehicle(FortAnimInput_JackalVehicle=Get JackalVehicleInput()) -> (
     bIsUsingJackalVehicle=~
     bIsSteeringLeft=~
     bIsSteeringRight=~
     SteerAngleDeadZoneDegrees=~
     SteerAngle=~
     LeanYaw=~
     QuantizedSteerAngle=~
     SteerAlpha=~
     RunForwardAlpha=~
     SlopePitchDegreeAngle=~
     SlopeRollDegreeAngle=~
     DistanceFromGround=~
     PivotPlayRate=~
     bIsReversing=~
     bIsBraking=~
     bIsMoving=~
     bIsMovingForward=~
     bIsSprinting=~
     bInAir=~
     bIsFalling=~
     bIsJumping=~
     bIsRelaxed=~
     bIsBoosting=~
     bHasReachedJumpChargeStartThreshold=~
     bHasReachedJumpChargeMidThreshold=~
     bHasReachedJumpChargeFullThreshold=~
     bAdjustRootForFemaleRider=~
     bIsPlayingEmoteOnHoverboard=~
     bShouldApplyAdditive=~
     bPlayPivotOnGroundAndNotBoosting=~
     bIsOnLowerHill=~
     bLowerBodyIdleToLoopTransition=~
     bInterruptHoverboardFullybody=~
     bEnterFullBodyHoverboardState=~
     bDefaultToJumpStartTransition=~
     bLocomotionPoseToJumpTransition=~
     bBoostingToBoostingJumpTransition=~
     bJumpToLocomotionPoseTransition=~
     bJumpApexToJumpFallTransition=~
     bIdleToMovementStartTransition=~
     bIdleToMovementLoopTransition=~
     bMovementLoopToMovementStopTransition=~
     bMovementLoopToPivotTransition=~
     bMovementLoopToIdleTransition=~
     bIdleAdditiveToCollisionNTransition=~
     bSplitBodyToHoverboardBRTransition=~
     bHoverboardBRToSplitBodyTransition=~
     bHoverboardBRMovementToJumpChargeTransition=~
     bIdlesToJackalVehicleTransition=~
     bPlayAdditiveLeans=~
     bPlayBalloonLeans=~
     bPlayJumpTrickVertical=~
     bPlayJumpTrick=~
     bPlayMovingFast=~
     bPlayHipAdjustmentAdditive=~
     bPlayDriveSouth=~
     bPlayHeadAimOffset=~
     bPlaySlopeAimOffset=~
     JumpCombatAdditiveWeight=~
     MeleeTwistIdle=~
     MeleeTwistLocomotionLoop=~
     ImpactDisplacement=~
     SteerYaw=~
     EmoteHoverboardPosition=~
     EmoteHoverboardRotation=~
     bShouldAttachFeetToHoverboard=~
     FootLeftLocationOffset=~
     FootLeftRotationOffset=~
     FootRightLocationOffset=~
     FootRightRotationOffset=~
     StoppedThreshold=~
     MovingForwardThreshold=~
     MovingFowardFastThreshold=~
     DefaultToJumpStartTransitionThreshold=~
     JumpTrickAngularVelocityThreshold=~
     JumpDistanceFromGroundThreshold=~
     VelocityStartThreshold=~
     MovingFastThreshold=~
     RelaxedSpeedThreshold=~
     JumpChargeStartThreshold=~
     JumpChargeMidThreshold=~
     JumpChargeFullThreshold=~
     RotatingAngularVelocityThreshold=~
     IdleToLoopTransitionSpeedThreshold=~
     IdleToMovementLoopTransitionThreshold=~
     MeleeTwistIdleMultiplier=~
     MeleeTwistLocomotionLoopMultiplier=~
     LeanYawForMaxPivotPlayRate=~
     SteerAlphaForMaxPivotPlayRate=~
    ).bIdleToMovementStartTransition)]
 }
 AnimationTransitionGraph_1 {
  path=VehicleFullBodyOverride.AnimGraphNode_StateMachine_0.HoverboardBR.AnimStateTransitionNode_5.AnimationTransitionGraph_1
  nodes [Result()]
 }
 AnimationTransitionGraph_1 {
  path=VehicleFullBodyOverride.AnimGraphNode_StateMachine_0.HoverboardBR.AnimStateTransitionNode_7.AnimationTransitionGraph_1
  nodes [Result()]
 }
 Transition {
  path=VehicleFullBodyOverride.AnimGraphNode_StateMachine_0.HoverboardBR.AnimStateTransitionNode_9.Transition
  nodes [Result(bCanEnterTransition=Break Fort Anim Input Jackal Vehicle(FortAnimInput_JackalVehicle=Get JackalVehicleInput()) -> (
     bIsUsingJackalVehicle=~
     bIsSteeringLeft=~
     bIsSteeringRight=~
     SteerAngleDeadZoneDegrees=~
     SteerAngle=~
     LeanYaw=~
     QuantizedSteerAngle=~
     SteerAlpha=~
     RunForwardAlpha=~
     SlopePitchDegreeAngle=~
     SlopeRollDegreeAngle=~
     DistanceFromGround=~
     PivotPlayRate=~
     bIsReversing=~
     bIsBraking=~
     bIsMoving=~
     bIsMovingForward=~
     bIsSprinting=~
     bInAir=~
     bIsFalling=~
     bIsJumping=~
     bIsRelaxed=~
     bIsBoosting=~
     bHasReachedJumpChargeStartThreshold=~
     bHasReachedJumpChargeMidThreshold=~
     bHasReachedJumpChargeFullThreshold=~
     bAdjustRootForFemaleRider=~
     bIsPlayingEmoteOnHoverboard=~
     bShouldApplyAdditive=~
     bPlayPivotOnGroundAndNotBoosting=~
     bIsOnLowerHill=~
     bLowerBodyIdleToLoopTransition=~
     bInterruptHoverboardFullybody=~
     bEnterFullBodyHoverboardState=~
     bDefaultToJumpStartTransition=~
     bLocomotionPoseToJumpTransition=~
     bBoostingToBoostingJumpTransition=~
     bJumpToLocomotionPoseTransition=~
     bJumpApexToJumpFallTransition=~
     bIdleToMovementStartTransition=~
     bIdleToMovementLoopTransition=~
     bMovementLoopToMovementStopTransition=~
     bMovementLoopToPivotTransition=~
     bMovementLoopToIdleTransition=~
     bIdleAdditiveToCollisionNTransition=~
     bSplitBodyToHoverboardBRTransition=~
     bHoverboardBRToSplitBodyTransition=~
     bHoverboardBRMovementToJumpChargeTransition=~
     bIdlesToJackalVehicleTransition=~
     bPlayAdditiveLeans=~
     bPlayBalloonLeans=~
     bPlayJumpTrickVertical=~
     bPlayJumpTrick=~
     bPlayMovingFast=~
     bPlayHipAdjustmentAdditive=~
     bPlayDriveSouth=~
     bPlayHeadAimOffset=~
     bPlaySlopeAimOffset=~
     JumpCombatAdditiveWeight=~
     MeleeTwistIdle=~
     MeleeTwistLocomotionLoop=~
     ImpactDisplacement=~
     SteerYaw=~
     EmoteHoverboardPosition=~
     EmoteHoverboardRotation=~
     bShouldAttachFeetToHoverboard=~
     FootLeftLocationOffset=~
     FootLeftRotationOffset=~
     FootRightLocationOffset=~
     FootRightRotationOffset=~
     StoppedThreshold=~
     MovingForwardThreshold=~
     MovingFowardFastThreshold=~
     DefaultToJumpStartTransitionThreshold=~
     JumpTrickAngularVelocityThreshold=~
     JumpDistanceFromGroundThreshold=~
     VelocityStartThreshold=~
     MovingFastThreshold=~
     RelaxedSpeedThreshold=~
     JumpChargeStartThreshold=~
     JumpChargeMidThreshold=~
     JumpChargeFullThreshold=~
     RotatingAngularVelocityThreshold=~
     IdleToLoopTransitionSpeedThreshold=~
     IdleToMovementLoopTransitionThreshold=~
     MeleeTwistIdleMultiplier=~
     MeleeTwistLocomotionLoopMultiplier=~
     LeanYawForMaxPivotPlayRate=~
     SteerAlphaForMaxPivotPlayRate=~
    ).bIdleToMovementLoopTransition)]
 }
 HoverboardBRFullBody {
  entry=HoverboardBRMovement
  path=VehicleFullBodyOverride.AnimGraphNode_StateMachine_1.HoverboardBRFullBody
  ids=id
  nodes [
   AnimStateNode_0 = HoverboardBRJumpChargeAdditive() {is=title AnimStateNode}
   AnimStateNode_2 = HoverboardBRMovement() {is=title AnimStateNode}
   AnimStateTransitionNode_1 = AnimStateTransitionNode() {crossfadeDuration=0.35}
   AnimStateTransitionNode_12 = AnimStateTransitionNode()
   AnimStateTransitionNode_14 = AnimStateTransitionNode() {crossfadeDuration=0.25}
  ]
  links [
   {
    dst=HoverboardBRMovement
    rule=VehicleFullBodyOverride.AnimGraphNode_StateMachine_1.HoverboardBRFullBody.AnimStateTransitionNode_1.Transition
    src=HoverboardBRJumpChargeAdditive
    via=AnimStateTransitionNode_1
   }
   {
    dst=HoverboardBRMovement
    rule=VehicleFullBodyOverride.AnimGraphNode_StateMachine_1.HoverboardBRFullBody.AnimStateTransitionNode_12.Transition
    src=HoverboardBRJumpChargeAdditive
    via=AnimStateTransitionNode_12
   }
   {
    dst=HoverboardBRJumpChargeAdditive
    rule=VehicleFullBodyOverride.AnimGraphNode_StateMachine_1.HoverboardBRFullBody.AnimStateTransitionNode_14.Transition
    src=HoverboardBRMovement
    via=AnimStateTransitionNode_14
   }
  ]
 }
 HoverboardBRJumpChargeAdditive {
  path=VehicleFullBodyOverride.AnimGraphNode_StateMachine_1.HoverboardBRFullBody.AnimStateNode_0.HoverboardBRJumpChargeAdditive
  nodes [
   n1 = Break Fort Anim Input Jackal Vehicle(FortAnimInput_JackalVehicle=Get JackalVehicleInput()) -> (
    bIsUsingJackalVehicle=~
    bIsSteeringLeft=~
    bIsSteeringRight=~
    SteerAngleDeadZoneDegrees=~
    SteerAngle=~
    LeanYaw=~
    QuantizedSteerAngle=~
    SteerAlpha=~
    RunForwardAlpha=~
    SlopePitchDegreeAngle=~
    SlopeRollDegreeAngle=~
    DistanceFromGround=~
    PivotPlayRate=~
    bIsReversing=~
    bIsBraking=~
    bIsMoving=~
    bIsMovingForward=~
    bIsSprinting=~
    bInAir=~
    bIsFalling=~
    bIsJumping=~
    bIsRelaxed=~
    bIsBoosting=~
    bHasReachedJumpChargeStartThreshold=~
    bHasReachedJumpChargeMidThreshold=~
    bHasReachedJumpChargeFullThreshold=~
    bAdjustRootForFemaleRider=~
    bIsPlayingEmoteOnHoverboard=~
    bShouldApplyAdditive=~
    bPlayPivotOnGroundAndNotBoosting=~
    bIsOnLowerHill=~
    bLowerBodyIdleToLoopTransition=~
    bInterruptHoverboardFullybody=~
    bEnterFullBodyHoverboardState=~
    bDefaultToJumpStartTransition=~
    bLocomotionPoseToJumpTransition=~
    bBoostingToBoostingJumpTransition=~
    bJumpToLocomotionPoseTransition=~
    bJumpApexToJumpFallTransition=~
    bIdleToMovementStartTransition=~
    bIdleToMovementLoopTransition=~
    bMovementLoopToMovementStopTransition=~
    bMovementLoopToPivotTransition=~
    bMovementLoopToIdleTransition=~
    bIdleAdditiveToCollisionNTransition=~
    bSplitBodyToHoverboardBRTransition=~
    bHoverboardBRToSplitBodyTransition=~
    bHoverboardBRMovementToJumpChargeTransition=~
    bIdlesToJackalVehicleTransition=~
    bPlayAdditiveLeans=~
    bPlayBalloonLeans=~
    bPlayJumpTrickVertical=~
    bPlayJumpTrick=~
    bPlayMovingFast=~
    bPlayHipAdjustmentAdditive=~
    bPlayDriveSouth=~
    bPlayHeadAimOffset=~
    bPlaySlopeAimOffset=~
    JumpCombatAdditiveWeight=~
    MeleeTwistIdle=~
    MeleeTwistLocomotionLoop=~
    ImpactDisplacement=~
    SteerYaw=~
    EmoteHoverboardPosition=~
    EmoteHoverboardRotation=~
    bShouldAttachFeetToHoverboard=~
    FootLeftLocationOffset=~
    FootLeftRotationOffset=~
    FootRightLocationOffset=~
    FootRightRotationOffset=~
    StoppedThreshold=~
    MovingForwardThreshold=~
    MovingFowardFastThreshold=~
    DefaultToJumpStartTransitionThreshold=~
    JumpTrickAngularVelocityThreshold=~
    JumpDistanceFromGroundThreshold=~
    VelocityStartThreshold=~
    MovingFastThreshold=~
    RelaxedSpeedThreshold=~
    JumpChargeStartThreshold=~
    JumpChargeMidThreshold=~
    JumpChargeFullThreshold=~
    RotatingAngularVelocityThreshold=~
    IdleToLoopTransitionSpeedThreshold=~
    IdleToMovementLoopTransitionThreshold=~
    MeleeTwistIdleMultiplier=~
    MeleeTwistLocomotionLoopMultiplier=~
    LeanYawForMaxPivotPlayRate=~
    SteerAlphaForMaxPivotPlayRate=~
   )
   n2 = Blend Poses by bool(
    BlendPose_0=Sequence Player 'DBNOCarry_JackalVehicle_JumpCharge_Pose'() {sequence=/Game/Animation/Game/MainPlayer/Vehicles/JackalVehicle/DBNO/DBNOCarry_JackalVehicle_JumpCharge_Pose}
    BlendPose_1=Sequence Player 'DBNOCarry_JackalVehicle_JumpCharge_Start'() {bLoopAnimation=false, sequence=/Game/Animation/Game/MainPlayer/Vehicles/JackalVehicle/DBNO/DBNOCarry_JackalVehicle_JumpCharge_Start}
    bActiveValue=@n1.bHasReachedJumpChargeFullThreshold
   ) {blendType=Linear}
   n3 = Sequence Player 'JackalVehicle_JumpCharge_Start'() {
    bLoopAnimation=false
    groupName=ChargeStart
    method=SyncGroup
    sequence=/Game/Animation/Game/MainPlayer/Vehicles/JackalVehicle/JackalVehicle_JumpCharge_Start
   }
   n4 = Blend Poses by bool(
    BlendPose_0=Sequence Player 'JackalVehicle_JumpCharge_Pose'() {sequence=/Game/Animation/Game/MainPlayer/Vehicles/JackalVehicle/JackalVehicle_JumpCharge_Pose}
    BlendPose_1=@n3
    bActiveValue=@n1.bHasReachedJumpChargeFullThreshold
   ) {blendType=Linear}
   n5 = Get bIsDBNOCarrying()
   n6 = AimOffset Player 'JackalVehicle_Slope_Pitch_AO_SprintJumpCharge'!(
    X:float=@n1.SlopePitchDegreeAngle
    BasePose:Pose Link Structure=Apply Additive(
     Base=Blend Poses by bool(BlendPose_0=@n2, BlendPose_1=@n4, bActiveValue=@n5) {blendType=Linear}
     Additive=Blendspace Player 'JackalVehicle_JumpCharging_Leans_BS'!(X=@n1.LeanYaw) -> (Pose) {blendSpace=/Game/Animation/Game/MainPlayer/Vehicles/JackalVehicle/JackalVehicle_JumpCharging_Leans_BS}
     bAlphaBoolEnabled=@n1.bHasReachedJumpChargeMidThreshold
    ) {alphaInputType=Bool, alphaBoolBlend {blendInTime=0.15, blendOption=Linear, blendOutTime=0.15}}
    Alpha:float=1
   ) -> (Pose:Pose Link Structure) {
    blendSpace=/Game/Animation/Game/MainPlayer/Vehicles/JackalVehicle/AO/JackalVehicle_Slope_Pitch_AO_SprintJumpCharge
    alphaBoolBlend {blendOption=Linear}
   }
   n7 = AimOffset Player 'JackalVehicle_Slope_Roll_AO_SprintJumpCharge'!(X:float=@n1.SlopeRollDegreeAngle, BasePose:Pose Link Structure=@n6, Alpha:float=1) -> (Pose:Pose Link Structure) {
    blendSpace=/Game/Animation/Game/MainPlayer/Vehicles/JackalVehicle/AO/JackalVehicle_Slope_Roll_AO_SprintJumpCharge
    alphaBoolBlend {blendOption=Linear}
   }
   Output Animation Pose(Result=@n7)
  ]
 }
 HoverboardBRMovement {
  path=VehicleFullBodyOverride.AnimGraphNode_StateMachine_1.HoverboardBRFullBody.AnimStateNode_2.HoverboardBRMovement
  nodes [
   AnimGraphNode_StateMachine_0 = HoverboardBRFunction() {is=title AnimGraphNode_StateMachine}
   n1 = Break Fort Anim Input Jackal Vehicle(FortAnimInput_JackalVehicle=Get JackalVehicleInput()) -> (
    bIsUsingJackalVehicle=~
    bIsSteeringLeft=~
    bIsSteeringRight=~
    SteerAngleDeadZoneDegrees=~
    SteerAngle=~
    LeanYaw=~
    QuantizedSteerAngle=~
    SteerAlpha=~
    RunForwardAlpha=~
    SlopePitchDegreeAngle=~
    SlopeRollDegreeAngle=~
    DistanceFromGround=~
    PivotPlayRate=~
    bIsReversing=~
    bIsBraking=~
    bIsMoving=~
    bIsMovingForward=~
    bIsSprinting=~
    bInAir=~
    bIsFalling=~
    bIsJumping=~
    bIsRelaxed=~
    bIsBoosting=~
    bHasReachedJumpChargeStartThreshold=~
    bHasReachedJumpChargeMidThreshold=~
    bHasReachedJumpChargeFullThreshold=~
    bAdjustRootForFemaleRider=~
    bIsPlayingEmoteOnHoverboard=~
    bShouldApplyAdditive=~
    bPlayPivotOnGroundAndNotBoosting=~
    bIsOnLowerHill=~
    bLowerBodyIdleToLoopTransition=~
    bInterruptHoverboardFullybody=~
    bEnterFullBodyHoverboardState=~
    bDefaultToJumpStartTransition=~
    bLocomotionPoseToJumpTransition=~
    bBoostingToBoostingJumpTransition=~
    bJumpToLocomotionPoseTransition=~
    bJumpApexToJumpFallTransition=~
    bIdleToMovementStartTransition=~
    bIdleToMovementLoopTransition=~
    bMovementLoopToMovementStopTransition=~
    bMovementLoopToPivotTransition=~
    bMovementLoopToIdleTransition=~
    bIdleAdditiveToCollisionNTransition=~
    bSplitBodyToHoverboardBRTransition=~
    bHoverboardBRToSplitBodyTransition=~
    bHoverboardBRMovementToJumpChargeTransition=~
    bIdlesToJackalVehicleTransition=~
    bPlayAdditiveLeans=~
    bPlayBalloonLeans=~
    bPlayJumpTrickVertical=~
    bPlayJumpTrick=~
    bPlayMovingFast=~
    bPlayHipAdjustmentAdditive=~
    bPlayDriveSouth=~
    bPlayHeadAimOffset=~
    bPlaySlopeAimOffset=~
    JumpCombatAdditiveWeight=~
    MeleeTwistIdle=~
    MeleeTwistLocomotionLoop=~
    ImpactDisplacement=~
    SteerYaw=~
    EmoteHoverboardPosition=~
    EmoteHoverboardRotation=~
    bShouldAttachFeetToHoverboard=~
    FootLeftLocationOffset=~
    FootLeftRotationOffset=~
    FootRightLocationOffset=~
    FootRightRotationOffset=~
    StoppedThreshold=~
    MovingForwardThreshold=~
    MovingFowardFastThreshold=~
    DefaultToJumpStartTransitionThreshold=~
    JumpTrickAngularVelocityThreshold=~
    JumpDistanceFromGroundThreshold=~
    VelocityStartThreshold=~
    MovingFastThreshold=~
    RelaxedSpeedThreshold=~
    JumpChargeStartThreshold=~
    JumpChargeMidThreshold=~
    JumpChargeFullThreshold=~
    RotatingAngularVelocityThreshold=~
    IdleToLoopTransitionSpeedThreshold=~
    IdleToMovementLoopTransitionThreshold=~
    MeleeTwistIdleMultiplier=~
    MeleeTwistLocomotionLoopMultiplier=~
    LeanYawForMaxPivotPlayRate=~
    SteerAlphaForMaxPivotPlayRate=~
   )
   n2 = AimOffset Player 'JackalVehicle_Driver_Head_AO'!(
    X:float=Get AimYaw() -> (AimYaw:float)
    Y:float=Get AimPitch()
    BasePose:Pose Link Structure=@AnimGraphNode_StateMachine_0
    bAlphaBoolEnabled:bool=@n1.bPlayHeadAimOffset
   ) -> (Pose:Pose Link Structure) {
    alphaInputType=Bool
    blendSpace=/Game/Animation/Game/MainPlayer/Vehicles/JackalVehicle/AO/JackalVehicle_Driver_Head_AO
    alphaBoolBlend {blendInTime=0.3, blendOutTime=0.4}
   }
   n3 = AimOffset Player 'JackalVehicle_Slope_Pitch_AO'!(X:float=@n1.SlopePitchDegreeAngle, BasePose:Pose Link Structure=@n2, bAlphaBoolEnabled:bool=@n1.bPlaySlopeAimOffset) -> (Pose:Pose Link Structure) {
    alphaInputType=Bool
    blendSpace=/Game/Animation/Game/MainPlayer/Vehicles/JackalVehicle/AO/JackalVehicle_Slope_Pitch_AO
    alphaBoolBlend {blendInTime=0.2, blendOption=Linear, blendOutTime=0.2}
   }
   n4 = AimOffset Player 'JackalVehicle_Slope_Roll_AO'!(X:float=@n1.SlopeRollDegreeAngle, BasePose:Pose Link Structure=@n3, bAlphaBoolEnabled:bool=@n1.bPlaySlopeAimOffset) -> (Pose:Pose Link Structure) {
    alphaInputType=Bool
    blendSpace=/Game/Animation/Game/MainPlayer/Vehicles/JackalVehicle/AO/JackalVehicle_Slope_Roll_AO
    alphaBoolBlend {blendInTime=0.2, blendOption=Linear, blendOutTime=0.2}
   }
   Output Animation Pose(Result=@n4)
  ]
 }
 HoverboardBRFunction {
  entry=LocomotionPose
  path=VehicleFullBodyOverride.AnimGraphNode_StateMachine_1.HoverboardBRFullBody.AnimStateNode_2.HoverboardBRMovement.AnimGraphNode_StateMachine_0.HoverboardBRFunction
  ids=id
  nodes [
   AnimStateNode_0 = LocomotionPose() {is=title AnimStateNode}
   AnimStateNode_1 = BoostingStart() {is=title AnimStateNode}
   AnimStateNode_2 = Jump() {is=title AnimStateNode}
   AnimStateNode_3 = BoostingEnd() {is=title AnimStateNode}
   AnimStateNode_4 = BoostingJump() {is=title AnimStateNode}
   AnimStateNode_5 = BoostingLoop() {is=title AnimStateNode}
   AnimStateNode_6 = BoostingJumpLoop() {is=title AnimStateNode}
   AnimStateTransitionNode_0 = AnimStateTransitionNode() {crossfadeDuration=0.1}
   AnimStateTransitionNode_1 = AnimStateTransitionNode() {bAutomaticRuleBasedOnSequencePlayerInState=true, crossfadeDuration=0.3, priorityOrder=2}
   AnimStateTransitionNode_10 = AnimStateTransitionNode() {crossfadeDuration=0.1}
   AnimStateTransitionNode_11 = AnimStateTransitionNode()
   AnimStateTransitionNode_2 = AnimStateTransitionNode() {crossfadeDuration=0.1, priorityOrder=2}
   AnimStateTransitionNode_3 = AnimStateTransitionNode() {bAutomaticRuleBasedOnSequencePlayerInState=true}
   AnimStateTransitionNode_4 = AnimStateTransitionNode() {bAutomaticRuleBasedOnSequencePlayerInState=true, crossfadeDuration=0.35, priorityOrder=2}
   AnimStateTransitionNode_5 = AnimStateTransitionNode() {crossfadeDuration=0.1}
   AnimStateTransitionNode_6 = AnimStateTransitionNode() {crossfadeDuration=0.1}
   AnimStateTransitionNode_7 = AnimStateTransitionNode() {crossfadeDuration=0.6}
   AnimStateTransitionNode_8 = AnimStateTransitionNode() {crossfadeDuration=0}
   AnimStateTransitionNode_9 = AnimStateTransitionNode() {crossfadeDuration=0.1}
  ]
  links [
   {
    dst=BoostingJump
    rule=VehicleFullBodyOverride.AnimGraphNode_StateMachine_1.HoverboardBRFullBody.AnimStateNode_2.HoverboardBRMovement.AnimGraphNode_StateMachine_0.HoverboardBRFunction.AnimStateTransitionNode_10.Transition
    src=BoostingEnd
    via=AnimStateTransitionNode_10
   }
   {
    dst=LocomotionPose
    rule=VehicleFullBodyOverride.AnimGraphNode_StateMachine_1.HoverboardBRFullBody.AnimStateNode_2.HoverboardBRMovement.AnimGraphNode_StateMachine_0.HoverboardBRFunction.AnimStateTransitionNode_4.AnimationTransitionGraph_1
    src=BoostingEnd
    via=AnimStateTransitionNode_4
   }
   {
    dst=BoostingJumpLoop
    rule=VehicleFullBodyOverride.AnimGraphNode_StateMachine_1.HoverboardBRFullBody.AnimStateNode_2.HoverboardBRMovement.AnimGraphNode_StateMachine_0.HoverboardBRFunction.AnimStateTransitionNode_3.AnimationTransitionGraph_0
    src=BoostingJump
    via=AnimStateTransitionNode_3
   }
   {
    dst=BoostingEnd
    rule=VehicleFullBodyOverride.AnimGraphNode_StateMachine_1.HoverboardBRFullBody.AnimStateNode_2.HoverboardBRMovement.AnimGraphNode_StateMachine_0.HoverboardBRFunction.AnimStateTransitionNode_11.Transition
    src=BoostingJumpLoop
    via=AnimStateTransitionNode_11
   }
   {
    dst=BoostingEnd
    rule=VehicleFullBodyOverride.AnimGraphNode_StateMachine_1.HoverboardBRFullBody.AnimStateNode_2.HoverboardBRMovement.AnimGraphNode_StateMachine_0.HoverboardBRFunction.AnimStateTransitionNode_2.AnimationTransitionGraph_2
    src=BoostingLoop
    via=AnimStateTransitionNode_2
   }
   {
    dst=BoostingJump
    rule=VehicleFullBodyOverride.AnimGraphNode_StateMachine_1.HoverboardBRFullBody.AnimStateNode_2.HoverboardBRMovement.AnimGraphNode_StateMachine_0.HoverboardBRFunction.AnimStateTransitionNode_9.Transition
    src=BoostingLoop
    via=AnimStateTransitionNode_9
   }
   {
    dst=BoostingJump
    rule=VehicleFullBodyOverride.AnimGraphNode_StateMachine_1.HoverboardBRFullBody.AnimStateNode_2.HoverboardBRMovement.AnimGraphNode_StateMachine_0.HoverboardBRFunction.AnimStateTransitionNode_5.Transition
    src=BoostingStart
    via=AnimStateTransitionNode_5
   }
   {
    dst=BoostingLoop
    rule=VehicleFullBodyOverride.AnimGraphNode_StateMachine_1.HoverboardBRFullBody.AnimStateNode_2.HoverboardBRMovement.AnimGraphNode_StateMachine_0.HoverboardBRFunction.AnimStateTransitionNode_1.AnimationTransitionGraph_2
    src=BoostingStart
    via=AnimStateTransitionNode_1
   }
   {
    dst=BoostingStart
    rule=VehicleFullBodyOverride.AnimGraphNode_StateMachine_1.HoverboardBRFullBody.AnimStateNode_2.HoverboardBRMovement.AnimGraphNode_StateMachine_0.HoverboardBRFunction.AnimStateTransitionNode_8.Transition
    src=Jump
    via=AnimStateTransitionNode_8
   }
   {
    dst=LocomotionPose
    rule=VehicleFullBodyOverride.AnimGraphNode_StateMachine_1.HoverboardBRFullBody.AnimStateNode_2.HoverboardBRMovement.AnimGraphNode_StateMachine_0.HoverboardBRFunction.AnimStateTransitionNode_7.Transition
    src=Jump
    via=AnimStateTransitionNode_7
   }
   {
    dst=BoostingStart
    rule=VehicleFullBodyOverride.AnimGraphNode_StateMachine_1.HoverboardBRFullBody.AnimStateNode_2.HoverboardBRMovement.AnimGraphNode_StateMachine_0.HoverboardBRFunction.AnimStateTransitionNode_0.Transition
    src=LocomotionPose
    via=AnimStateTransitionNode_0
   }
   {
    dst=Jump
    rule=VehicleFullBodyOverride.AnimGraphNode_StateMachine_1.HoverboardBRFullBody.AnimStateNode_2.HoverboardBRMovement.AnimGraphNode_StateMachine_0.HoverboardBRFunction.AnimStateTransitionNode_6.Transition
    src=LocomotionPose
    via=AnimStateTransitionNode_6
   }
  ]
 }
 LocomotionPose {
  path=VehicleFullBodyOverride.AnimGraphNode_StateMachine_1.HoverboardBRFullBody.AnimStateNode_2.HoverboardBRMovement.AnimGraphNode_StateMachine_0.HoverboardBRFunction.AnimStateNode_0.LocomotionPose
  nodes [Output Animation Pose(Result=Use cached pose 'HoverboardBRPose'())]
 }
 BoostingStart {
  path=VehicleFullBodyOverride.AnimGraphNode_StateMachine_1.HoverboardBRFullBody.AnimStateNode_2.HoverboardBRMovement.AnimGraphNode_StateMachine_0.HoverboardBRFunction.AnimStateNode_1.BoostingStart
  nodes [Output Animation Pose(Result=Blend Poses by bool(
     BlendPose_0=Sequence Player 'DBNOCarry_JackalVehicle_Boosting_Start'() {bLoopAnimation=false, sequence=/Game/Animation/Game/MainPlayer/Vehicles/JackalVehicle/DBNO/DBNOCarry_JackalVehicle_Boosting_Start}
     BlendPose_1=Sequence Player 'JackalVehicle_Boosting_Start'() {bLoopAnimation=false, sequence=/Game/Animation/Game/MainPlayer/Vehicles/JackalVehicle/JackalVehicle_Boosting_Start}
     bActiveValue=Get bIsDBNOCarrying()
    ) {blendType=Linear})]
 }
 Jump {
  path=VehicleFullBodyOverride.AnimGraphNode_StateMachine_1.HoverboardBRFullBody.AnimStateNode_2.HoverboardBRMovement.AnimGraphNode_StateMachine_0.HoverboardBRFunction.AnimStateNode_2.Jump
  nodes [
   AnimGraphNode_StateMachine_0 = HoverboardJump() {is=title AnimGraphNode_StateMachine}
   Output Animation Pose(Result=@AnimGraphNode_StateMachine_0)
  ]
 }
 HoverboardJump {
  entry=Default
  path=VehicleFullBodyOverride.AnimGraphNode_StateMachine_1.HoverboardBRFullBody.AnimStateNode_2.HoverboardBRMovement.AnimGraphNode_StateMachine_0.HoverboardBRFunction.AnimStateNode_2.Jump.AnimGraphNode_StateMachine_0.HoverboardJump
  ids=id
  nodes [
   AnimStateNode_0 = Default() {is=title AnimStateNode, bAlwaysResetOnEntry=true}
   AnimStateNode_11 = JumpStart() {is=title AnimStateNode, bAlwaysResetOnEntry=true}
   AnimStateNode_2 = JumpFall() {is=title AnimStateNode}
   AnimStateNode_7 = JumpLand() {is=title AnimStateNode}
   AnimStateNode_9 = JumpApex() {is=title AnimStateNode}
   AnimStateTransitionNode_11 = AnimStateTransitionNode() {bAutomaticRuleBasedOnSequencePlayerInState=true}
   AnimStateTransitionNode_13 = AnimStateTransitionNode() {crossfadeDuration=0.1}
   AnimStateTransitionNode_19 = AnimStateTransitionNode() {crossfadeDuration=0.15}
   AnimStateTransitionNode_3 = AnimStateTransitionNode() {crossfadeDuration=0.35}
   AnimStateTransitionNode_31 = AnimStateTransitionNode() {crossfadeDuration=0.35}
   AnimStateTransitionNode_4 = AnimStateTransitionNode() {crossfadeDuration=0.15}
   AnimStateTransitionNode_5 = AnimStateTransitionNode() {bAutomaticRuleBasedOnSequencePlayerInState=true}
  ]
  links [
   {
    dst=JumpStart
    rule=VehicleFullBodyOverride.AnimGraphNode_StateMachine_1.HoverboardBRFullBody.AnimStateNode_2.HoverboardBRMovement.AnimGraphNode_StateMachine_0.HoverboardBRFunction.AnimStateNode_2.Jump.AnimGraphNode_StateMachine_0.HoverboardJump.AnimStateTransitionNode_4.Transition
    src=Default
    via=AnimStateTransitionNode_4
   }
   {
    dst=JumpFall
    rule=VehicleFullBodyOverride.AnimGraphNode_StateMachine_1.HoverboardBRFullBody.AnimStateNode_2.HoverboardBRMovement.AnimGraphNode_StateMachine_0.HoverboardBRFunction.AnimStateNode_2.Jump.AnimGraphNode_StateMachine_0.HoverboardJump.AnimStateTransitionNode_3.AnimationTransitionGraph_1
    src=JumpApex
    via=AnimStateTransitionNode_3
   }
   {
    dst=JumpFall
    rule=VehicleFullBodyOverride.AnimGraphNode_StateMachine_1.HoverboardBRFullBody.AnimStateNode_2.HoverboardBRMovement.AnimGraphNode_StateMachine_0.HoverboardBRFunction.AnimStateNode_2.Jump.AnimGraphNode_StateMachine_0.HoverboardJump.AnimStateTransitionNode_31.AnimationTransitionGraph_1
    src=JumpApex
    via=AnimStateTransitionNode_31
   }
   {
    dst=JumpLand
    rule=VehicleFullBodyOverride.AnimGraphNode_StateMachine_1.HoverboardBRFullBody.AnimStateNode_2.HoverboardBRMovement.AnimGraphNode_StateMachine_0.HoverboardBRFunction.AnimStateNode_2.Jump.AnimGraphNode_StateMachine_0.HoverboardJump.AnimStateTransitionNode_13.AnimationTransitionGraph_1
    src=JumpFall
    via=AnimStateTransitionNode_13
   }
   {
    dst=Default
    rule=VehicleFullBodyOverride.AnimGraphNode_StateMachine_1.HoverboardBRFullBody.AnimStateNode_2.HoverboardBRMovement.AnimGraphNode_StateMachine_0.HoverboardBRFunction.AnimStateNode_2.Jump.AnimGraphNode_StateMachine_0.HoverboardJump.AnimStateTransitionNode_5.Transition
    src=JumpLand
    via=AnimStateTransitionNode_5
   }
   {
    dst=JumpStart
    rule=VehicleFullBodyOverride.AnimGraphNode_StateMachine_1.HoverboardBRFullBody.AnimStateNode_2.HoverboardBRMovement.AnimGraphNode_StateMachine_0.HoverboardBRFunction.AnimStateNode_2.Jump.AnimGraphNode_StateMachine_0.HoverboardJump.AnimStateTransitionNode_19.AnimationTransitionGraph_1
    src=JumpLand
    via=AnimStateTransitionNode_19
   }
   {
    dst=JumpApex
    rule=VehicleFullBodyOverride.AnimGraphNode_StateMachine_1.HoverboardBRFullBody.AnimStateNode_2.HoverboardBRMovement.AnimGraphNode_StateMachine_0.HoverboardBRFunction.AnimStateNode_2.Jump.AnimGraphNode_StateMachine_0.HoverboardJump.AnimStateTransitionNode_11.AnimationTransitionGraph_1
    src=JumpStart
    via=AnimStateTransitionNode_11
   }
  ]
 }
 Default {
  path=VehicleFullBodyOverride.AnimGraphNode_StateMachine_1.HoverboardBRFullBody.AnimStateNode_2.HoverboardBRMovement.AnimGraphNode_StateMachine_0.HoverboardBRFunction.AnimStateNode_2.Jump.AnimGraphNode_StateMachine_0.HoverboardJump.AnimStateNode_0.Default
  nodes [Output Animation Pose(Result=Blend Poses by bool(
     BlendPose_0=Sequence Player 'JackalVehicle_Lean_Center_Balloons'() {sequence=/Game/Animation/Game/MainPlayer/Vehicles/JackalVehicle/JackalVehicle_Lean_Center_Balloons}
     BlendPose_1=Use cached pose 'HoverboardBRPose'()
     BlendTime_0=0.2
     BlendTime_1=0.2
     bActiveValue=Get bHasBalloons() -> (bHasBalloons:bool)
    ) {blendType=Linear})]
 }
 JumpStart {
  path=VehicleFullBodyOverride.AnimGraphNode_StateMachine_1.HoverboardBRFullBody.AnimStateNode_2.HoverboardBRMovement.AnimGraphNode_StateMachine_0.HoverboardBRFunction.AnimStateNode_2.Jump.AnimGraphNode_StateMachine_0.HoverboardJump.AnimStateNode_11.JumpStart
  nodes [
   n1 = Sequence Player 'DBNOCarry_JackalVehicle_JumpNormal_Start'() {
    bLoopAnimation=false
    sequence=/Game/Animation/Game/MainPlayer/Vehicles/JackalVehicle/DBNO/DBNOCarry_JackalVehicle_JumpNormal_Start
    startPosition=0.067
   }
   n2 = Sequence Player 'DBNOCarry_JackalVehicle_JumpTrick_Vertical_Start'() {
    bLoopAnimation=false
    sequence=/Game/Animation/Game/MainPlayer/Vehicles/JackalVehicle/DBNO/DBNOCarry_JackalVehicle_JumpTrick_Vertical_Start
    startPosition=0.067
   }
   n3 = Break Fort Anim Input Jackal Vehicle(FortAnimInput_JackalVehicle=Get JackalVehicleInput()) -> (
    bIsUsingJackalVehicle=~
    bIsSteeringLeft=~
    bIsSteeringRight=~
    SteerAngleDeadZoneDegrees=~
    SteerAngle=~
    LeanYaw=~
    QuantizedSteerAngle=~
    SteerAlpha=~
    RunForwardAlpha=~
    SlopePitchDegreeAngle=~
    SlopeRollDegreeAngle=~
    DistanceFromGround=~
    PivotPlayRate=~
    bIsReversing=~
    bIsBraking=~
    bIsMoving=~
    bIsMovingForward=~
    bIsSprinting=~
    bInAir=~
    bIsFalling=~
    bIsJumping=~
    bIsRelaxed=~
    bIsBoosting=~
    bHasReachedJumpChargeStartThreshold=~
    bHasReachedJumpChargeMidThreshold=~
    bHasReachedJumpChargeFullThreshold=~
    bAdjustRootForFemaleRider=~
    bIsPlayingEmoteOnHoverboard=~
    bShouldApplyAdditive=~
    bPlayPivotOnGroundAndNotBoosting=~
    bIsOnLowerHill=~
    bLowerBodyIdleToLoopTransition=~
    bInterruptHoverboardFullybody=~
    bEnterFullBodyHoverboardState=~
    bDefaultToJumpStartTransition=~
    bLocomotionPoseToJumpTransition=~
    bBoostingToBoostingJumpTransition=~
    bJumpToLocomotionPoseTransition=~
    bJumpApexToJumpFallTransition=~
    bIdleToMovementStartTransition=~
    bIdleToMovementLoopTransition=~
    bMovementLoopToMovementStopTransition=~
    bMovementLoopToPivotTransition=~
    bMovementLoopToIdleTransition=~
    bIdleAdditiveToCollisionNTransition=~
    bSplitBodyToHoverboardBRTransition=~
    bHoverboardBRToSplitBodyTransition=~
    bHoverboardBRMovementToJumpChargeTransition=~
    bIdlesToJackalVehicleTransition=~
    bPlayAdditiveLeans=~
    bPlayBalloonLeans=~
    bPlayJumpTrickVertical=~
    bPlayJumpTrick=~
    bPlayMovingFast=~
    bPlayHipAdjustmentAdditive=~
    bPlayDriveSouth=~
    bPlayHeadAimOffset=~
    bPlaySlopeAimOffset=~
    JumpCombatAdditiveWeight=~
    MeleeTwistIdle=~
    MeleeTwistLocomotionLoop=~
    ImpactDisplacement=~
    SteerYaw=~
    EmoteHoverboardPosition=~
    EmoteHoverboardRotation=~
    bShouldAttachFeetToHoverboard=~
    FootLeftLocationOffset=~
    FootLeftRotationOffset=~
    FootRightLocationOffset=~
    FootRightRotationOffset=~
    StoppedThreshold=~
    MovingForwardThreshold=~
    MovingFowardFastThreshold=~
    DefaultToJumpStartTransitionThreshold=~
    JumpTrickAngularVelocityThreshold=~
    JumpDistanceFromGroundThreshold=~
    VelocityStartThreshold=~
    MovingFastThreshold=~
    RelaxedSpeedThreshold=~
    JumpChargeStartThreshold=~
    JumpChargeMidThreshold=~
    JumpChargeFullThreshold=~
    RotatingAngularVelocityThreshold=~
    IdleToLoopTransitionSpeedThreshold=~
    IdleToMovementLoopTransitionThreshold=~
    MeleeTwistIdleMultiplier=~
    MeleeTwistLocomotionLoopMultiplier=~
    LeanYawForMaxPivotPlayRate=~
    SteerAlphaForMaxPivotPlayRate=~
   )
   n4 = Sequence Player 'JackalVehicle_JumpNormal_Start'() {bLoopAnimation=false, sequence=/Game/Animation/Game/MainPlayer/Vehicles/JackalVehicle/JackalVehicle_JumpNormal_Start, startPosition=0.067}
   n5 = Sequence Player 'JackalVehicle_JumpTrick_Start'() {bLoopAnimation=false, sequence=/Game/Animation/Game/MainPlayer/Vehicles/JackalVehicle/JackalVehicle_JumpTrick_Start, startPosition=0.067}
   n6 = Blend Poses by bool(BlendPose_0=@n5, BlendPose_1=@n4, BlendTime_0=0.25, BlendTime_1=0.25, bActiveValue=@n3.bPlayJumpTrick) {blendType=Linear}
   n7 = Sequence Player 'JackalVehicle_JumpTrick_Vertical_Start'() {
    bLoopAnimation=false
    sequence=/Game/Animation/Game/MainPlayer/Vehicles/JackalVehicle/JackalVehicle_JumpTrick_Vertical_Start
    startPosition=0.067
   }
   Output Animation Pose(Result=Blend Poses by bool(
     BlendPose_0=Blend Poses by bool(BlendPose_0=@n2, BlendPose_1=@n1, BlendTime_0=0.2, BlendTime_1=0.2, bActiveValue=@n3.bPlayJumpTrickVertical) {blendType=Linear}
     BlendPose_1=Blend Poses by bool(BlendPose_0=@n7, BlendPose_1=@n6, BlendTime_0=0.2, BlendTime_1=0.2, bActiveValue=@n3.bPlayJumpTrickVertical) {blendType=Linear}
     bActiveValue=Get bIsDBNOCarrying()
    ) {blendType=Linear})
  ]
 }
 JumpFall {
  path=VehicleFullBodyOverride.AnimGraphNode_StateMachine_1.HoverboardBRFullBody.AnimStateNode_2.HoverboardBRMovement.AnimGraphNode_StateMachine_0.HoverboardBRFunction.AnimStateNode_2.Jump.AnimGraphNode_StateMachine_0.HoverboardJump.AnimStateNode_2.JumpFall
  nodes [
   n1 = Sequence Player 'DBNOCarry_JackalVehicle_JumpNormal_Fall'() {bLoopAnimation=false, sequence=/Game/Animation/Game/MainPlayer/Vehicles/JackalVehicle/DBNO/DBNOCarry_JackalVehicle_JumpNormal_Fall}
   n2 = Sequence Player 'DBNOCarry_JackalVehicle_JumpTrick_Vertical_Loop'() {sequence=/Game/Animation/Game/MainPlayer/Vehicles/JackalVehicle/DBNO/DBNOCarry_JackalVehicle_JumpTrick_Vertical_Loop}
   n3 = Break Fort Anim Input Jackal Vehicle(FortAnimInput_JackalVehicle=Get JackalVehicleInput()) -> (
    bIsUsingJackalVehicle=~
    bIsSteeringLeft=~
    bIsSteeringRight=~
    SteerAngleDeadZoneDegrees=~
    SteerAngle=~
    LeanYaw=~
    QuantizedSteerAngle=~
    SteerAlpha=~
    RunForwardAlpha=~
    SlopePitchDegreeAngle=~
    SlopeRollDegreeAngle=~
    DistanceFromGround=~
    PivotPlayRate=~
    bIsReversing=~
    bIsBraking=~
    bIsMoving=~
    bIsMovingForward=~
    bIsSprinting=~
    bInAir=~
    bIsFalling=~
    bIsJumping=~
    bIsRelaxed=~
    bIsBoosting=~
    bHasReachedJumpChargeStartThreshold=~
    bHasReachedJumpChargeMidThreshold=~
    bHasReachedJumpChargeFullThreshold=~
    bAdjustRootForFemaleRider=~
    bIsPlayingEmoteOnHoverboard=~
    bShouldApplyAdditive=~
    bPlayPivotOnGroundAndNotBoosting=~
    bIsOnLowerHill=~
    bLowerBodyIdleToLoopTransition=~
    bInterruptHoverboardFullybody=~
    bEnterFullBodyHoverboardState=~
    bDefaultToJumpStartTransition=~
    bLocomotionPoseToJumpTransition=~
    bBoostingToBoostingJumpTransition=~
    bJumpToLocomotionPoseTransition=~
    bJumpApexToJumpFallTransition=~
    bIdleToMovementStartTransition=~
    bIdleToMovementLoopTransition=~
    bMovementLoopToMovementStopTransition=~
    bMovementLoopToPivotTransition=~
    bMovementLoopToIdleTransition=~
    bIdleAdditiveToCollisionNTransition=~
    bSplitBodyToHoverboardBRTransition=~
    bHoverboardBRToSplitBodyTransition=~
    bHoverboardBRMovementToJumpChargeTransition=~
    bIdlesToJackalVehicleTransition=~
    bPlayAdditiveLeans=~
    bPlayBalloonLeans=~
    bPlayJumpTrickVertical=~
    bPlayJumpTrick=~
    bPlayMovingFast=~
    bPlayHipAdjustmentAdditive=~
    bPlayDriveSouth=~
    bPlayHeadAimOffset=~
    bPlaySlopeAimOffset=~
    JumpCombatAdditiveWeight=~
    MeleeTwistIdle=~
    MeleeTwistLocomotionLoop=~
    ImpactDisplacement=~
    SteerYaw=~
    EmoteHoverboardPosition=~
    EmoteHoverboardRotation=~
    bShouldAttachFeetToHoverboard=~
    FootLeftLocationOffset=~
    FootLeftRotationOffset=~
    FootRightLocationOffset=~
    FootRightRotationOffset=~
    StoppedThreshold=~
    MovingForwardThreshold=~
    MovingFowardFastThreshold=~
    DefaultToJumpStartTransitionThreshold=~
    JumpTrickAngularVelocityThreshold=~
    JumpDistanceFromGroundThreshold=~
    VelocityStartThreshold=~
    MovingFastThreshold=~
    RelaxedSpeedThreshold=~
    JumpChargeStartThreshold=~
    JumpChargeMidThreshold=~
    JumpChargeFullThreshold=~
    RotatingAngularVelocityThreshold=~
    IdleToLoopTransitionSpeedThreshold=~
    IdleToMovementLoopTransitionThreshold=~
    MeleeTwistIdleMultiplier=~
    MeleeTwistLocomotionLoopMultiplier=~
    LeanYawForMaxPivotPlayRate=~
    SteerAlphaForMaxPivotPlayRate=~
   )
   n4 = Sequence Player 'JackalVehicle_JumpTrick_Fall'() {bLoopAnimation=false, playRate=0.6, sequence=/Game/Animation/Game/MainPlayer/Vehicles/JackalVehicle/JackalVehicle_JumpTrick_Fall}
   n5 = Sequence Player 'JackalVehicle_JumpNormal_Fall'() {bLoopAnimation=false, playRate=0.3, sequence=/Game/Animation/Game/MainPlayer/Vehicles/JackalVehicle/JackalVehicle_JumpNormal_Fall}
   n6 = Blend Poses by bool(BlendPose_0=@n4, BlendPose_1=@n5, BlendTime_0=0.3, BlendTime_1=0.3, bActiveValue=@n3.bPlayJumpTrick) {blendType=Linear}
   n7 = Sequence Player 'JackalVehicle_JumpTrick_Vertical'() {sequence=/Game/Animation/Game/MainPlayer/Vehicles/JackalVehicle/JackalVehicle_JumpTrick_Vertical}
   Output Animation Pose(Result=Blend Poses by bool(
     BlendPose_0=Blend Poses by bool(BlendPose_0=@n2, BlendPose_1=@n1, BlendTime_0=0.3, BlendTime_1=0.3, bActiveValue=@n3.bPlayJumpTrickVertical) {blendType=Linear}
     BlendPose_1=Blend Poses by bool(BlendPose_0=@n7, BlendPose_1=@n6, BlendTime_0=0.3, BlendTime_1=0.3, bActiveValue=@n3.bPlayJumpTrickVertical) {blendType=Linear}
     bActiveValue=Get bIsDBNOCarrying()
    ) {blendType=Linear})
  ]
 }
 JumpLand {
  path=VehicleFullBodyOverride.AnimGraphNode_StateMachine_1.HoverboardBRFullBody.AnimStateNode_2.HoverboardBRMovement.AnimGraphNode_StateMachine_0.HoverboardBRFunction.AnimStateNode_2.Jump.AnimGraphNode_StateMachine_0.HoverboardJump.AnimStateNode_7.JumpLand
  nodes [Output Animation Pose(Result=Blend Poses by bool(
     BlendPose_0=Sequence Player 'DBNOCarry_JackalVehicle_JumpNormal_Land'() {bLoopAnimation=false, sequence=/Game/Animation/Game/MainPlayer/Vehicles/JackalVehicle/DBNO/DBNOCarry_JackalVehicle_JumpNormal_Land}
     BlendPose_1=Sequence Player 'JackalVehicle_JumpNormal_Land'() {bLoopAnimation=false, sequence=/Game/Animation/Game/MainPlayer/Vehicles/JackalVehicle/JackalVehicle_JumpNormal_Land}
     bActiveValue=Get bIsDBNOCarrying()
    ) {blendType=Linear})]
 }
 JumpApex {
  path=VehicleFullBodyOverride.AnimGraphNode_StateMachine_1.HoverboardBRFullBody.AnimStateNode_2.HoverboardBRMovement.AnimGraphNode_StateMachine_0.HoverboardBRFunction.AnimStateNode_2.Jump.AnimGraphNode_StateMachine_0.HoverboardJump.AnimStateNode_9.JumpApex
  nodes [
   n1 = Sequence Player 'DBNOCarry_JackalVehicle_JumpNormal_Loop'() {sequence=/Game/Animation/Game/MainPlayer/Vehicles/JackalVehicle/DBNO/DBNOCarry_JackalVehicle_JumpNormal_Loop}
   n2 = Sequence Player 'DBNOCarry_JackalVehicle_JumpTrick_Vertical_Loop'() {sequence=&DBNOCarry_JackalVehicle_JumpTrick_Vertical_Loop}
   n3 = Break Fort Anim Input Jackal Vehicle(FortAnimInput_JackalVehicle=Get JackalVehicleInput()) -> (
    bIsUsingJackalVehicle=~
    bIsSteeringLeft=~
    bIsSteeringRight=~
    SteerAngleDeadZoneDegrees=~
    SteerAngle=~
    LeanYaw=~
    QuantizedSteerAngle=~
    SteerAlpha=~
    RunForwardAlpha=~
    SlopePitchDegreeAngle=~
    SlopeRollDegreeAngle=~
    DistanceFromGround=~
    PivotPlayRate=~
    bIsReversing=~
    bIsBraking=~
    bIsMoving=~
    bIsMovingForward=~
    bIsSprinting=~
    bInAir=~
    bIsFalling=~
    bIsJumping=~
    bIsRelaxed=~
    bIsBoosting=~
    bHasReachedJumpChargeStartThreshold=~
    bHasReachedJumpChargeMidThreshold=~
    bHasReachedJumpChargeFullThreshold=~
    bAdjustRootForFemaleRider=~
    bIsPlayingEmoteOnHoverboard=~
    bShouldApplyAdditive=~
    bPlayPivotOnGroundAndNotBoosting=~
    bIsOnLowerHill=~
    bLowerBodyIdleToLoopTransition=~
    bInterruptHoverboardFullybody=~
    bEnterFullBodyHoverboardState=~
    bDefaultToJumpStartTransition=~
    bLocomotionPoseToJumpTransition=~
    bBoostingToBoostingJumpTransition=~
    bJumpToLocomotionPoseTransition=~
    bJumpApexToJumpFallTransition=~
    bIdleToMovementStartTransition=~
    bIdleToMovementLoopTransition=~
    bMovementLoopToMovementStopTransition=~
    bMovementLoopToPivotTransition=~
    bMovementLoopToIdleTransition=~
    bIdleAdditiveToCollisionNTransition=~
    bSplitBodyToHoverboardBRTransition=~
    bHoverboardBRToSplitBodyTransition=~
    bHoverboardBRMovementToJumpChargeTransition=~
    bIdlesToJackalVehicleTransition=~
    bPlayAdditiveLeans=~
    bPlayBalloonLeans=~
    bPlayJumpTrickVertical=~
    bPlayJumpTrick=~
    bPlayMovingFast=~
    bPlayHipAdjustmentAdditive=~
    bPlayDriveSouth=~
    bPlayHeadAimOffset=~
    bPlaySlopeAimOffset=~
    JumpCombatAdditiveWeight=~
    MeleeTwistIdle=~
    MeleeTwistLocomotionLoop=~
    ImpactDisplacement=~
    SteerYaw=~
    EmoteHoverboardPosition=~
    EmoteHoverboardRotation=~
    bShouldAttachFeetToHoverboard=~
    FootLeftLocationOffset=~
    FootLeftRotationOffset=~
    FootRightLocationOffset=~
    FootRightRotationOffset=~
    StoppedThreshold=~
    MovingForwardThreshold=~
    MovingFowardFastThreshold=~
    DefaultToJumpStartTransitionThreshold=~
    JumpTrickAngularVelocityThreshold=~
    JumpDistanceFromGroundThreshold=~
    VelocityStartThreshold=~
    MovingFastThreshold=~
    RelaxedSpeedThreshold=~
    JumpChargeStartThreshold=~
    JumpChargeMidThreshold=~
    JumpChargeFullThreshold=~
    RotatingAngularVelocityThreshold=~
    IdleToLoopTransitionSpeedThreshold=~
    IdleToMovementLoopTransitionThreshold=~
    MeleeTwistIdleMultiplier=~
    MeleeTwistLocomotionLoopMultiplier=~
    LeanYawForMaxPivotPlayRate=~
    SteerAlphaForMaxPivotPlayRate=~
   )
   n4 = Blend Poses by bool(
    BlendPose_0=Sequence Player 'JackalVehicle_JumpTrick_Loop'() {sequence=/Game/Animation/Game/MainPlayer/Vehicles/JackalVehicle/JackalVehicle_JumpTrick_Loop}
    BlendPose_1=Sequence Player 'JackalVehicle_JumpNormal_Loop'() {sequence=/Game/Animation/Game/MainPlayer/Vehicles/JackalVehicle/JackalVehicle_JumpNormal_Loop}
    BlendTime_0=0.3
    BlendTime_1=0.3
    bActiveValue=@n3.bPlayJumpTrick
   ) {blendType=Linear}
   n5 = Sequence Player 'JackalVehicle_JumpTrick_Vertical'() {sequence=&JackalVehicle_JumpTrick_Vertical}
   Output Animation Pose(Result=Blend Poses by bool(
     BlendPose_0=Blend Poses by bool(BlendPose_0=@n2, BlendPose_1=@n1, BlendTime_0=0.3, BlendTime_1=0.3, bActiveValue=@n3.bPlayJumpTrickVertical) {blendType=Linear}
     BlendPose_1=Blend Poses by bool(BlendPose_0=@n5, BlendPose_1=@n4, BlendTime_0=0.3, BlendTime_1=0.3, bActiveValue=@n3.bPlayJumpTrickVertical) {blendType=Linear}
     bActiveValue=Get bIsDBNOCarrying()
    ) {blendType=Linear})
  ]
 }
 AnimationTransitionGraph_1 {
  path=VehicleFullBodyOverride.AnimGraphNode_StateMachine_1.HoverboardBRFullBody.AnimStateNode_2.HoverboardBRMovement.AnimGraphNode_StateMachine_0.HoverboardBRFunction.AnimStateNode_2.Jump.AnimGraphNode_StateMachine_0.HoverboardJump.AnimStateTransitionNode_11.AnimationTransitionGraph_1
  nodes [Result()]
 }
 AnimationTransitionGraph_1 {
  path=VehicleFullBodyOverride.AnimGraphNode_StateMachine_1.HoverboardBRFullBody.AnimStateNode_2.HoverboardBRMovement.AnimGraphNode_StateMachine_0.HoverboardBRFunction.AnimStateNode_2.Jump.AnimGraphNode_StateMachine_0.HoverboardJump.AnimStateTransitionNode_13.AnimationTransitionGraph_1
  nodes [
   n1 = Get JackalVehicleInput()
   Result(bCanEnterTransition=NOT Boolean(A=Break Fort Anim Input Jackal Vehicle(FortAnimInput_JackalVehicle=@n1) -> (
      bIsUsingJackalVehicle=~
      bIsSteeringLeft=~
      bIsSteeringRight=~
      SteerAngleDeadZoneDegrees=~
      SteerAngle=~
      LeanYaw=~
      QuantizedSteerAngle=~
      SteerAlpha=~
      RunForwardAlpha=~
      SlopePitchDegreeAngle=~
      SlopeRollDegreeAngle=~
      DistanceFromGround=~
      PivotPlayRate=~
      bIsReversing=~
      bIsBraking=~
      bIsMoving=~
      bIsMovingForward=~
      bIsSprinting=~
      bInAir=~
      bIsFalling=~
      bIsJumping=~
      bIsRelaxed=~
      bIsBoosting=~
      bHasReachedJumpChargeStartThreshold=~
      bHasReachedJumpChargeMidThreshold=~
      bHasReachedJumpChargeFullThreshold=~
      bAdjustRootForFemaleRider=~
      bIsPlayingEmoteOnHoverboard=~
      bShouldApplyAdditive=~
      bPlayPivotOnGroundAndNotBoosting=~
      bIsOnLowerHill=~
      bLowerBodyIdleToLoopTransition=~
      bInterruptHoverboardFullybody=~
      bEnterFullBodyHoverboardState=~
      bDefaultToJumpStartTransition=~
      bLocomotionPoseToJumpTransition=~
      bBoostingToBoostingJumpTransition=~
      bJumpToLocomotionPoseTransition=~
      bJumpApexToJumpFallTransition=~
      bIdleToMovementStartTransition=~
      bIdleToMovementLoopTransition=~
      bMovementLoopToMovementStopTransition=~
      bMovementLoopToPivotTransition=~
      bMovementLoopToIdleTransition=~
      bIdleAdditiveToCollisionNTransition=~
      bSplitBodyToHoverboardBRTransition=~
      bHoverboardBRToSplitBodyTransition=~
      bHoverboardBRMovementToJumpChargeTransition=~
      bIdlesToJackalVehicleTransition=~
      bPlayAdditiveLeans=~
      bPlayBalloonLeans=~
      bPlayJumpTrickVertical=~
      bPlayJumpTrick=~
      bPlayMovingFast=~
      bPlayHipAdjustmentAdditive=~
      bPlayDriveSouth=~
      bPlayHeadAimOffset=~
      bPlaySlopeAimOffset=~
      JumpCombatAdditiveWeight=~
      MeleeTwistIdle=~
      MeleeTwistLocomotionLoop=~
      ImpactDisplacement=~
      SteerYaw=~
      EmoteHoverboardPosition=~
      EmoteHoverboardRotation=~
      bShouldAttachFeetToHoverboard=~
      FootLeftLocationOffset=~
      FootLeftRotationOffset=~
      FootRightLocationOffset=~
      FootRightRotationOffset=~
      StoppedThreshold=~
      MovingForwardThreshold=~
      MovingFowardFastThreshold=~
      DefaultToJumpStartTransitionThreshold=~
      JumpTrickAngularVelocityThreshold=~
      JumpDistanceFromGroundThreshold=~
      VelocityStartThreshold=~
      MovingFastThreshold=~
      RelaxedSpeedThreshold=~
      JumpChargeStartThreshold=~
      JumpChargeMidThreshold=~
      JumpChargeFullThreshold=~
      RotatingAngularVelocityThreshold=~
      IdleToLoopTransitionSpeedThreshold=~
      IdleToMovementLoopTransitionThreshold=~
      MeleeTwistIdleMultiplier=~
      MeleeTwistLocomotionLoopMultiplier=~
      LeanYawForMaxPivotPlayRate=~
      SteerAlphaForMaxPivotPlayRate=~
     ).bInAir))
  ]
 }
 AnimationTransitionGraph_1 {
  path=VehicleFullBodyOverride.AnimGraphNode_StateMachine_1.HoverboardBRFullBody.AnimStateNode_2.HoverboardBRMovement.AnimGraphNode_StateMachine_0.HoverboardBRFunction.AnimStateNode_2.Jump.AnimGraphNode_StateMachine_0.HoverboardJump.AnimStateTransitionNode_19.AnimationTransitionGraph_1
  nodes [Result(bCanEnterTransition=Break Fort Anim Input Jackal Vehicle(FortAnimInput_JackalVehicle=Get JackalVehicleInput()) -> (
     bIsUsingJackalVehicle=~
     bIsSteeringLeft=~
     bIsSteeringRight=~
     SteerAngleDeadZoneDegrees=~
     SteerAngle=~
     LeanYaw=~
     QuantizedSteerAngle=~
     SteerAlpha=~
     RunForwardAlpha=~
     SlopePitchDegreeAngle=~
     SlopeRollDegreeAngle=~
     DistanceFromGround=~
     PivotPlayRate=~
     bIsReversing=~
     bIsBraking=~
     bIsMoving=~
     bIsMovingForward=~
     bIsSprinting=~
     bInAir=~
     bIsFalling=~
     bIsJumping=~
     bIsRelaxed=~
     bIsBoosting=~
     bHasReachedJumpChargeStartThreshold=~
     bHasReachedJumpChargeMidThreshold=~
     bHasReachedJumpChargeFullThreshold=~
     bAdjustRootForFemaleRider=~
     bIsPlayingEmoteOnHoverboard=~
     bShouldApplyAdditive=~
     bPlayPivotOnGroundAndNotBoosting=~
     bIsOnLowerHill=~
     bLowerBodyIdleToLoopTransition=~
     bInterruptHoverboardFullybody=~
     bEnterFullBodyHoverboardState=~
     bDefaultToJumpStartTransition=~
     bLocomotionPoseToJumpTransition=~
     bBoostingToBoostingJumpTransition=~
     bJumpToLocomotionPoseTransition=~
     bJumpApexToJumpFallTransition=~
     bIdleToMovementStartTransition=~
     bIdleToMovementLoopTransition=~
     bMovementLoopToMovementStopTransition=~
     bMovementLoopToPivotTransition=~
     bMovementLoopToIdleTransition=~
     bIdleAdditiveToCollisionNTransition=~
     bSplitBodyToHoverboardBRTransition=~
     bHoverboardBRToSplitBodyTransition=~
     bHoverboardBRMovementToJumpChargeTransition=~
     bIdlesToJackalVehicleTransition=~
     bPlayAdditiveLeans=~
     bPlayBalloonLeans=~
     bPlayJumpTrickVertical=~
     bPlayJumpTrick=~
     bPlayMovingFast=~
     bPlayHipAdjustmentAdditive=~
     bPlayDriveSouth=~
     bPlayHeadAimOffset=~
     bPlaySlopeAimOffset=~
     JumpCombatAdditiveWeight=~
     MeleeTwistIdle=~
     MeleeTwistLocomotionLoop=~
     ImpactDisplacement=~
     SteerYaw=~
     EmoteHoverboardPosition=~
     EmoteHoverboardRotation=~
     bShouldAttachFeetToHoverboard=~
     FootLeftLocationOffset=~
     FootLeftRotationOffset=~
     FootRightLocationOffset=~
     FootRightRotationOffset=~
     StoppedThreshold=~
     MovingForwardThreshold=~
     MovingFowardFastThreshold=~
     DefaultToJumpStartTransitionThreshold=~
     JumpTrickAngularVelocityThreshold=~
     JumpDistanceFromGroundThreshold=~
     VelocityStartThreshold=~
     MovingFastThreshold=~
     RelaxedSpeedThreshold=~
     JumpChargeStartThreshold=~
     JumpChargeMidThreshold=~
     JumpChargeFullThreshold=~
     RotatingAngularVelocityThreshold=~
     IdleToLoopTransitionSpeedThreshold=~
     IdleToMovementLoopTransitionThreshold=~
     MeleeTwistIdleMultiplier=~
     MeleeTwistLocomotionLoopMultiplier=~
     LeanYawForMaxPivotPlayRate=~
     SteerAlphaForMaxPivotPlayRate=~
    ).bIsJumping)]
 }
 AnimationTransitionGraph_1 {
  path=VehicleFullBodyOverride.AnimGraphNode_StateMachine_1.HoverboardBRFullBody.AnimStateNode_2.HoverboardBRMovement.AnimGraphNode_StateMachine_0.HoverboardBRFunction.AnimStateNode_2.Jump.AnimGraphNode_StateMachine_0.HoverboardJump.AnimStateTransitionNode_3.AnimationTransitionGraph_1
  nodes [Result(bCanEnterTransition=Break Fort Anim Input Jackal Vehicle(FortAnimInput_JackalVehicle=Get JackalVehicleInput()) -> (
     bIsUsingJackalVehicle=~
     bIsSteeringLeft=~
     bIsSteeringRight=~
     SteerAngleDeadZoneDegrees=~
     SteerAngle=~
     LeanYaw=~
     QuantizedSteerAngle=~
     SteerAlpha=~
     RunForwardAlpha=~
     SlopePitchDegreeAngle=~
     SlopeRollDegreeAngle=~
     DistanceFromGround=~
     PivotPlayRate=~
     bIsReversing=~
     bIsBraking=~
     bIsMoving=~
     bIsMovingForward=~
     bIsSprinting=~
     bInAir=~
     bIsFalling=~
     bIsJumping=~
     bIsRelaxed=~
     bIsBoosting=~
     bHasReachedJumpChargeStartThreshold=~
     bHasReachedJumpChargeMidThreshold=~
     bHasReachedJumpChargeFullThreshold=~
     bAdjustRootForFemaleRider=~
     bIsPlayingEmoteOnHoverboard=~
     bShouldApplyAdditive=~
     bPlayPivotOnGroundAndNotBoosting=~
     bIsOnLowerHill=~
     bLowerBodyIdleToLoopTransition=~
     bInterruptHoverboardFullybody=~
     bEnterFullBodyHoverboardState=~
     bDefaultToJumpStartTransition=~
     bLocomotionPoseToJumpTransition=~
     bBoostingToBoostingJumpTransition=~
     bJumpToLocomotionPoseTransition=~
     bJumpApexToJumpFallTransition=~
     bIdleToMovementStartTransition=~
     bIdleToMovementLoopTransition=~
     bMovementLoopToMovementStopTransition=~
     bMovementLoopToPivotTransition=~
     bMovementLoopToIdleTransition=~
     bIdleAdditiveToCollisionNTransition=~
     bSplitBodyToHoverboardBRTransition=~
     bHoverboardBRToSplitBodyTransition=~
     bHoverboardBRMovementToJumpChargeTransition=~
     bIdlesToJackalVehicleTransition=~
     bPlayAdditiveLeans=~
     bPlayBalloonLeans=~
     bPlayJumpTrickVertical=~
     bPlayJumpTrick=~
     bPlayMovingFast=~
     bPlayHipAdjustmentAdditive=~
     bPlayDriveSouth=~
     bPlayHeadAimOffset=~
     bPlaySlopeAimOffset=~
     JumpCombatAdditiveWeight=~
     MeleeTwistIdle=~
     MeleeTwistLocomotionLoop=~
     ImpactDisplacement=~
     SteerYaw=~
     EmoteHoverboardPosition=~
     EmoteHoverboardRotation=~
     bShouldAttachFeetToHoverboard=~
     FootLeftLocationOffset=~
     FootLeftRotationOffset=~
     FootRightLocationOffset=~
     FootRightRotationOffset=~
     StoppedThreshold=~
     MovingForwardThreshold=~
     MovingFowardFastThreshold=~
     DefaultToJumpStartTransitionThreshold=~
     JumpTrickAngularVelocityThreshold=~
     JumpDistanceFromGroundThreshold=~
     VelocityStartThreshold=~
     MovingFastThreshold=~
     RelaxedSpeedThreshold=~
     JumpChargeStartThreshold=~
     JumpChargeMidThreshold=~
     JumpChargeFullThreshold=~
     RotatingAngularVelocityThreshold=~
     IdleToLoopTransitionSpeedThreshold=~
     IdleToMovementLoopTransitionThreshold=~
     MeleeTwistIdleMultiplier=~
     MeleeTwistLocomotionLoopMultiplier=~
     LeanYawForMaxPivotPlayRate=~
     SteerAlphaForMaxPivotPlayRate=~
    ).bJumpApexToJumpFallTransition)]
 }
 AnimationTransitionGraph_1 {
  path=VehicleFullBodyOverride.AnimGraphNode_StateMachine_1.HoverboardBRFullBody.AnimStateNode_2.HoverboardBRMovement.AnimGraphNode_StateMachine_0.HoverboardBRFunction.AnimStateNode_2.Jump.AnimGraphNode_StateMachine_0.HoverboardJump.AnimStateTransitionNode_31.AnimationTransitionGraph_1
  nodes [
   n1 = Get JackalVehicleInput()
   Result(bCanEnterTransition=NOT Boolean(A=Break Fort Anim Input Jackal Vehicle(FortAnimInput_JackalVehicle=@n1) -> (
      bIsUsingJackalVehicle=~
      bIsSteeringLeft=~
      bIsSteeringRight=~
      SteerAngleDeadZoneDegrees=~
      SteerAngle=~
      LeanYaw=~
      QuantizedSteerAngle=~
      SteerAlpha=~
      RunForwardAlpha=~
      SlopePitchDegreeAngle=~
      SlopeRollDegreeAngle=~
      DistanceFromGround=~
      PivotPlayRate=~
      bIsReversing=~
      bIsBraking=~
      bIsMoving=~
      bIsMovingForward=~
      bIsSprinting=~
      bInAir=~
      bIsFalling=~
      bIsJumping=~
      bIsRelaxed=~
      bIsBoosting=~
      bHasReachedJumpChargeStartThreshold=~
      bHasReachedJumpChargeMidThreshold=~
      bHasReachedJumpChargeFullThreshold=~
      bAdjustRootForFemaleRider=~
      bIsPlayingEmoteOnHoverboard=~
      bShouldApplyAdditive=~
      bPlayPivotOnGroundAndNotBoosting=~
      bIsOnLowerHill=~
      bLowerBodyIdleToLoopTransition=~
      bInterruptHoverboardFullybody=~
      bEnterFullBodyHoverboardState=~
      bDefaultToJumpStartTransition=~
      bLocomotionPoseToJumpTransition=~
      bBoostingToBoostingJumpTransition=~
      bJumpToLocomotionPoseTransition=~
      bJumpApexToJumpFallTransition=~
      bIdleToMovementStartTransition=~
      bIdleToMovementLoopTransition=~
      bMovementLoopToMovementStopTransition=~
      bMovementLoopToPivotTransition=~
      bMovementLoopToIdleTransition=~
      bIdleAdditiveToCollisionNTransition=~
      bSplitBodyToHoverboardBRTransition=~
      bHoverboardBRToSplitBodyTransition=~
      bHoverboardBRMovementToJumpChargeTransition=~
      bIdlesToJackalVehicleTransition=~
      bPlayAdditiveLeans=~
      bPlayBalloonLeans=~
      bPlayJumpTrickVertical=~
      bPlayJumpTrick=~
      bPlayMovingFast=~
      bPlayHipAdjustmentAdditive=~
      bPlayDriveSouth=~
      bPlayHeadAimOffset=~
      bPlaySlopeAimOffset=~
      JumpCombatAdditiveWeight=~
      MeleeTwistIdle=~
      MeleeTwistLocomotionLoop=~
      ImpactDisplacement=~
      SteerYaw=~
      EmoteHoverboardPosition=~
      EmoteHoverboardRotation=~
      bShouldAttachFeetToHoverboard=~
      FootLeftLocationOffset=~
      FootLeftRotationOffset=~
      FootRightLocationOffset=~
      FootRightRotationOffset=~
      StoppedThreshold=~
      MovingForwardThreshold=~
      MovingFowardFastThreshold=~
      DefaultToJumpStartTransitionThreshold=~
      JumpTrickAngularVelocityThreshold=~
      JumpDistanceFromGroundThreshold=~
      VelocityStartThreshold=~
      MovingFastThreshold=~
      RelaxedSpeedThreshold=~
      JumpChargeStartThreshold=~
      JumpChargeMidThreshold=~
      JumpChargeFullThreshold=~
      RotatingAngularVelocityThreshold=~
      IdleToLoopTransitionSpeedThreshold=~
      IdleToMovementLoopTransitionThreshold=~
      MeleeTwistIdleMultiplier=~
      MeleeTwistLocomotionLoopMultiplier=~
      LeanYawForMaxPivotPlayRate=~
      SteerAlphaForMaxPivotPlayRate=~
     ).bInAir))
  ]
 }
 Transition {
  path=VehicleFullBodyOverride.AnimGraphNode_StateMachine_1.HoverboardBRFullBody.AnimStateNode_2.HoverboardBRMovement.AnimGraphNode_StateMachine_0.HoverboardBRFunction.AnimStateNode_2.Jump.AnimGraphNode_StateMachine_0.HoverboardJump.AnimStateTransitionNode_4.Transition
  nodes [Result(bCanEnterTransition=Break Fort Anim Input Jackal Vehicle(FortAnimInput_JackalVehicle=Get JackalVehicleInput()) -> (
     bIsUsingJackalVehicle=~
     bIsSteeringLeft=~
     bIsSteeringRight=~
     SteerAngleDeadZoneDegrees=~
     SteerAngle=~
     LeanYaw=~
     QuantizedSteerAngle=~
     SteerAlpha=~
     RunForwardAlpha=~
     SlopePitchDegreeAngle=~
     SlopeRollDegreeAngle=~
     DistanceFromGround=~
     PivotPlayRate=~
     bIsReversing=~
     bIsBraking=~
     bIsMoving=~
     bIsMovingForward=~
     bIsSprinting=~
     bInAir=~
     bIsFalling=~
     bIsJumping=~
     bIsRelaxed=~
     bIsBoosting=~
     bHasReachedJumpChargeStartThreshold=~
     bHasReachedJumpChargeMidThreshold=~
     bHasReachedJumpChargeFullThreshold=~
     bAdjustRootForFemaleRider=~
     bIsPlayingEmoteOnHoverboard=~
     bShouldApplyAdditive=~
     bPlayPivotOnGroundAndNotBoosting=~
     bIsOnLowerHill=~
     bLowerBodyIdleToLoopTransition=~
     bInterruptHoverboardFullybody=~
     bEnterFullBodyHoverboardState=~
     bDefaultToJumpStartTransition=~
     bLocomotionPoseToJumpTransition=~
     bBoostingToBoostingJumpTransition=~
     bJumpToLocomotionPoseTransition=~
     bJumpApexToJumpFallTransition=~
     bIdleToMovementStartTransition=~
     bIdleToMovementLoopTransition=~
     bMovementLoopToMovementStopTransition=~
     bMovementLoopToPivotTransition=~
     bMovementLoopToIdleTransition=~
     bIdleAdditiveToCollisionNTransition=~
     bSplitBodyToHoverboardBRTransition=~
     bHoverboardBRToSplitBodyTransition=~
     bHoverboardBRMovementToJumpChargeTransition=~
     bIdlesToJackalVehicleTransition=~
     bPlayAdditiveLeans=~
     bPlayBalloonLeans=~
     bPlayJumpTrickVertical=~
     bPlayJumpTrick=~
     bPlayMovingFast=~
     bPlayHipAdjustmentAdditive=~
     bPlayDriveSouth=~
     bPlayHeadAimOffset=~
     bPlaySlopeAimOffset=~
     JumpCombatAdditiveWeight=~
     MeleeTwistIdle=~
     MeleeTwistLocomotionLoop=~
     ImpactDisplacement=~
     SteerYaw=~
     EmoteHoverboardPosition=~
     EmoteHoverboardRotation=~
     bShouldAttachFeetToHoverboard=~
     FootLeftLocationOffset=~
     FootLeftRotationOffset=~
     FootRightLocationOffset=~
     FootRightRotationOffset=~
     StoppedThreshold=~
     MovingForwardThreshold=~
     MovingFowardFastThreshold=~
     DefaultToJumpStartTransitionThreshold=~
     JumpTrickAngularVelocityThreshold=~
     JumpDistanceFromGroundThreshold=~
     VelocityStartThreshold=~
     MovingFastThreshold=~
     RelaxedSpeedThreshold=~
     JumpChargeStartThreshold=~
     JumpChargeMidThreshold=~
     JumpChargeFullThreshold=~
     RotatingAngularVelocityThreshold=~
     IdleToLoopTransitionSpeedThreshold=~
     IdleToMovementLoopTransitionThreshold=~
     MeleeTwistIdleMultiplier=~
     MeleeTwistLocomotionLoopMultiplier=~
     LeanYawForMaxPivotPlayRate=~
     SteerAlphaForMaxPivotPlayRate=~
    ).bDefaultToJumpStartTransition)]
 }
 Transition {
  path=VehicleFullBodyOverride.AnimGraphNode_StateMachine_1.HoverboardBRFullBody.AnimStateNode_2.HoverboardBRMovement.AnimGraphNode_StateMachine_0.HoverboardBRFunction.AnimStateNode_2.Jump.AnimGraphNode_StateMachine_0.HoverboardJump.AnimStateTransitionNode_5.Transition
  nodes [Result()]
 }
 BoostingEnd {
  path=VehicleFullBodyOverride.AnimGraphNode_StateMachine_1.HoverboardBRFullBody.AnimStateNode_2.HoverboardBRMovement.AnimGraphNode_StateMachine_0.HoverboardBRFunction.AnimStateNode_3.BoostingEnd
  nodes [Output Animation Pose(Result=Blend Poses by bool(
     BlendPose_0=Sequence Player 'DBNOCarry_JackalVehicle_BoostingEnd'() {bLoopAnimation=false, sequence=/Game/Animation/Game/MainPlayer/Vehicles/JackalVehicle/DBNO/DBNOCarry_JackalVehicle_BoostingEnd}
     BlendPose_1=Sequence Player 'JackalVehicle_BoostingEnd'() {bLoopAnimation=false, sequence=/Game/Animation/Game/MainPlayer/Vehicles/JackalVehicle/JackalVehicle_BoostingEnd}
     bActiveValue=Get bIsDBNOCarrying()
    ) {blendType=Linear})]
 }
 BoostingJump {
  path=VehicleFullBodyOverride.AnimGraphNode_StateMachine_1.HoverboardBRFullBody.AnimStateNode_2.HoverboardBRMovement.AnimGraphNode_StateMachine_0.HoverboardBRFunction.AnimStateNode_4.BoostingJump
  nodes [
   n1 = Layered blend per bone(
    BasePose=Sequence Player 'JackalVehicle_Boosting_Pose'() {bIgnoreForRelevancyTest=true, sequence=/Game/Animation/Game/MainPlayer/Vehicles/JackalVehicle/JackalVehicle_Boosting_Pose}
    BlendPoses_0=Sequence Player 'JackalVehicle_BoostingJump'() {sequence=/Game/Animation/Game/MainPlayer/Vehicles/JackalVehicle/JackalVehicle_BoostingJump}
   ) {
    bMeshSpaceRotationBlend=true
    layerSetup [{branchFilters [{boneName=pelvis}, {blendDepth=-1, boneName=thigh_l}, {blendDepth=-1, boneName=thigh_r}]}]
   }
   Output Animation Pose(Result=@n1)
  ]
 }
 BoostingLoop {
  path=VehicleFullBodyOverride.AnimGraphNode_StateMachine_1.HoverboardBRFullBody.AnimStateNode_2.HoverboardBRMovement.AnimGraphNode_StateMachine_0.HoverboardBRFunction.AnimStateNode_5.BoostingLoop
  nodes [Output Animation Pose(Result=Blend Poses by bool(
     BlendPose_0=Sequence Player 'DBNOCarry_JackalVehicle_Boosting_Pose'() {sequence=/Game/Animation/Game/MainPlayer/Vehicles/JackalVehicle/DBNO/DBNOCarry_JackalVehicle_Boosting_Pose}
     BlendPose_1=Sequence Player 'JackalVehicle_Boosting_Pose'() {sequence=&JackalVehicle_Boosting_Pose}
     bActiveValue=Get bIsDBNOCarrying()
    ) {blendType=Linear})]
 }
 BoostingJumpLoop {
  path=VehicleFullBodyOverride.AnimGraphNode_StateMachine_1.HoverboardBRFullBody.AnimStateNode_2.HoverboardBRMovement.AnimGraphNode_StateMachine_0.HoverboardBRFunction.AnimStateNode_6.BoostingJumpLoop
  nodes [Output Animation Pose(Result=Sequence Player 'JackalVehicle_Boosting_Pose'() {sequence=&JackalVehicle_Boosting_Pose})]
 }
 Transition {
  path=VehicleFullBodyOverride.AnimGraphNode_StateMachine_1.HoverboardBRFullBody.AnimStateNode_2.HoverboardBRMovement.AnimGraphNode_StateMachine_0.HoverboardBRFunction.AnimStateTransitionNode_0.Transition
  nodes [Result(bCanEnterTransition=Break Fort Anim Input Jackal Vehicle(FortAnimInput_JackalVehicle=Get JackalVehicleInput()) -> (
     bIsUsingJackalVehicle=~
     bIsSteeringLeft=~
     bIsSteeringRight=~
     SteerAngleDeadZoneDegrees=~
     SteerAngle=~
     LeanYaw=~
     QuantizedSteerAngle=~
     SteerAlpha=~
     RunForwardAlpha=~
     SlopePitchDegreeAngle=~
     SlopeRollDegreeAngle=~
     DistanceFromGround=~
     PivotPlayRate=~
     bIsReversing=~
     bIsBraking=~
     bIsMoving=~
     bIsMovingForward=~
     bIsSprinting=~
     bInAir=~
     bIsFalling=~
     bIsJumping=~
     bIsRelaxed=~
     bIsBoosting=~
     bHasReachedJumpChargeStartThreshold=~
     bHasReachedJumpChargeMidThreshold=~
     bHasReachedJumpChargeFullThreshold=~
     bAdjustRootForFemaleRider=~
     bIsPlayingEmoteOnHoverboard=~
     bShouldApplyAdditive=~
     bPlayPivotOnGroundAndNotBoosting=~
     bIsOnLowerHill=~
     bLowerBodyIdleToLoopTransition=~
     bInterruptHoverboardFullybody=~
     bEnterFullBodyHoverboardState=~
     bDefaultToJumpStartTransition=~
     bLocomotionPoseToJumpTransition=~
     bBoostingToBoostingJumpTransition=~
     bJumpToLocomotionPoseTransition=~
     bJumpApexToJumpFallTransition=~
     bIdleToMovementStartTransition=~
     bIdleToMovementLoopTransition=~
     bMovementLoopToMovementStopTransition=~
     bMovementLoopToPivotTransition=~
     bMovementLoopToIdleTransition=~
     bIdleAdditiveToCollisionNTransition=~
     bSplitBodyToHoverboardBRTransition=~
     bHoverboardBRToSplitBodyTransition=~
     bHoverboardBRMovementToJumpChargeTransition=~
     bIdlesToJackalVehicleTransition=~
     bPlayAdditiveLeans=~
     bPlayBalloonLeans=~
     bPlayJumpTrickVertical=~
     bPlayJumpTrick=~
     bPlayMovingFast=~
     bPlayHipAdjustmentAdditive=~
     bPlayDriveSouth=~
     bPlayHeadAimOffset=~
     bPlaySlopeAimOffset=~
     JumpCombatAdditiveWeight=~
     MeleeTwistIdle=~
     MeleeTwistLocomotionLoop=~
     ImpactDisplacement=~
     SteerYaw=~
     EmoteHoverboardPosition=~
     EmoteHoverboardRotation=~
     bShouldAttachFeetToHoverboard=~
     FootLeftLocationOffset=~
     FootLeftRotationOffset=~
     FootRightLocationOffset=~
     FootRightRotationOffset=~
     StoppedThreshold=~
     MovingForwardThreshold=~
     MovingFowardFastThreshold=~
     DefaultToJumpStartTransitionThreshold=~
     JumpTrickAngularVelocityThreshold=~
     JumpDistanceFromGroundThreshold=~
     VelocityStartThreshold=~
     MovingFastThreshold=~
     RelaxedSpeedThreshold=~
     JumpChargeStartThreshold=~
     JumpChargeMidThreshold=~
     JumpChargeFullThreshold=~
     RotatingAngularVelocityThreshold=~
     IdleToLoopTransitionSpeedThreshold=~
     IdleToMovementLoopTransitionThreshold=~
     MeleeTwistIdleMultiplier=~
     MeleeTwistLocomotionLoopMultiplier=~
     LeanYawForMaxPivotPlayRate=~
     SteerAlphaForMaxPivotPlayRate=~
    ).bIsBoosting)]
 }
 AnimationTransitionGraph_2 {
  path=VehicleFullBodyOverride.AnimGraphNode_StateMachine_1.HoverboardBRFullBody.AnimStateNode_2.HoverboardBRMovement.AnimGraphNode_StateMachine_0.HoverboardBRFunction.AnimStateTransitionNode_1.AnimationTransitionGraph_2
  nodes [Result()]
 }
 Transition {
  path=VehicleFullBodyOverride.AnimGraphNode_StateMachine_1.HoverboardBRFullBody.AnimStateNode_2.HoverboardBRMovement.AnimGraphNode_StateMachine_0.HoverboardBRFunction.AnimStateTransitionNode_10.Transition
  nodes [Result(bCanEnterTransition=Break Fort Anim Input Jackal Vehicle(FortAnimInput_JackalVehicle=Get JackalVehicleInput()) -> (
     bIsUsingJackalVehicle=~
     bIsSteeringLeft=~
     bIsSteeringRight=~
     SteerAngleDeadZoneDegrees=~
     SteerAngle=~
     LeanYaw=~
     QuantizedSteerAngle=~
     SteerAlpha=~
     RunForwardAlpha=~
     SlopePitchDegreeAngle=~
     SlopeRollDegreeAngle=~
     DistanceFromGround=~
     PivotPlayRate=~
     bIsReversing=~
     bIsBraking=~
     bIsMoving=~
     bIsMovingForward=~
     bIsSprinting=~
     bInAir=~
     bIsFalling=~
     bIsJumping=~
     bIsRelaxed=~
     bIsBoosting=~
     bHasReachedJumpChargeStartThreshold=~
     bHasReachedJumpChargeMidThreshold=~
     bHasReachedJumpChargeFullThreshold=~
     bAdjustRootForFemaleRider=~
     bIsPlayingEmoteOnHoverboard=~
     bShouldApplyAdditive=~
     bPlayPivotOnGroundAndNotBoosting=~
     bIsOnLowerHill=~
     bLowerBodyIdleToLoopTransition=~
     bInterruptHoverboardFullybody=~
     bEnterFullBodyHoverboardState=~
     bDefaultToJumpStartTransition=~
     bLocomotionPoseToJumpTransition=~
     bBoostingToBoostingJumpTransition=~
     bJumpToLocomotionPoseTransition=~
     bJumpApexToJumpFallTransition=~
     bIdleToMovementStartTransition=~
     bIdleToMovementLoopTransition=~
     bMovementLoopToMovementStopTransition=~
     bMovementLoopToPivotTransition=~
     bMovementLoopToIdleTransition=~
     bIdleAdditiveToCollisionNTransition=~
     bSplitBodyToHoverboardBRTransition=~
     bHoverboardBRToSplitBodyTransition=~
     bHoverboardBRMovementToJumpChargeTransition=~
     bIdlesToJackalVehicleTransition=~
     bPlayAdditiveLeans=~
     bPlayBalloonLeans=~
     bPlayJumpTrickVertical=~
     bPlayJumpTrick=~
     bPlayMovingFast=~
     bPlayHipAdjustmentAdditive=~
     bPlayDriveSouth=~
     bPlayHeadAimOffset=~
     bPlaySlopeAimOffset=~
     JumpCombatAdditiveWeight=~
     MeleeTwistIdle=~
     MeleeTwistLocomotionLoop=~
     ImpactDisplacement=~
     SteerYaw=~
     EmoteHoverboardPosition=~
     EmoteHoverboardRotation=~
     bShouldAttachFeetToHoverboard=~
     FootLeftLocationOffset=~
     FootLeftRotationOffset=~
     FootRightLocationOffset=~
     FootRightRotationOffset=~
     StoppedThreshold=~
     MovingForwardThreshold=~
     MovingFowardFastThreshold=~
     DefaultToJumpStartTransitionThreshold=~
     JumpTrickAngularVelocityThreshold=~
     JumpDistanceFromGroundThreshold=~
     VelocityStartThreshold=~
     MovingFastThreshold=~
     RelaxedSpeedThreshold=~
     JumpChargeStartThreshold=~
     JumpChargeMidThreshold=~
     JumpChargeFullThreshold=~
     RotatingAngularVelocityThreshold=~
     IdleToLoopTransitionSpeedThreshold=~
     IdleToMovementLoopTransitionThreshold=~
     MeleeTwistIdleMultiplier=~
     MeleeTwistLocomotionLoopMultiplier=~
     LeanYawForMaxPivotPlayRate=~
     SteerAlphaForMaxPivotPlayRate=~
    ).bBoostingToBoostingJumpTransition)]
 }
 Transition {
  path=VehicleFullBodyOverride.AnimGraphNode_StateMachine_1.HoverboardBRFullBody.AnimStateNode_2.HoverboardBRMovement.AnimGraphNode_StateMachine_0.HoverboardBRFunction.AnimStateTransitionNode_11.Transition
  nodes [
   n1 = Get JackalVehicleInput()
   Result(bCanEnterTransition=NOT Boolean(A=Break Fort Anim Input Jackal Vehicle(FortAnimInput_JackalVehicle=@n1) -> (
      bIsUsingJackalVehicle=~
      bIsSteeringLeft=~
      bIsSteeringRight=~
      SteerAngleDeadZoneDegrees=~
      SteerAngle=~
      LeanYaw=~
      QuantizedSteerAngle=~
      SteerAlpha=~
      RunForwardAlpha=~
      SlopePitchDegreeAngle=~
      SlopeRollDegreeAngle=~
      DistanceFromGround=~
      PivotPlayRate=~
      bIsReversing=~
      bIsBraking=~
      bIsMoving=~
      bIsMovingForward=~
      bIsSprinting=~
      bInAir=~
      bIsFalling=~
      bIsJumping=~
      bIsRelaxed=~
      bIsBoosting=~
      bHasReachedJumpChargeStartThreshold=~
      bHasReachedJumpChargeMidThreshold=~
      bHasReachedJumpChargeFullThreshold=~
      bAdjustRootForFemaleRider=~
      bIsPlayingEmoteOnHoverboard=~
      bShouldApplyAdditive=~
      bPlayPivotOnGroundAndNotBoosting=~
      bIsOnLowerHill=~
      bLowerBodyIdleToLoopTransition=~
      bInterruptHoverboardFullybody=~
      bEnterFullBodyHoverboardState=~
      bDefaultToJumpStartTransition=~
      bLocomotionPoseToJumpTransition=~
      bBoostingToBoostingJumpTransition=~
      bJumpToLocomotionPoseTransition=~
      bJumpApexToJumpFallTransition=~
      bIdleToMovementStartTransition=~
      bIdleToMovementLoopTransition=~
      bMovementLoopToMovementStopTransition=~
      bMovementLoopToPivotTransition=~
      bMovementLoopToIdleTransition=~
      bIdleAdditiveToCollisionNTransition=~
      bSplitBodyToHoverboardBRTransition=~
      bHoverboardBRToSplitBodyTransition=~
      bHoverboardBRMovementToJumpChargeTransition=~
      bIdlesToJackalVehicleTransition=~
      bPlayAdditiveLeans=~
      bPlayBalloonLeans=~
      bPlayJumpTrickVertical=~
      bPlayJumpTrick=~
      bPlayMovingFast=~
      bPlayHipAdjustmentAdditive=~
      bPlayDriveSouth=~
      bPlayHeadAimOffset=~
      bPlaySlopeAimOffset=~
      JumpCombatAdditiveWeight=~
      MeleeTwistIdle=~
      MeleeTwistLocomotionLoop=~
      ImpactDisplacement=~
      SteerYaw=~
      EmoteHoverboardPosition=~
      EmoteHoverboardRotation=~
      bShouldAttachFeetToHoverboard=~
      FootLeftLocationOffset=~
      FootLeftRotationOffset=~
      FootRightLocationOffset=~
      FootRightRotationOffset=~
      StoppedThreshold=~
      MovingForwardThreshold=~
      MovingFowardFastThreshold=~
      DefaultToJumpStartTransitionThreshold=~
      JumpTrickAngularVelocityThreshold=~
      JumpDistanceFromGroundThreshold=~
      VelocityStartThreshold=~
      MovingFastThreshold=~
      RelaxedSpeedThreshold=~
      JumpChargeStartThreshold=~
      JumpChargeMidThreshold=~
      JumpChargeFullThreshold=~
      RotatingAngularVelocityThreshold=~
      IdleToLoopTransitionSpeedThreshold=~
      IdleToMovementLoopTransitionThreshold=~
      MeleeTwistIdleMultiplier=~
      MeleeTwistLocomotionLoopMultiplier=~
      LeanYawForMaxPivotPlayRate=~
      SteerAlphaForMaxPivotPlayRate=~
     ).bIsBoosting))
  ]
 }
 AnimationTransitionGraph_2 {
  path=VehicleFullBodyOverride.AnimGraphNode_StateMachine_1.HoverboardBRFullBody.AnimStateNode_2.HoverboardBRMovement.AnimGraphNode_StateMachine_0.HoverboardBRFunction.AnimStateTransitionNode_2.AnimationTransitionGraph_2
  nodes [
   n1 = Get JackalVehicleInput()
   Result(bCanEnterTransition=NOT Boolean(A=Break Fort Anim Input Jackal Vehicle(FortAnimInput_JackalVehicle=@n1) -> (
      bIsUsingJackalVehicle=~
      bIsSteeringLeft=~
      bIsSteeringRight=~
      SteerAngleDeadZoneDegrees=~
      SteerAngle=~
      LeanYaw=~
      QuantizedSteerAngle=~
      SteerAlpha=~
      RunForwardAlpha=~
      SlopePitchDegreeAngle=~
      SlopeRollDegreeAngle=~
      DistanceFromGround=~
      PivotPlayRate=~
      bIsReversing=~
      bIsBraking=~
      bIsMoving=~
      bIsMovingForward=~
      bIsSprinting=~
      bInAir=~
      bIsFalling=~
      bIsJumping=~
      bIsRelaxed=~
      bIsBoosting=~
      bHasReachedJumpChargeStartThreshold=~
      bHasReachedJumpChargeMidThreshold=~
      bHasReachedJumpChargeFullThreshold=~
      bAdjustRootForFemaleRider=~
      bIsPlayingEmoteOnHoverboard=~
      bShouldApplyAdditive=~
      bPlayPivotOnGroundAndNotBoosting=~
      bIsOnLowerHill=~
      bLowerBodyIdleToLoopTransition=~
      bInterruptHoverboardFullybody=~
      bEnterFullBodyHoverboardState=~
      bDefaultToJumpStartTransition=~
      bLocomotionPoseToJumpTransition=~
      bBoostingToBoostingJumpTransition=~
      bJumpToLocomotionPoseTransition=~
      bJumpApexToJumpFallTransition=~
      bIdleToMovementStartTransition=~
      bIdleToMovementLoopTransition=~
      bMovementLoopToMovementStopTransition=~
      bMovementLoopToPivotTransition=~
      bMovementLoopToIdleTransition=~
      bIdleAdditiveToCollisionNTransition=~
      bSplitBodyToHoverboardBRTransition=~
      bHoverboardBRToSplitBodyTransition=~
      bHoverboardBRMovementToJumpChargeTransition=~
      bIdlesToJackalVehicleTransition=~
      bPlayAdditiveLeans=~
      bPlayBalloonLeans=~
      bPlayJumpTrickVertical=~
      bPlayJumpTrick=~
      bPlayMovingFast=~
      bPlayHipAdjustmentAdditive=~
      bPlayDriveSouth=~
      bPlayHeadAimOffset=~
      bPlaySlopeAimOffset=~
      JumpCombatAdditiveWeight=~
      MeleeTwistIdle=~
      MeleeTwistLocomotionLoop=~
      ImpactDisplacement=~
      SteerYaw=~
      EmoteHoverboardPosition=~
      EmoteHoverboardRotation=~
      bShouldAttachFeetToHoverboard=~
      FootLeftLocationOffset=~
      FootLeftRotationOffset=~
      FootRightLocationOffset=~
      FootRightRotationOffset=~
      StoppedThreshold=~
      MovingForwardThreshold=~
      MovingFowardFastThreshold=~
      DefaultToJumpStartTransitionThreshold=~
      JumpTrickAngularVelocityThreshold=~
      JumpDistanceFromGroundThreshold=~
      VelocityStartThreshold=~
      MovingFastThreshold=~
      RelaxedSpeedThreshold=~
      JumpChargeStartThreshold=~
      JumpChargeMidThreshold=~
      JumpChargeFullThreshold=~
      RotatingAngularVelocityThreshold=~
      IdleToLoopTransitionSpeedThreshold=~
      IdleToMovementLoopTransitionThreshold=~
      MeleeTwistIdleMultiplier=~
      MeleeTwistLocomotionLoopMultiplier=~
      LeanYawForMaxPivotPlayRate=~
      SteerAlphaForMaxPivotPlayRate=~
     ).bIsBoosting))
  ]
 }
 AnimationTransitionGraph_0 {
  path=VehicleFullBodyOverride.AnimGraphNode_StateMachine_1.HoverboardBRFullBody.AnimStateNode_2.HoverboardBRMovement.AnimGraphNode_StateMachine_0.HoverboardBRFunction.AnimStateTransitionNode_3.AnimationTransitionGraph_0
  nodes [Result()]
 }
 AnimationTransitionGraph_1 {
  path=VehicleFullBodyOverride.AnimGraphNode_StateMachine_1.HoverboardBRFullBody.AnimStateNode_2.HoverboardBRMovement.AnimGraphNode_StateMachine_0.HoverboardBRFunction.AnimStateTransitionNode_4.AnimationTransitionGraph_1
  nodes [Result()]
 }
 Transition {
  path=VehicleFullBodyOverride.AnimGraphNode_StateMachine_1.HoverboardBRFullBody.AnimStateNode_2.HoverboardBRMovement.AnimGraphNode_StateMachine_0.HoverboardBRFunction.AnimStateTransitionNode_5.Transition
  nodes [Result(bCanEnterTransition=Break Fort Anim Input Jackal Vehicle(FortAnimInput_JackalVehicle=Get JackalVehicleInput()) -> (
     bIsUsingJackalVehicle=~
     bIsSteeringLeft=~
     bIsSteeringRight=~
     SteerAngleDeadZoneDegrees=~
     SteerAngle=~
     LeanYaw=~
     QuantizedSteerAngle=~
     SteerAlpha=~
     RunForwardAlpha=~
     SlopePitchDegreeAngle=~
     SlopeRollDegreeAngle=~
     DistanceFromGround=~
     PivotPlayRate=~
     bIsReversing=~
     bIsBraking=~
     bIsMoving=~
     bIsMovingForward=~
     bIsSprinting=~
     bInAir=~
     bIsFalling=~
     bIsJumping=~
     bIsRelaxed=~
     bIsBoosting=~
     bHasReachedJumpChargeStartThreshold=~
     bHasReachedJumpChargeMidThreshold=~
     bHasReachedJumpChargeFullThreshold=~
     bAdjustRootForFemaleRider=~
     bIsPlayingEmoteOnHoverboard=~
     bShouldApplyAdditive=~
     bPlayPivotOnGroundAndNotBoosting=~
     bIsOnLowerHill=~
     bLowerBodyIdleToLoopTransition=~
     bInterruptHoverboardFullybody=~
     bEnterFullBodyHoverboardState=~
     bDefaultToJumpStartTransition=~
     bLocomotionPoseToJumpTransition=~
     bBoostingToBoostingJumpTransition=~
     bJumpToLocomotionPoseTransition=~
     bJumpApexToJumpFallTransition=~
     bIdleToMovementStartTransition=~
     bIdleToMovementLoopTransition=~
     bMovementLoopToMovementStopTransition=~
     bMovementLoopToPivotTransition=~
     bMovementLoopToIdleTransition=~
     bIdleAdditiveToCollisionNTransition=~
     bSplitBodyToHoverboardBRTransition=~
     bHoverboardBRToSplitBodyTransition=~
     bHoverboardBRMovementToJumpChargeTransition=~
     bIdlesToJackalVehicleTransition=~
     bPlayAdditiveLeans=~
     bPlayBalloonLeans=~
     bPlayJumpTrickVertical=~
     bPlayJumpTrick=~
     bPlayMovingFast=~
     bPlayHipAdjustmentAdditive=~
     bPlayDriveSouth=~
     bPlayHeadAimOffset=~
     bPlaySlopeAimOffset=~
     JumpCombatAdditiveWeight=~
     MeleeTwistIdle=~
     MeleeTwistLocomotionLoop=~
     ImpactDisplacement=~
     SteerYaw=~
     EmoteHoverboardPosition=~
     EmoteHoverboardRotation=~
     bShouldAttachFeetToHoverboard=~
     FootLeftLocationOffset=~
     FootLeftRotationOffset=~
     FootRightLocationOffset=~
     FootRightRotationOffset=~
     StoppedThreshold=~
     MovingForwardThreshold=~
     MovingFowardFastThreshold=~
     DefaultToJumpStartTransitionThreshold=~
     JumpTrickAngularVelocityThreshold=~
     JumpDistanceFromGroundThreshold=~
     VelocityStartThreshold=~
     MovingFastThreshold=~
     RelaxedSpeedThreshold=~
     JumpChargeStartThreshold=~
     JumpChargeMidThreshold=~
     JumpChargeFullThreshold=~
     RotatingAngularVelocityThreshold=~
     IdleToLoopTransitionSpeedThreshold=~
     IdleToMovementLoopTransitionThreshold=~
     MeleeTwistIdleMultiplier=~
     MeleeTwistLocomotionLoopMultiplier=~
     LeanYawForMaxPivotPlayRate=~
     SteerAlphaForMaxPivotPlayRate=~
    ).bBoostingToBoostingJumpTransition)]
 }
 Transition {
  path=VehicleFullBodyOverride.AnimGraphNode_StateMachine_1.HoverboardBRFullBody.AnimStateNode_2.HoverboardBRMovement.AnimGraphNode_StateMachine_0.HoverboardBRFunction.AnimStateTransitionNode_6.Transition
  nodes [Result(bCanEnterTransition=Get bLocomotionPoseToJumpTransitionAndNotOnWater() -> (bLocomotionPoseToJumpTransitionAndNotOnWater:bool))]
 }
 Transition {
  path=VehicleFullBodyOverride.AnimGraphNode_StateMachine_1.HoverboardBRFullBody.AnimStateNode_2.HoverboardBRMovement.AnimGraphNode_StateMachine_0.HoverboardBRFunction.AnimStateTransitionNode_7.Transition
  nodes [Result(bCanEnterTransition=Get bJumpToLocomotionPoseTransitionOrOnWater() -> (bJumpToLocomotionPoseTransitionOrOnWater:bool))]
 }
 Transition {
  path=VehicleFullBodyOverride.AnimGraphNode_StateMachine_1.HoverboardBRFullBody.AnimStateNode_2.HoverboardBRMovement.AnimGraphNode_StateMachine_0.HoverboardBRFunction.AnimStateTransitionNode_8.Transition
  nodes [Result(bCanEnterTransition=Break Fort Anim Input Jackal Vehicle(FortAnimInput_JackalVehicle=Get JackalVehicleInput()) -> (
     bIsUsingJackalVehicle=~
     bIsSteeringLeft=~
     bIsSteeringRight=~
     SteerAngleDeadZoneDegrees=~
     SteerAngle=~
     LeanYaw=~
     QuantizedSteerAngle=~
     SteerAlpha=~
     RunForwardAlpha=~
     SlopePitchDegreeAngle=~
     SlopeRollDegreeAngle=~
     DistanceFromGround=~
     PivotPlayRate=~
     bIsReversing=~
     bIsBraking=~
     bIsMoving=~
     bIsMovingForward=~
     bIsSprinting=~
     bInAir=~
     bIsFalling=~
     bIsJumping=~
     bIsRelaxed=~
     bIsBoosting=~
     bHasReachedJumpChargeStartThreshold=~
     bHasReachedJumpChargeMidThreshold=~
     bHasReachedJumpChargeFullThreshold=~
     bAdjustRootForFemaleRider=~
     bIsPlayingEmoteOnHoverboard=~
     bShouldApplyAdditive=~
     bPlayPivotOnGroundAndNotBoosting=~
     bIsOnLowerHill=~
     bLowerBodyIdleToLoopTransition=~
     bInterruptHoverboardFullybody=~
     bEnterFullBodyHoverboardState=~
     bDefaultToJumpStartTransition=~
     bLocomotionPoseToJumpTransition=~
     bBoostingToBoostingJumpTransition=~
     bJumpToLocomotionPoseTransition=~
     bJumpApexToJumpFallTransition=~
     bIdleToMovementStartTransition=~
     bIdleToMovementLoopTransition=~
     bMovementLoopToMovementStopTransition=~
     bMovementLoopToPivotTransition=~
     bMovementLoopToIdleTransition=~
     bIdleAdditiveToCollisionNTransition=~
     bSplitBodyToHoverboardBRTransition=~
     bHoverboardBRToSplitBodyTransition=~
     bHoverboardBRMovementToJumpChargeTransition=~
     bIdlesToJackalVehicleTransition=~
     bPlayAdditiveLeans=~
     bPlayBalloonLeans=~
     bPlayJumpTrickVertical=~
     bPlayJumpTrick=~
     bPlayMovingFast=~
     bPlayHipAdjustmentAdditive=~
     bPlayDriveSouth=~
     bPlayHeadAimOffset=~
     bPlaySlopeAimOffset=~
     JumpCombatAdditiveWeight=~
     MeleeTwistIdle=~
     MeleeTwistLocomotionLoop=~
     ImpactDisplacement=~
     SteerYaw=~
     EmoteHoverboardPosition=~
     EmoteHoverboardRotation=~
     bShouldAttachFeetToHoverboard=~
     FootLeftLocationOffset=~
     FootLeftRotationOffset=~
     FootRightLocationOffset=~
     FootRightRotationOffset=~
     StoppedThreshold=~
     MovingForwardThreshold=~
     MovingFowardFastThreshold=~
     DefaultToJumpStartTransitionThreshold=~
     JumpTrickAngularVelocityThreshold=~
     JumpDistanceFromGroundThreshold=~
     VelocityStartThreshold=~
     MovingFastThreshold=~
     RelaxedSpeedThreshold=~
     JumpChargeStartThreshold=~
     JumpChargeMidThreshold=~
     JumpChargeFullThreshold=~
     RotatingAngularVelocityThreshold=~
     IdleToLoopTransitionSpeedThreshold=~
     IdleToMovementLoopTransitionThreshold=~
     MeleeTwistIdleMultiplier=~
     MeleeTwistLocomotionLoopMultiplier=~
     LeanYawForMaxPivotPlayRate=~
     SteerAlphaForMaxPivotPlayRate=~
    ).bIsBoosting)]
 }
 Transition {
  path=VehicleFullBodyOverride.AnimGraphNode_StateMachine_1.HoverboardBRFullBody.AnimStateNode_2.HoverboardBRMovement.AnimGraphNode_StateMachine_0.HoverboardBRFunction.AnimStateTransitionNode_9.Transition
  nodes [Result(bCanEnterTransition=Break Fort Anim Input Jackal Vehicle(FortAnimInput_JackalVehicle=Get JackalVehicleInput()) -> (
     bIsUsingJackalVehicle=~
     bIsSteeringLeft=~
     bIsSteeringRight=~
     SteerAngleDeadZoneDegrees=~
     SteerAngle=~
     LeanYaw=~
     QuantizedSteerAngle=~
     SteerAlpha=~
     RunForwardAlpha=~
     SlopePitchDegreeAngle=~
     SlopeRollDegreeAngle=~
     DistanceFromGround=~
     PivotPlayRate=~
     bIsReversing=~
     bIsBraking=~
     bIsMoving=~
     bIsMovingForward=~
     bIsSprinting=~
     bInAir=~
     bIsFalling=~
     bIsJumping=~
     bIsRelaxed=~
     bIsBoosting=~
     bHasReachedJumpChargeStartThreshold=~
     bHasReachedJumpChargeMidThreshold=~
     bHasReachedJumpChargeFullThreshold=~
     bAdjustRootForFemaleRider=~
     bIsPlayingEmoteOnHoverboard=~
     bShouldApplyAdditive=~
     bPlayPivotOnGroundAndNotBoosting=~
     bIsOnLowerHill=~
     bLowerBodyIdleToLoopTransition=~
     bInterruptHoverboardFullybody=~
     bEnterFullBodyHoverboardState=~
     bDefaultToJumpStartTransition=~
     bLocomotionPoseToJumpTransition=~
     bBoostingToBoostingJumpTransition=~
     bJumpToLocomotionPoseTransition=~
     bJumpApexToJumpFallTransition=~
     bIdleToMovementStartTransition=~
     bIdleToMovementLoopTransition=~
     bMovementLoopToMovementStopTransition=~
     bMovementLoopToPivotTransition=~
     bMovementLoopToIdleTransition=~
     bIdleAdditiveToCollisionNTransition=~
     bSplitBodyToHoverboardBRTransition=~
     bHoverboardBRToSplitBodyTransition=~
     bHoverboardBRMovementToJumpChargeTransition=~
     bIdlesToJackalVehicleTransition=~
     bPlayAdditiveLeans=~
     bPlayBalloonLeans=~
     bPlayJumpTrickVertical=~
     bPlayJumpTrick=~
     bPlayMovingFast=~
     bPlayHipAdjustmentAdditive=~
     bPlayDriveSouth=~
     bPlayHeadAimOffset=~
     bPlaySlopeAimOffset=~
     JumpCombatAdditiveWeight=~
     MeleeTwistIdle=~
     MeleeTwistLocomotionLoop=~
     ImpactDisplacement=~
     SteerYaw=~
     EmoteHoverboardPosition=~
     EmoteHoverboardRotation=~
     bShouldAttachFeetToHoverboard=~
     FootLeftLocationOffset=~
     FootLeftRotationOffset=~
     FootRightLocationOffset=~
     FootRightRotationOffset=~
     StoppedThreshold=~
     MovingForwardThreshold=~
     MovingFowardFastThreshold=~
     DefaultToJumpStartTransitionThreshold=~
     JumpTrickAngularVelocityThreshold=~
     JumpDistanceFromGroundThreshold=~
     VelocityStartThreshold=~
     MovingFastThreshold=~
     RelaxedSpeedThreshold=~
     JumpChargeStartThreshold=~
     JumpChargeMidThreshold=~
     JumpChargeFullThreshold=~
     RotatingAngularVelocityThreshold=~
     IdleToLoopTransitionSpeedThreshold=~
     IdleToMovementLoopTransitionThreshold=~
     MeleeTwistIdleMultiplier=~
     MeleeTwistLocomotionLoopMultiplier=~
     LeanYawForMaxPivotPlayRate=~
     SteerAlphaForMaxPivotPlayRate=~
    ).bBoostingToBoostingJumpTransition)]
 }
 Transition {
  path=VehicleFullBodyOverride.AnimGraphNode_StateMachine_1.HoverboardBRFullBody.AnimStateTransitionNode_1.Transition
  nodes [Result(bCanEnterTransition=Break Fort Anim Input Jackal Vehicle(FortAnimInput_JackalVehicle=Get JackalVehicleInput()) -> (
     bIsUsingJackalVehicle=~
     bIsSteeringLeft=~
     bIsSteeringRight=~
     SteerAngleDeadZoneDegrees=~
     SteerAngle=~
     LeanYaw=~
     QuantizedSteerAngle=~
     SteerAlpha=~
     RunForwardAlpha=~
     SlopePitchDegreeAngle=~
     SlopeRollDegreeAngle=~
     DistanceFromGround=~
     PivotPlayRate=~
     bIsReversing=~
     bIsBraking=~
     bIsMoving=~
     bIsMovingForward=~
     bIsSprinting=~
     bInAir=~
     bIsFalling=~
     bIsJumping=~
     bIsRelaxed=~
     bIsBoosting=~
     bHasReachedJumpChargeStartThreshold=~
     bHasReachedJumpChargeMidThreshold=~
     bHasReachedJumpChargeFullThreshold=~
     bAdjustRootForFemaleRider=~
     bIsPlayingEmoteOnHoverboard=~
     bShouldApplyAdditive=~
     bPlayPivotOnGroundAndNotBoosting=~
     bIsOnLowerHill=~
     bLowerBodyIdleToLoopTransition=~
     bInterruptHoverboardFullybody=~
     bEnterFullBodyHoverboardState=~
     bDefaultToJumpStartTransition=~
     bLocomotionPoseToJumpTransition=~
     bBoostingToBoostingJumpTransition=~
     bJumpToLocomotionPoseTransition=~
     bJumpApexToJumpFallTransition=~
     bIdleToMovementStartTransition=~
     bIdleToMovementLoopTransition=~
     bMovementLoopToMovementStopTransition=~
     bMovementLoopToPivotTransition=~
     bMovementLoopToIdleTransition=~
     bIdleAdditiveToCollisionNTransition=~
     bSplitBodyToHoverboardBRTransition=~
     bHoverboardBRToSplitBodyTransition=~
     bHoverboardBRMovementToJumpChargeTransition=~
     bIdlesToJackalVehicleTransition=~
     bPlayAdditiveLeans=~
     bPlayBalloonLeans=~
     bPlayJumpTrickVertical=~
     bPlayJumpTrick=~
     bPlayMovingFast=~
     bPlayHipAdjustmentAdditive=~
     bPlayDriveSouth=~
     bPlayHeadAimOffset=~
     bPlaySlopeAimOffset=~
     JumpCombatAdditiveWeight=~
     MeleeTwistIdle=~
     MeleeTwistLocomotionLoop=~
     ImpactDisplacement=~
     SteerYaw=~
     EmoteHoverboardPosition=~
     EmoteHoverboardRotation=~
     bShouldAttachFeetToHoverboard=~
     FootLeftLocationOffset=~
     FootLeftRotationOffset=~
     FootRightLocationOffset=~
     FootRightRotationOffset=~
     StoppedThreshold=~
     MovingForwardThreshold=~
     MovingFowardFastThreshold=~
     DefaultToJumpStartTransitionThreshold=~
     JumpTrickAngularVelocityThreshold=~
     JumpDistanceFromGroundThreshold=~
     VelocityStartThreshold=~
     MovingFastThreshold=~
     RelaxedSpeedThreshold=~
     JumpChargeStartThreshold=~
     JumpChargeMidThreshold=~
     JumpChargeFullThreshold=~
     RotatingAngularVelocityThreshold=~
     IdleToLoopTransitionSpeedThreshold=~
     IdleToMovementLoopTransitionThreshold=~
     MeleeTwistIdleMultiplier=~
     MeleeTwistLocomotionLoopMultiplier=~
     LeanYawForMaxPivotPlayRate=~
     SteerAlphaForMaxPivotPlayRate=~
    ).bIsBoosting)]
 }
 Transition {
  path=VehicleFullBodyOverride.AnimGraphNode_StateMachine_1.HoverboardBRFullBody.AnimStateTransitionNode_12.Transition
  nodes [
   n1 = Get JackalVehicleInput()
   Result(bCanEnterTransition=NOT Boolean(A=Break Fort Anim Input Jackal Vehicle(FortAnimInput_JackalVehicle=@n1) -> (
      bIsUsingJackalVehicle=~
      bIsSteeringLeft=~
      bIsSteeringRight=~
      SteerAngleDeadZoneDegrees=~
      SteerAngle=~
      LeanYaw=~
      QuantizedSteerAngle=~
      SteerAlpha=~
      RunForwardAlpha=~
      SlopePitchDegreeAngle=~
      SlopeRollDegreeAngle=~
      DistanceFromGround=~
      PivotPlayRate=~
      bIsReversing=~
      bIsBraking=~
      bIsMoving=~
      bIsMovingForward=~
      bIsSprinting=~
      bInAir=~
      bIsFalling=~
      bIsJumping=~
      bIsRelaxed=~
      bIsBoosting=~
      bHasReachedJumpChargeStartThreshold=~
      bHasReachedJumpChargeMidThreshold=~
      bHasReachedJumpChargeFullThreshold=~
      bAdjustRootForFemaleRider=~
      bIsPlayingEmoteOnHoverboard=~
      bShouldApplyAdditive=~
      bPlayPivotOnGroundAndNotBoosting=~
      bIsOnLowerHill=~
      bLowerBodyIdleToLoopTransition=~
      bInterruptHoverboardFullybody=~
      bEnterFullBodyHoverboardState=~
      bDefaultToJumpStartTransition=~
      bLocomotionPoseToJumpTransition=~
      bBoostingToBoostingJumpTransition=~
      bJumpToLocomotionPoseTransition=~
      bJumpApexToJumpFallTransition=~
      bIdleToMovementStartTransition=~
      bIdleToMovementLoopTransition=~
      bMovementLoopToMovementStopTransition=~
      bMovementLoopToPivotTransition=~
      bMovementLoopToIdleTransition=~
      bIdleAdditiveToCollisionNTransition=~
      bSplitBodyToHoverboardBRTransition=~
      bHoverboardBRToSplitBodyTransition=~
      bHoverboardBRMovementToJumpChargeTransition=~
      bIdlesToJackalVehicleTransition=~
      bPlayAdditiveLeans=~
      bPlayBalloonLeans=~
      bPlayJumpTrickVertical=~
      bPlayJumpTrick=~
      bPlayMovingFast=~
      bPlayHipAdjustmentAdditive=~
      bPlayDriveSouth=~
      bPlayHeadAimOffset=~
      bPlaySlopeAimOffset=~
      JumpCombatAdditiveWeight=~
      MeleeTwistIdle=~
      MeleeTwistLocomotionLoop=~
      ImpactDisplacement=~
      SteerYaw=~
      EmoteHoverboardPosition=~
      EmoteHoverboardRotation=~
      bShouldAttachFeetToHoverboard=~
      FootLeftLocationOffset=~
      FootLeftRotationOffset=~
      FootRightLocationOffset=~
      FootRightRotationOffset=~
      StoppedThreshold=~
      MovingForwardThreshold=~
      MovingFowardFastThreshold=~
      DefaultToJumpStartTransitionThreshold=~
      JumpTrickAngularVelocityThreshold=~
      JumpDistanceFromGroundThreshold=~
      VelocityStartThreshold=~
      MovingFastThreshold=~
      RelaxedSpeedThreshold=~
      JumpChargeStartThreshold=~
      JumpChargeMidThreshold=~
      JumpChargeFullThreshold=~
      RotatingAngularVelocityThreshold=~
      IdleToLoopTransitionSpeedThreshold=~
      IdleToMovementLoopTransitionThreshold=~
      MeleeTwistIdleMultiplier=~
      MeleeTwistLocomotionLoopMultiplier=~
      LeanYawForMaxPivotPlayRate=~
      SteerAlphaForMaxPivotPlayRate=~
     ).bHasReachedJumpChargeStartThreshold))
  ]
 }
 Transition {
  path=VehicleFullBodyOverride.AnimGraphNode_StateMachine_1.HoverboardBRFullBody.AnimStateTransitionNode_14.Transition
  nodes [Result(bCanEnterTransition=Break Fort Anim Input Jackal Vehicle(FortAnimInput_JackalVehicle=Get JackalVehicleInput()) -> (
     bIsUsingJackalVehicle=~
     bIsSteeringLeft=~
     bIsSteeringRight=~
     SteerAngleDeadZoneDegrees=~
     SteerAngle=~
     LeanYaw=~
     QuantizedSteerAngle=~
     SteerAlpha=~
     RunForwardAlpha=~
     SlopePitchDegreeAngle=~
     SlopeRollDegreeAngle=~
     DistanceFromGround=~
     PivotPlayRate=~
     bIsReversing=~
     bIsBraking=~
     bIsMoving=~
     bIsMovingForward=~
     bIsSprinting=~
     bInAir=~
     bIsFalling=~
     bIsJumping=~
     bIsRelaxed=~
     bIsBoosting=~
     bHasReachedJumpChargeStartThreshold=~
     bHasReachedJumpChargeMidThreshold=~
     bHasReachedJumpChargeFullThreshold=~
     bAdjustRootForFemaleRider=~
     bIsPlayingEmoteOnHoverboard=~
     bShouldApplyAdditive=~
     bPlayPivotOnGroundAndNotBoosting=~
     bIsOnLowerHill=~
     bLowerBodyIdleToLoopTransition=~
     bInterruptHoverboardFullybody=~
     bEnterFullBodyHoverboardState=~
     bDefaultToJumpStartTransition=~
     bLocomotionPoseToJumpTransition=~
     bBoostingToBoostingJumpTransition=~
     bJumpToLocomotionPoseTransition=~
     bJumpApexToJumpFallTransition=~
     bIdleToMovementStartTransition=~
     bIdleToMovementLoopTransition=~
     bMovementLoopToMovementStopTransition=~
     bMovementLoopToPivotTransition=~
     bMovementLoopToIdleTransition=~
     bIdleAdditiveToCollisionNTransition=~
     bSplitBodyToHoverboardBRTransition=~
     bHoverboardBRToSplitBodyTransition=~
     bHoverboardBRMovementToJumpChargeTransition=~
     bIdlesToJackalVehicleTransition=~
     bPlayAdditiveLeans=~
     bPlayBalloonLeans=~
     bPlayJumpTrickVertical=~
     bPlayJumpTrick=~
     bPlayMovingFast=~
     bPlayHipAdjustmentAdditive=~
     bPlayDriveSouth=~
     bPlayHeadAimOffset=~
     bPlaySlopeAimOffset=~
     JumpCombatAdditiveWeight=~
     MeleeTwistIdle=~
     MeleeTwistLocomotionLoop=~
     ImpactDisplacement=~
     SteerYaw=~
     EmoteHoverboardPosition=~
     EmoteHoverboardRotation=~
     bShouldAttachFeetToHoverboard=~
     FootLeftLocationOffset=~
     FootLeftRotationOffset=~
     FootRightLocationOffset=~
     FootRightRotationOffset=~
     StoppedThreshold=~
     MovingForwardThreshold=~
     MovingFowardFastThreshold=~
     DefaultToJumpStartTransitionThreshold=~
     JumpTrickAngularVelocityThreshold=~
     JumpDistanceFromGroundThreshold=~
     VelocityStartThreshold=~
     MovingFastThreshold=~
     RelaxedSpeedThreshold=~
     JumpChargeStartThreshold=~
     JumpChargeMidThreshold=~
     JumpChargeFullThreshold=~
     RotatingAngularVelocityThreshold=~
     IdleToLoopTransitionSpeedThreshold=~
     IdleToMovementLoopTransitionThreshold=~
     MeleeTwistIdleMultiplier=~
     MeleeTwistLocomotionLoopMultiplier=~
     LeanYawForMaxPivotPlayRate=~
     SteerAlphaForMaxPivotPlayRate=~
    ).bHoverboardBRMovementToJumpChargeTransition)]
 }
 HoverboardAdditive {
  entry=IdleAddtive
  path=VehicleFullBodyOverride.AnimGraphNode_StateMachine_2.HoverboardAdditive
  ids=id
  nodes [
   AnimStateNode_3 = CollisionN() {is=title AnimStateNode}
   AnimStateNode_4 = IdleAddtive() {is=title AnimStateNode}
   AnimStateTransitionNode_0 = AnimStateTransitionNode() {crossfadeDuration=0}
   AnimStateTransitionNode_4 = AnimStateTransitionNode() {bAutomaticRuleBasedOnSequencePlayerInState=true}
  ]
  links [
   {
    dst=IdleAddtive
    rule=VehicleFullBodyOverride.AnimGraphNode_StateMachine_2.HoverboardAdditive.AnimStateTransitionNode_4.Transition
    src=CollisionN
    via=AnimStateTransitionNode_4
   }
   {
    dst=CollisionN
    rule=VehicleFullBodyOverride.AnimGraphNode_StateMachine_2.HoverboardAdditive.AnimStateTransitionNode_0.Transition
    src=IdleAddtive
    via=AnimStateTransitionNode_0
   }
  ]
 }
 CollisionN {
  path=VehicleFullBodyOverride.AnimGraphNode_StateMachine_2.HoverboardAdditive.AnimStateNode_3.CollisionN
  nodes [Output Animation Pose(Result="Sequence Player (additive) 'JackalVehicle_Collision_N'"() {bLoopAnimation=false, sequence=/Game/Animation/Game/MainPlayer/Vehicles/JackalVehicle/JackalVehicle_Collision_N})]
 }
 IdleAddtive {
  path=VehicleFullBodyOverride.AnimGraphNode_StateMachine_2.HoverboardAdditive.AnimStateNode_4.IdleAddtive
  nodes [
   n1 = Blend Poses by bool(
    BlendPose_0="Sequence Player (additive) 'JackalVehicle_Idle_Fast_CMM'"() {sequence=/Game/Animation/Game/MainPlayer/Vehicles/JackalVehicle/JackalVehicle_Idle_Fast_CMM}
    BlendPose_1="Sequence Player (additive) 'JackalVehicle_Idle_CMM'"() {sequence=/Game/Animation/Game/MainPlayer/Vehicles/JackalVehicle/JackalVehicle_Idle_CMM}
    bActiveValue=Break Fort Anim Input Jackal Vehicle(FortAnimInput_JackalVehicle=Get JackalVehicleInput()) -> (
     bIsUsingJackalVehicle=~
     bIsSteeringLeft=~
     bIsSteeringRight=~
     SteerAngleDeadZoneDegrees=~
     SteerAngle=~
     LeanYaw=~
     QuantizedSteerAngle=~
     SteerAlpha=~
     RunForwardAlpha=~
     SlopePitchDegreeAngle=~
     SlopeRollDegreeAngle=~
     DistanceFromGround=~
     PivotPlayRate=~
     bIsReversing=~
     bIsBraking=~
     bIsMoving=~
     bIsMovingForward=~
     bIsSprinting=~
     bInAir=~
     bIsFalling=~
     bIsJumping=~
     bIsRelaxed=~
     bIsBoosting=~
     bHasReachedJumpChargeStartThreshold=~
     bHasReachedJumpChargeMidThreshold=~
     bHasReachedJumpChargeFullThreshold=~
     bAdjustRootForFemaleRider=~
     bIsPlayingEmoteOnHoverboard=~
     bShouldApplyAdditive=~
     bPlayPivotOnGroundAndNotBoosting=~
     bIsOnLowerHill=~
     bLowerBodyIdleToLoopTransition=~
     bInterruptHoverboardFullybody=~
     bEnterFullBodyHoverboardState=~
     bDefaultToJumpStartTransition=~
     bLocomotionPoseToJumpTransition=~
     bBoostingToBoostingJumpTransition=~
     bJumpToLocomotionPoseTransition=~
     bJumpApexToJumpFallTransition=~
     bIdleToMovementStartTransition=~
     bIdleToMovementLoopTransition=~
     bMovementLoopToMovementStopTransition=~
     bMovementLoopToPivotTransition=~
     bMovementLoopToIdleTransition=~
     bIdleAdditiveToCollisionNTransition=~
     bSplitBodyToHoverboardBRTransition=~
     bHoverboardBRToSplitBodyTransition=~
     bHoverboardBRMovementToJumpChargeTransition=~
     bIdlesToJackalVehicleTransition=~
     bPlayAdditiveLeans=~
     bPlayBalloonLeans=~
     bPlayJumpTrickVertical=~
     bPlayJumpTrick=~
     bPlayMovingFast=~
     bPlayHipAdjustmentAdditive=~
     bPlayDriveSouth=~
     bPlayHeadAimOffset=~
     bPlaySlopeAimOffset=~
     JumpCombatAdditiveWeight=~
     MeleeTwistIdle=~
     MeleeTwistLocomotionLoop=~
     ImpactDisplacement=~
     SteerYaw=~
     EmoteHoverboardPosition=~
     EmoteHoverboardRotation=~
     bShouldAttachFeetToHoverboard=~
     FootLeftLocationOffset=~
     FootLeftRotationOffset=~
     FootRightLocationOffset=~
     FootRightRotationOffset=~
     StoppedThreshold=~
     MovingForwardThreshold=~
     MovingFowardFastThreshold=~
     DefaultToJumpStartTransitionThreshold=~
     JumpTrickAngularVelocityThreshold=~
     JumpDistanceFromGroundThreshold=~
     VelocityStartThreshold=~
     MovingFastThreshold=~
     RelaxedSpeedThreshold=~
     JumpChargeStartThreshold=~
     JumpChargeMidThreshold=~
     JumpChargeFullThreshold=~
     RotatingAngularVelocityThreshold=~
     IdleToLoopTransitionSpeedThreshold=~
     IdleToMovementLoopTransitionThreshold=~
     MeleeTwistIdleMultiplier=~
     MeleeTwistLocomotionLoopMultiplier=~
     LeanYawForMaxPivotPlayRate=~
     SteerAlphaForMaxPivotPlayRate=~
    ).bPlayMovingFast
   ) {blendType=Linear}
   n2 = "Sequence Player (additive) 'JackalVehicle_HipAdjustment_CMM'"() {sequence=/Game/Animation/Game/MainPlayer/Vehicles/JackalVehicle/JackalVehicle_HipAdjustment_CMM}
   n3 = Break Fort Anim Input Jackal Vehicle(FortAnimInput_JackalVehicle=Get JackalVehicleInput()) -> (
    bIsUsingJackalVehicle=~
    bIsSteeringLeft=~
    bIsSteeringRight=~
    SteerAngleDeadZoneDegrees=~
    SteerAngle=~
    LeanYaw=~
    QuantizedSteerAngle=~
    SteerAlpha=~
    RunForwardAlpha=~
    SlopePitchDegreeAngle=~
    SlopeRollDegreeAngle=~
    DistanceFromGround=~
    PivotPlayRate=~
    bIsReversing=~
    bIsBraking=~
    bIsMoving=~
    bIsMovingForward=~
    bIsSprinting=~
    bInAir=~
    bIsFalling=~
    bIsJumping=~
    bIsRelaxed=~
    bIsBoosting=~
    bHasReachedJumpChargeStartThreshold=~
    bHasReachedJumpChargeMidThreshold=~
    bHasReachedJumpChargeFullThreshold=~
    bAdjustRootForFemaleRider=~
    bIsPlayingEmoteOnHoverboard=~
    bShouldApplyAdditive=~
    bPlayPivotOnGroundAndNotBoosting=~
    bIsOnLowerHill=~
    bLowerBodyIdleToLoopTransition=~
    bInterruptHoverboardFullybody=~
    bEnterFullBodyHoverboardState=~
    bDefaultToJumpStartTransition=~
    bLocomotionPoseToJumpTransition=~
    bBoostingToBoostingJumpTransition=~
    bJumpToLocomotionPoseTransition=~
    bJumpApexToJumpFallTransition=~
    bIdleToMovementStartTransition=~
    bIdleToMovementLoopTransition=~
    bMovementLoopToMovementStopTransition=~
    bMovementLoopToPivotTransition=~
    bMovementLoopToIdleTransition=~
    bIdleAdditiveToCollisionNTransition=~
    bSplitBodyToHoverboardBRTransition=~
    bHoverboardBRToSplitBodyTransition=~
    bHoverboardBRMovementToJumpChargeTransition=~
    bIdlesToJackalVehicleTransition=~
    bPlayAdditiveLeans=~
    bPlayBalloonLeans=~
    bPlayJumpTrickVertical=~
    bPlayJumpTrick=~
    bPlayMovingFast=~
    bPlayHipAdjustmentAdditive=~
    bPlayDriveSouth=~
    bPlayHeadAimOffset=~
    bPlaySlopeAimOffset=~
    JumpCombatAdditiveWeight=~
    MeleeTwistIdle=~
    MeleeTwistLocomotionLoop=~
    ImpactDisplacement=~
    SteerYaw=~
    EmoteHoverboardPosition=~
    EmoteHoverboardRotation=~
    bShouldAttachFeetToHoverboard=~
    FootLeftLocationOffset=~
    FootLeftRotationOffset=~
    FootRightLocationOffset=~
    FootRightRotationOffset=~
    StoppedThreshold=~
    MovingForwardThreshold=~
    MovingFowardFastThreshold=~
    DefaultToJumpStartTransitionThreshold=~
    JumpTrickAngularVelocityThreshold=~
    JumpDistanceFromGroundThreshold=~
    VelocityStartThreshold=~
    MovingFastThreshold=~
    RelaxedSpeedThreshold=~
    JumpChargeStartThreshold=~
    JumpChargeMidThreshold=~
    JumpChargeFullThreshold=~
    RotatingAngularVelocityThreshold=~
    IdleToLoopTransitionSpeedThreshold=~
    IdleToMovementLoopTransitionThreshold=~
    MeleeTwistIdleMultiplier=~
    MeleeTwistLocomotionLoopMultiplier=~
    LeanYawForMaxPivotPlayRate=~
    SteerAlphaForMaxPivotPlayRate=~
   )
   n4 = Get LocalVelocityYawAngle() -> (LocalVelocityYawAngle:float)
   Output Animation Pose(Result=Apply Additive(
     Base=Apply Additive(Base=@n1, Additive=@n2, bAlphaBoolEnabled=@n3.bPlayHipAdjustmentAdditive) {alphaInputType=Bool, alphaBoolBlend {blendInTime=0.2, blendOption=Linear, blendOutTime=0.2}}
     Additive=Blendspace Player 'JackalVehicle_Lean_Balloons'!(X=@n4) -> (Pose) {blendSpace=/Game/Animation/Game/MainPlayer/Vehicles/JackalVehicle/JackalVehicle_Lean_Balloons}
     bAlphaBoolEnabled=@n3.bPlayBalloonLeans
    ) {alphaInputType=Bool, alphaBoolBlend {blendInTime=0.2, blendOption=Linear, blendOutTime=0.2}})
  ]
 }
 Transition {
  path=VehicleFullBodyOverride.AnimGraphNode_StateMachine_2.HoverboardAdditive.AnimStateTransitionNode_0.Transition
  nodes [Result(bCanEnterTransition=Break Fort Anim Input Jackal Vehicle(FortAnimInput_JackalVehicle=Get JackalVehicleInput()) -> (
     bIsUsingJackalVehicle=~
     bIsSteeringLeft=~
     bIsSteeringRight=~
     SteerAngleDeadZoneDegrees=~
     SteerAngle=~
     LeanYaw=~
     QuantizedSteerAngle=~
     SteerAlpha=~
     RunForwardAlpha=~
     SlopePitchDegreeAngle=~
     SlopeRollDegreeAngle=~
     DistanceFromGround=~
     PivotPlayRate=~
     bIsReversing=~
     bIsBraking=~
     bIsMoving=~
     bIsMovingForward=~
     bIsSprinting=~
     bInAir=~
     bIsFalling=~
     bIsJumping=~
     bIsRelaxed=~
     bIsBoosting=~
     bHasReachedJumpChargeStartThreshold=~
     bHasReachedJumpChargeMidThreshold=~
     bHasReachedJumpChargeFullThreshold=~
     bAdjustRootForFemaleRider=~
     bIsPlayingEmoteOnHoverboard=~
     bShouldApplyAdditive=~
     bPlayPivotOnGroundAndNotBoosting=~
     bIsOnLowerHill=~
     bLowerBodyIdleToLoopTransition=~
     bInterruptHoverboardFullybody=~
     bEnterFullBodyHoverboardState=~
     bDefaultToJumpStartTransition=~
     bLocomotionPoseToJumpTransition=~
     bBoostingToBoostingJumpTransition=~
     bJumpToLocomotionPoseTransition=~
     bJumpApexToJumpFallTransition=~
     bIdleToMovementStartTransition=~
     bIdleToMovementLoopTransition=~
     bMovementLoopToMovementStopTransition=~
     bMovementLoopToPivotTransition=~
     bMovementLoopToIdleTransition=~
     bIdleAdditiveToCollisionNTransition=~
     bSplitBodyToHoverboardBRTransition=~
     bHoverboardBRToSplitBodyTransition=~
     bHoverboardBRMovementToJumpChargeTransition=~
     bIdlesToJackalVehicleTransition=~
     bPlayAdditiveLeans=~
     bPlayBalloonLeans=~
     bPlayJumpTrickVertical=~
     bPlayJumpTrick=~
     bPlayMovingFast=~
     bPlayHipAdjustmentAdditive=~
     bPlayDriveSouth=~
     bPlayHeadAimOffset=~
     bPlaySlopeAimOffset=~
     JumpCombatAdditiveWeight=~
     MeleeTwistIdle=~
     MeleeTwistLocomotionLoop=~
     ImpactDisplacement=~
     SteerYaw=~
     EmoteHoverboardPosition=~
     EmoteHoverboardRotation=~
     bShouldAttachFeetToHoverboard=~
     FootLeftLocationOffset=~
     FootLeftRotationOffset=~
     FootRightLocationOffset=~
     FootRightRotationOffset=~
     StoppedThreshold=~
     MovingForwardThreshold=~
     MovingFowardFastThreshold=~
     DefaultToJumpStartTransitionThreshold=~
     JumpTrickAngularVelocityThreshold=~
     JumpDistanceFromGroundThreshold=~
     VelocityStartThreshold=~
     MovingFastThreshold=~
     RelaxedSpeedThreshold=~
     JumpChargeStartThreshold=~
     JumpChargeMidThreshold=~
     JumpChargeFullThreshold=~
     RotatingAngularVelocityThreshold=~
     IdleToLoopTransitionSpeedThreshold=~
     IdleToMovementLoopTransitionThreshold=~
     MeleeTwistIdleMultiplier=~
     MeleeTwistLocomotionLoopMultiplier=~
     LeanYawForMaxPivotPlayRate=~
     SteerAlphaForMaxPivotPlayRate=~
    ).bIdleAdditiveToCollisionNTransition)]
 }
 Transition {path=VehicleFullBodyOverride.AnimGraphNode_StateMachine_2.HoverboardAdditive.AnimStateTransitionNode_4.Transition, nodes [Result()]}
 VehicleLowerBodyOverride {nodes [
   Input Pose() {name=InPoseUpperBody}
   n1 = Sequence Player 'JackalVehicle_JumpCharge_Combat_Start'() {
    bLoopAnimation=false
    groupName=ChargeStart
    method=SyncGroup
    sequence=/Game/Animation/Game/MainPlayer/Vehicles/JackalVehicle/JackalVehicle_JumpCharge_Combat_Start
   }
   n2 = Break Fort Anim Input Jackal Vehicle(FortAnimInput_JackalVehicle=Get JackalVehicleInput()) -> (
    bIsUsingJackalVehicle=~
    bIsSteeringLeft=~
    bIsSteeringRight=~
    SteerAngleDeadZoneDegrees=~
    SteerAngle=~
    LeanYaw=~
    QuantizedSteerAngle=~
    SteerAlpha=~
    RunForwardAlpha=~
    SlopePitchDegreeAngle=~
    SlopeRollDegreeAngle=~
    DistanceFromGround=~
    PivotPlayRate=~
    bIsReversing=~
    bIsBraking=~
    bIsMoving=~
    bIsMovingForward=~
    bIsSprinting=~
    bInAir=~
    bIsFalling=~
    bIsJumping=~
    bIsRelaxed=~
    bIsBoosting=~
    bHasReachedJumpChargeStartThreshold=~
    bHasReachedJumpChargeMidThreshold=~
    bHasReachedJumpChargeFullThreshold=~
    bAdjustRootForFemaleRider=~
    bIsPlayingEmoteOnHoverboard=~
    bShouldApplyAdditive=~
    bPlayPivotOnGroundAndNotBoosting=~
    bIsOnLowerHill=~
    bLowerBodyIdleToLoopTransition=~
    bInterruptHoverboardFullybody=~
    bEnterFullBodyHoverboardState=~
    bDefaultToJumpStartTransition=~
    bLocomotionPoseToJumpTransition=~
    bBoostingToBoostingJumpTransition=~
    bJumpToLocomotionPoseTransition=~
    bJumpApexToJumpFallTransition=~
    bIdleToMovementStartTransition=~
    bIdleToMovementLoopTransition=~
    bMovementLoopToMovementStopTransition=~
    bMovementLoopToPivotTransition=~
    bMovementLoopToIdleTransition=~
    bIdleAdditiveToCollisionNTransition=~
    bSplitBodyToHoverboardBRTransition=~
    bHoverboardBRToSplitBodyTransition=~
    bHoverboardBRMovementToJumpChargeTransition=~
    bIdlesToJackalVehicleTransition=~
    bPlayAdditiveLeans=~
    bPlayBalloonLeans=~
    bPlayJumpTrickVertical=~
    bPlayJumpTrick=~
    bPlayMovingFast=~
    bPlayHipAdjustmentAdditive=~
    bPlayDriveSouth=~
    bPlayHeadAimOffset=~
    bPlaySlopeAimOffset=~
    JumpCombatAdditiveWeight=~
    MeleeTwistIdle=~
    MeleeTwistLocomotionLoop=~
    ImpactDisplacement=~
    SteerYaw=~
    EmoteHoverboardPosition=~
    EmoteHoverboardRotation=~
    bShouldAttachFeetToHoverboard=~
    FootLeftLocationOffset=~
    FootLeftRotationOffset=~
    FootRightLocationOffset=~
    FootRightRotationOffset=~
    StoppedThreshold=~
    MovingForwardThreshold=~
    MovingFowardFastThreshold=~
    DefaultToJumpStartTransitionThreshold=~
    JumpTrickAngularVelocityThreshold=~
    JumpDistanceFromGroundThreshold=~
    VelocityStartThreshold=~
    MovingFastThreshold=~
    RelaxedSpeedThreshold=~
    JumpChargeStartThreshold=~
    JumpChargeMidThreshold=~
    JumpChargeFullThreshold=~
    RotatingAngularVelocityThreshold=~
    IdleToLoopTransitionSpeedThreshold=~
    IdleToMovementLoopTransitionThreshold=~
    MeleeTwistIdleMultiplier=~
    MeleeTwistLocomotionLoopMultiplier=~
    LeanYawForMaxPivotPlayRate=~
    SteerAlphaForMaxPivotPlayRate=~
   )
   n3 = Rotate Root Bone!(
    BasePose:Pose Link Structure=Blend Poses by bool(
     BlendPose_0=Sequence Player 'JackalVehicle_JumpCharge_Combat_Pose'() {sequence=/Game/Animation/Game/MainPlayer/Vehicles/JackalVehicle/JackalVehicle_JumpCharge_Combat_Pose}
     BlendPose_1=@n1
     bActiveValue=@n2.bHasReachedJumpChargeMidThreshold
    ) {blendType=Linear}
    Yaw:float=@n2.MeleeTwistLocomotionLoop
    MeshToComponent:Rotator="0,-90, 0"
   ) -> (Pose:Pose Link Structure) {is=title AnimGraphNode_RotateRootBone}
   AnimGraphNode_StateMachine_0 = JackalVehicleLowerBody() {is=title AnimGraphNode_StateMachine}
   n4 = Get CommonVehicleInput() -> (CommonVehicleInput:Fort Anim Input Common Vehicle Structure)
   Output Pose(Result=Blend Poses by bool(
     BlendPose_0=Use cached pose 'LowerBodyInputPoseFromMainPlayer'()
     BlendPose_1=Blend Poses by bool(BlendPose_0=@n3, BlendPose_1=@AnimGraphNode_StateMachine_0, bActiveValue=@n2.bHasReachedJumpChargeStartThreshold) {blendType=Linear}
     BlendTime_0=0.2
     BlendTime_1=0.4
     bActiveValue=Break Fort Anim Input Common Vehicle!(FortAnimInput_CommonVehicle:"Fort Anim Input Common Vehicle Structure (by ref)"=@n4) -> (
      bIsUsingVehicle:bool=~
      bIsJumpingVehicle:bool=~
      bCanChargeJump:bool=~
      bIsChargingJump:bool=~
      bIsJumping:bool=~
      bIsOnGround:bool=~
      bCanDriverAimWeapon:bool=~
     ) {is=title K2Node_BreakStruct}.bIsJumping
    ) {blendType=Linear}) {layerGroup=VehicleOverride}
   AnimGraphNode_SubInput_0 = Input Pose() {name=InPoseLowerBodyDefault}
   Save cached pose 'LowerBodyInputPoseFromMainPlayer'(Pose=@AnimGraphNode_SubInput_0) {cacheName=LowerBodyInputPoseFromMainPlayer}
  ]}
 JackalVehicleLowerBody {
  entry=Idle
  path=VehicleLowerBodyOverride.AnimGraphNode_StateMachine_0.JackalVehicleLowerBody
  ids=id
  nodes [
   AnimStateNode_0 = Idle() {bAlwaysResetOnEntry=true}
   AnimStateNode_1 = Start() {is=title AnimStateNode}
   AnimStateNode_2 = Loop() {is=title AnimStateNode}
   AnimStateNode_3 = Stop() {is=title AnimStateNode}
   AnimStateTransitionNode_0 = AnimStateTransitionNode() {priorityOrder=2}
   AnimStateTransitionNode_1 = AnimStateTransitionNode() {bAutomaticRuleBasedOnSequencePlayerInState=true, crossfadeDuration=0.4}
   AnimStateTransitionNode_2 = AnimStateTransitionNode()
   AnimStateTransitionNode_3 = AnimStateTransitionNode() {bAutomaticRuleBasedOnSequencePlayerInState=true, crossfadeDuration=0.35}
   AnimStateTransitionNode_4 = AnimStateTransitionNode()
  ]
  links [
   {
    dst=Loop
    rule=VehicleLowerBodyOverride.AnimGraphNode_StateMachine_0.JackalVehicleLowerBody.AnimStateTransitionNode_4.Transition
    src=Idle
    via=AnimStateTransitionNode_4
   }
   {
    dst=Start
    rule=VehicleLowerBodyOverride.AnimGraphNode_StateMachine_0.JackalVehicleLowerBody.AnimStateTransitionNode_0.Transition
    src=Idle
    via=AnimStateTransitionNode_0
   }
   {
    dst=Stop
    rule=VehicleLowerBodyOverride.AnimGraphNode_StateMachine_0.JackalVehicleLowerBody.AnimStateTransitionNode_2.Transition
    src=Loop
    via=AnimStateTransitionNode_2
   }
   {
    dst=Loop
    rule=VehicleLowerBodyOverride.AnimGraphNode_StateMachine_0.JackalVehicleLowerBody.AnimStateTransitionNode_1.Transition
    src=Start
    via=AnimStateTransitionNode_1
   }
   {
    dst=Idle
    rule=VehicleLowerBodyOverride.AnimGraphNode_StateMachine_0.JackalVehicleLowerBody.AnimStateTransitionNode_3.Transition
    src=Stop
    via=AnimStateTransitionNode_3
   }
  ]
 }
 Idle {
  path=VehicleLowerBodyOverride.AnimGraphNode_StateMachine_0.JackalVehicleLowerBody.AnimStateNode_0.Idle
  nodes [
   n1 = Get JackalVehicleInput()
   Output Animation Pose(Result=Rotate Root Bone(
     BasePose=Sequence Player 'JackalVehicle_Idle_Lower_Pose_CMM'() {sequence=/Game/Animation/Game/MainPlayer/Vehicles/JackalVehicle/JackalVehicle_Idle_Lower_Pose_CMM}
     Yaw=Break Fort Anim Input Jackal Vehicle(FortAnimInput_JackalVehicle=@n1) -> (
      bIsUsingJackalVehicle=~
      bIsSteeringLeft=~
      bIsSteeringRight=~
      SteerAngleDeadZoneDegrees=~
      SteerAngle=~
      LeanYaw=~
      QuantizedSteerAngle=~
      SteerAlpha=~
      RunForwardAlpha=~
      SlopePitchDegreeAngle=~
      SlopeRollDegreeAngle=~
      DistanceFromGround=~
      PivotPlayRate=~
      bIsReversing=~
      bIsBraking=~
      bIsMoving=~
      bIsMovingForward=~
      bIsSprinting=~
      bInAir=~
      bIsFalling=~
      bIsJumping=~
      bIsRelaxed=~
      bIsBoosting=~
      bHasReachedJumpChargeStartThreshold=~
      bHasReachedJumpChargeMidThreshold=~
      bHasReachedJumpChargeFullThreshold=~
      bAdjustRootForFemaleRider=~
      bIsPlayingEmoteOnHoverboard=~
      bShouldApplyAdditive=~
      bPlayPivotOnGroundAndNotBoosting=~
      bIsOnLowerHill=~
      bLowerBodyIdleToLoopTransition=~
      bInterruptHoverboardFullybody=~
      bEnterFullBodyHoverboardState=~
      bDefaultToJumpStartTransition=~
      bLocomotionPoseToJumpTransition=~
      bBoostingToBoostingJumpTransition=~
      bJumpToLocomotionPoseTransition=~
      bJumpApexToJumpFallTransition=~
      bIdleToMovementStartTransition=~
      bIdleToMovementLoopTransition=~
      bMovementLoopToMovementStopTransition=~
      bMovementLoopToPivotTransition=~
      bMovementLoopToIdleTransition=~
      bIdleAdditiveToCollisionNTransition=~
      bSplitBodyToHoverboardBRTransition=~
      bHoverboardBRToSplitBodyTransition=~
      bHoverboardBRMovementToJumpChargeTransition=~
      bIdlesToJackalVehicleTransition=~
      bPlayAdditiveLeans=~
      bPlayBalloonLeans=~
      bPlayJumpTrickVertical=~
      bPlayJumpTrick=~
      bPlayMovingFast=~
      bPlayHipAdjustmentAdditive=~
      bPlayDriveSouth=~
      bPlayHeadAimOffset=~
      bPlaySlopeAimOffset=~
      JumpCombatAdditiveWeight=~
      MeleeTwistIdle=~
      MeleeTwistLocomotionLoop=~
      ImpactDisplacement=~
      SteerYaw=~
      EmoteHoverboardPosition=~
      EmoteHoverboardRotation=~
      bShouldAttachFeetToHoverboard=~
      FootLeftLocationOffset=~
      FootLeftRotationOffset=~
      FootRightLocationOffset=~
      FootRightRotationOffset=~
      StoppedThreshold=~
      MovingForwardThreshold=~
      MovingFowardFastThreshold=~
      DefaultToJumpStartTransitionThreshold=~
      JumpTrickAngularVelocityThreshold=~
      JumpDistanceFromGroundThreshold=~
      VelocityStartThreshold=~
      MovingFastThreshold=~
      RelaxedSpeedThreshold=~
      JumpChargeStartThreshold=~
      JumpChargeMidThreshold=~
      JumpChargeFullThreshold=~
      RotatingAngularVelocityThreshold=~
      IdleToLoopTransitionSpeedThreshold=~
      IdleToMovementLoopTransitionThreshold=~
      MeleeTwistIdleMultiplier=~
      MeleeTwistLocomotionLoopMultiplier=~
      LeanYawForMaxPivotPlayRate=~
      SteerAlphaForMaxPivotPlayRate=~
     ).MeleeTwistIdle
     MeshToComponent="0,-90, 0"
    ))
  ]
 }
 Start {
  path=VehicleLowerBodyOverride.AnimGraphNode_StateMachine_0.JackalVehicleLowerBody.AnimStateNode_1.Start
  nodes [
   n1 = "Blend Poses (EFortCardinalDirection)"(
    BlendPose_0=Sequence Player 'JackalVehicle_Drive_N_Combat_Start'() {bLoopAnimation=false, sequence=/Game/Animation/Game/MainPlayer/Vehicles/JackalVehicle/JackalVehicle_Drive_N_Combat_Start}
    BlendPose_1=Sequence Player 'JackalVehicle_Drive_S_Combat_Start'() {bLoopAnimation=false, sequence=/Game/Animation/Game/MainPlayer/Vehicles/JackalVehicle/JackalVehicle_Drive_S_Combat_Start}
    BlendTime_0=0.1
    BlendTime_1=0.1
    ActiveEnumValue=Get LocomotionCardinalDirection() -> (LocomotionCardinalDirection=North)
   ) {blendType=Linear, blendPose [{linkId=-1, sourceLinkId=-1}, {linkId=-1, sourceLinkId=-1}], blendTime [0.1,0.1]}
   Output Animation Pose(Result=@n1)
  ]
 }
 Loop {
  path=VehicleLowerBodyOverride.AnimGraphNode_StateMachine_0.JackalVehicleLowerBody.AnimStateNode_2.Loop
  nodes [
   n1 = Get JackalVehicleInput()
   Output Animation Pose(Result=Rotate Root Bone(
     BasePose=Sequence Player 'JackalVehicle_Idle_Lower_Pose_CMM'() {sequence=&JackalVehicle_Idle_Lower_Pose_CMM}
     Yaw=Break Fort Anim Input Jackal Vehicle(FortAnimInput_JackalVehicle=@n1) -> (
      bIsUsingJackalVehicle=~
      bIsSteeringLeft=~
      bIsSteeringRight=~
      SteerAngleDeadZoneDegrees=~
      SteerAngle=~
      LeanYaw=~
      QuantizedSteerAngle=~
      SteerAlpha=~
      RunForwardAlpha=~
      SlopePitchDegreeAngle=~
      SlopeRollDegreeAngle=~
      DistanceFromGround=~
      PivotPlayRate=~
      bIsReversing=~
      bIsBraking=~
      bIsMoving=~
      bIsMovingForward=~
      bIsSprinting=~
      bInAir=~
      bIsFalling=~
      bIsJumping=~
      bIsRelaxed=~
      bIsBoosting=~
      bHasReachedJumpChargeStartThreshold=~
      bHasReachedJumpChargeMidThreshold=~
      bHasReachedJumpChargeFullThreshold=~
      bAdjustRootForFemaleRider=~
      bIsPlayingEmoteOnHoverboard=~
      bShouldApplyAdditive=~
      bPlayPivotOnGroundAndNotBoosting=~
      bIsOnLowerHill=~
      bLowerBodyIdleToLoopTransition=~
      bInterruptHoverboardFullybody=~
      bEnterFullBodyHoverboardState=~
      bDefaultToJumpStartTransition=~
      bLocomotionPoseToJumpTransition=~
      bBoostingToBoostingJumpTransition=~
      bJumpToLocomotionPoseTransition=~
      bJumpApexToJumpFallTransition=~
      bIdleToMovementStartTransition=~
      bIdleToMovementLoopTransition=~
      bMovementLoopToMovementStopTransition=~
      bMovementLoopToPivotTransition=~
      bMovementLoopToIdleTransition=~
      bIdleAdditiveToCollisionNTransition=~
      bSplitBodyToHoverboardBRTransition=~
      bHoverboardBRToSplitBodyTransition=~
      bHoverboardBRMovementToJumpChargeTransition=~
      bIdlesToJackalVehicleTransition=~
      bPlayAdditiveLeans=~
      bPlayBalloonLeans=~
      bPlayJumpTrickVertical=~
      bPlayJumpTrick=~
      bPlayMovingFast=~
      bPlayHipAdjustmentAdditive=~
      bPlayDriveSouth=~
      bPlayHeadAimOffset=~
      bPlaySlopeAimOffset=~
      JumpCombatAdditiveWeight=~
      MeleeTwistIdle=~
      MeleeTwistLocomotionLoop=~
      ImpactDisplacement=~
      SteerYaw=~
      EmoteHoverboardPosition=~
      EmoteHoverboardRotation=~
      bShouldAttachFeetToHoverboard=~
      FootLeftLocationOffset=~
      FootLeftRotationOffset=~
      FootRightLocationOffset=~
      FootRightRotationOffset=~
      StoppedThreshold=~
      MovingForwardThreshold=~
      MovingFowardFastThreshold=~
      DefaultToJumpStartTransitionThreshold=~
      JumpTrickAngularVelocityThreshold=~
      JumpDistanceFromGroundThreshold=~
      VelocityStartThreshold=~
      MovingFastThreshold=~
      RelaxedSpeedThreshold=~
      JumpChargeStartThreshold=~
      JumpChargeMidThreshold=~
      JumpChargeFullThreshold=~
      RotatingAngularVelocityThreshold=~
      IdleToLoopTransitionSpeedThreshold=~
      IdleToMovementLoopTransitionThreshold=~
      MeleeTwistIdleMultiplier=~
      MeleeTwistLocomotionLoopMultiplier=~
      LeanYawForMaxPivotPlayRate=~
      SteerAlphaForMaxPivotPlayRate=~
     ).MeleeTwistLocomotionLoop
     MeshToComponent="0,-90, 0"
    ))
  ]
 }
 Stop {
  path=VehicleLowerBodyOverride.AnimGraphNode_StateMachine_0.JackalVehicleLowerBody.AnimStateNode_3.Stop
  nodes [Output Animation Pose(Result=Sequence Player 'JackalVehicle_Drive_N_Combat_Stop'() {bLoopAnimation=false, sequence=/Game/Animation/Game/MainPlayer/Vehicles/JackalVehicle/JackalVehicle_Drive_N_Combat_Stop})]
 }
 Transition {
  path=VehicleLowerBodyOverride.AnimGraphNode_StateMachine_0.JackalVehicleLowerBody.AnimStateTransitionNode_0.Transition
  nodes [Result(bCanEnterTransition=Break Fort Anim Input Jackal Vehicle(FortAnimInput_JackalVehicle=Get JackalVehicleInput()) -> (
     bIsUsingJackalVehicle=~
     bIsSteeringLeft=~
     bIsSteeringRight=~
     SteerAngleDeadZoneDegrees=~
     SteerAngle=~
     LeanYaw=~
     QuantizedSteerAngle=~
     SteerAlpha=~
     RunForwardAlpha=~
     SlopePitchDegreeAngle=~
     SlopeRollDegreeAngle=~
     DistanceFromGround=~
     PivotPlayRate=~
     bIsReversing=~
     bIsBraking=~
     bIsMoving=~
     bIsMovingForward=~
     bIsSprinting=~
     bInAir=~
     bIsFalling=~
     bIsJumping=~
     bIsRelaxed=~
     bIsBoosting=~
     bHasReachedJumpChargeStartThreshold=~
     bHasReachedJumpChargeMidThreshold=~
     bHasReachedJumpChargeFullThreshold=~
     bAdjustRootForFemaleRider=~
     bIsPlayingEmoteOnHoverboard=~
     bShouldApplyAdditive=~
     bPlayPivotOnGroundAndNotBoosting=~
     bIsOnLowerHill=~
     bLowerBodyIdleToLoopTransition=~
     bInterruptHoverboardFullybody=~
     bEnterFullBodyHoverboardState=~
     bDefaultToJumpStartTransition=~
     bLocomotionPoseToJumpTransition=~
     bBoostingToBoostingJumpTransition=~
     bJumpToLocomotionPoseTransition=~
     bJumpApexToJumpFallTransition=~
     bIdleToMovementStartTransition=~
     bIdleToMovementLoopTransition=~
     bMovementLoopToMovementStopTransition=~
     bMovementLoopToPivotTransition=~
     bMovementLoopToIdleTransition=~
     bIdleAdditiveToCollisionNTransition=~
     bSplitBodyToHoverboardBRTransition=~
     bHoverboardBRToSplitBodyTransition=~
     bHoverboardBRMovementToJumpChargeTransition=~
     bIdlesToJackalVehicleTransition=~
     bPlayAdditiveLeans=~
     bPlayBalloonLeans=~
     bPlayJumpTrickVertical=~
     bPlayJumpTrick=~
     bPlayMovingFast=~
     bPlayHipAdjustmentAdditive=~
     bPlayDriveSouth=~
     bPlayHeadAimOffset=~
     bPlaySlopeAimOffset=~
     JumpCombatAdditiveWeight=~
     MeleeTwistIdle=~
     MeleeTwistLocomotionLoop=~
     ImpactDisplacement=~
     SteerYaw=~
     EmoteHoverboardPosition=~
     EmoteHoverboardRotation=~
     bShouldAttachFeetToHoverboard=~
     FootLeftLocationOffset=~
     FootLeftRotationOffset=~
     FootRightLocationOffset=~
     FootRightRotationOffset=~
     StoppedThreshold=~
     MovingForwardThreshold=~
     MovingFowardFastThreshold=~
     DefaultToJumpStartTransitionThreshold=~
     JumpTrickAngularVelocityThreshold=~
     JumpDistanceFromGroundThreshold=~
     VelocityStartThreshold=~
     MovingFastThreshold=~
     RelaxedSpeedThreshold=~
     JumpChargeStartThreshold=~
     JumpChargeMidThreshold=~
     JumpChargeFullThreshold=~
     RotatingAngularVelocityThreshold=~
     IdleToLoopTransitionSpeedThreshold=~
     IdleToMovementLoopTransitionThreshold=~
     MeleeTwistIdleMultiplier=~
     MeleeTwistLocomotionLoopMultiplier=~
     LeanYawForMaxPivotPlayRate=~
     SteerAlphaForMaxPivotPlayRate=~
    ).bIsMovingForward)]
 }
 Transition {path=VehicleLowerBodyOverride.AnimGraphNode_StateMachine_0.JackalVehicleLowerBody.AnimStateTransitionNode_1.Transition, nodes [Result()]}
 Transition {
  path=VehicleLowerBodyOverride.AnimGraphNode_StateMachine_0.JackalVehicleLowerBody.AnimStateTransitionNode_2.Transition
  nodes [
   n1 = Get JackalVehicleInput()
   Result(bCanEnterTransition=NOT Boolean(A=Break Fort Anim Input Jackal Vehicle(FortAnimInput_JackalVehicle=@n1) -> (
      bIsUsingJackalVehicle=~
      bIsSteeringLeft=~
      bIsSteeringRight=~
      SteerAngleDeadZoneDegrees=~
      SteerAngle=~
      LeanYaw=~
      QuantizedSteerAngle=~
      SteerAlpha=~
      RunForwardAlpha=~
      SlopePitchDegreeAngle=~
      SlopeRollDegreeAngle=~
      DistanceFromGround=~
      PivotPlayRate=~
      bIsReversing=~
      bIsBraking=~
      bIsMoving=~
      bIsMovingForward=~
      bIsSprinting=~
      bInAir=~
      bIsFalling=~
      bIsJumping=~
      bIsRelaxed=~
      bIsBoosting=~
      bHasReachedJumpChargeStartThreshold=~
      bHasReachedJumpChargeMidThreshold=~
      bHasReachedJumpChargeFullThreshold=~
      bAdjustRootForFemaleRider=~
      bIsPlayingEmoteOnHoverboard=~
      bShouldApplyAdditive=~
      bPlayPivotOnGroundAndNotBoosting=~
      bIsOnLowerHill=~
      bLowerBodyIdleToLoopTransition=~
      bInterruptHoverboardFullybody=~
      bEnterFullBodyHoverboardState=~
      bDefaultToJumpStartTransition=~
      bLocomotionPoseToJumpTransition=~
      bBoostingToBoostingJumpTransition=~
      bJumpToLocomotionPoseTransition=~
      bJumpApexToJumpFallTransition=~
      bIdleToMovementStartTransition=~
      bIdleToMovementLoopTransition=~
      bMovementLoopToMovementStopTransition=~
      bMovementLoopToPivotTransition=~
      bMovementLoopToIdleTransition=~
      bIdleAdditiveToCollisionNTransition=~
      bSplitBodyToHoverboardBRTransition=~
      bHoverboardBRToSplitBodyTransition=~
      bHoverboardBRMovementToJumpChargeTransition=~
      bIdlesToJackalVehicleTransition=~
      bPlayAdditiveLeans=~
      bPlayBalloonLeans=~
      bPlayJumpTrickVertical=~
      bPlayJumpTrick=~
      bPlayMovingFast=~
      bPlayHipAdjustmentAdditive=~
      bPlayDriveSouth=~
      bPlayHeadAimOffset=~
      bPlaySlopeAimOffset=~
      JumpCombatAdditiveWeight=~
      MeleeTwistIdle=~
      MeleeTwistLocomotionLoop=~
      ImpactDisplacement=~
      SteerYaw=~
      EmoteHoverboardPosition=~
      EmoteHoverboardRotation=~
      bShouldAttachFeetToHoverboard=~
      FootLeftLocationOffset=~
      FootLeftRotationOffset=~
      FootRightLocationOffset=~
      FootRightRotationOffset=~
      StoppedThreshold=~
      MovingForwardThreshold=~
      MovingFowardFastThreshold=~
      DefaultToJumpStartTransitionThreshold=~
      JumpTrickAngularVelocityThreshold=~
      JumpDistanceFromGroundThreshold=~
      VelocityStartThreshold=~
      MovingFastThreshold=~
      RelaxedSpeedThreshold=~
      JumpChargeStartThreshold=~
      JumpChargeMidThreshold=~
      JumpChargeFullThreshold=~
      RotatingAngularVelocityThreshold=~
      IdleToLoopTransitionSpeedThreshold=~
      IdleToMovementLoopTransitionThreshold=~
      MeleeTwistIdleMultiplier=~
      MeleeTwistLocomotionLoopMultiplier=~
      LeanYawForMaxPivotPlayRate=~
      SteerAlphaForMaxPivotPlayRate=~
     ).bIsMovingForward))
  ]
 }
 Transition {path=VehicleLowerBodyOverride.AnimGraphNode_StateMachine_0.JackalVehicleLowerBody.AnimStateTransitionNode_3.Transition, nodes [Result()]}
 Transition {
  path=VehicleLowerBodyOverride.AnimGraphNode_StateMachine_0.JackalVehicleLowerBody.AnimStateTransitionNode_4.Transition
  nodes [Result(bCanEnterTransition=Break Fort Anim Input Jackal Vehicle(FortAnimInput_JackalVehicle=Get JackalVehicleInput()) -> (
     bIsUsingJackalVehicle=~
     bIsSteeringLeft=~
     bIsSteeringRight=~
     SteerAngleDeadZoneDegrees=~
     SteerAngle=~
     LeanYaw=~
     QuantizedSteerAngle=~
     SteerAlpha=~
     RunForwardAlpha=~
     SlopePitchDegreeAngle=~
     SlopeRollDegreeAngle=~
     DistanceFromGround=~
     PivotPlayRate=~
     bIsReversing=~
     bIsBraking=~
     bIsMoving=~
     bIsMovingForward=~
     bIsSprinting=~
     bInAir=~
     bIsFalling=~
     bIsJumping=~
     bIsRelaxed=~
     bIsBoosting=~
     bHasReachedJumpChargeStartThreshold=~
     bHasReachedJumpChargeMidThreshold=~
     bHasReachedJumpChargeFullThreshold=~
     bAdjustRootForFemaleRider=~
     bIsPlayingEmoteOnHoverboard=~
     bShouldApplyAdditive=~
     bPlayPivotOnGroundAndNotBoosting=~
     bIsOnLowerHill=~
     bLowerBodyIdleToLoopTransition=~
     bInterruptHoverboardFullybody=~
     bEnterFullBodyHoverboardState=~
     bDefaultToJumpStartTransition=~
     bLocomotionPoseToJumpTransition=~
     bBoostingToBoostingJumpTransition=~
     bJumpToLocomotionPoseTransition=~
     bJumpApexToJumpFallTransition=~
     bIdleToMovementStartTransition=~
     bIdleToMovementLoopTransition=~
     bMovementLoopToMovementStopTransition=~
     bMovementLoopToPivotTransition=~
     bMovementLoopToIdleTransition=~
     bIdleAdditiveToCollisionNTransition=~
     bSplitBodyToHoverboardBRTransition=~
     bHoverboardBRToSplitBodyTransition=~
     bHoverboardBRMovementToJumpChargeTransition=~
     bIdlesToJackalVehicleTransition=~
     bPlayAdditiveLeans=~
     bPlayBalloonLeans=~
     bPlayJumpTrickVertical=~
     bPlayJumpTrick=~
     bPlayMovingFast=~
     bPlayHipAdjustmentAdditive=~
     bPlayDriveSouth=~
     bPlayHeadAimOffset=~
     bPlaySlopeAimOffset=~
     JumpCombatAdditiveWeight=~
     MeleeTwistIdle=~
     MeleeTwistLocomotionLoop=~
     ImpactDisplacement=~
     SteerYaw=~
     EmoteHoverboardPosition=~
     EmoteHoverboardRotation=~
     bShouldAttachFeetToHoverboard=~
     FootLeftLocationOffset=~
     FootLeftRotationOffset=~
     FootRightLocationOffset=~
     FootRightRotationOffset=~
     StoppedThreshold=~
     MovingForwardThreshold=~
     MovingFowardFastThreshold=~
     DefaultToJumpStartTransitionThreshold=~
     JumpTrickAngularVelocityThreshold=~
     JumpDistanceFromGroundThreshold=~
     VelocityStartThreshold=~
     MovingFastThreshold=~
     RelaxedSpeedThreshold=~
     JumpChargeStartThreshold=~
     JumpChargeMidThreshold=~
     JumpChargeFullThreshold=~
     RotatingAngularVelocityThreshold=~
     IdleToLoopTransitionSpeedThreshold=~
     IdleToMovementLoopTransitionThreshold=~
     MeleeTwistIdleMultiplier=~
     MeleeTwistLocomotionLoopMultiplier=~
     LeanYawForMaxPivotPlayRate=~
     SteerAlphaForMaxPivotPlayRate=~
    ).bLowerBodyIdleToLoopTransition)]
 }
 VehicleSplitBodyOverride {nodes [
   Input Pose() {name=InPoseUpperAndLowerBody}
   n1 = Break Fort Anim Input Jackal Vehicle(FortAnimInput_JackalVehicle=Get JackalVehicleInput()) -> (
    bIsUsingJackalVehicle=~
    bIsSteeringLeft=~
    bIsSteeringRight=~
    SteerAngleDeadZoneDegrees=~
    SteerAngle=~
    LeanYaw=~
    QuantizedSteerAngle=~
    SteerAlpha=~
    RunForwardAlpha=~
    SlopePitchDegreeAngle=~
    SlopeRollDegreeAngle=~
    DistanceFromGround=~
    PivotPlayRate=~
    bIsReversing=~
    bIsBraking=~
    bIsMoving=~
    bIsMovingForward=~
    bIsSprinting=~
    bInAir=~
    bIsFalling=~
    bIsJumping=~
    bIsRelaxed=~
    bIsBoosting=~
    bHasReachedJumpChargeStartThreshold=~
    bHasReachedJumpChargeMidThreshold=~
    bHasReachedJumpChargeFullThreshold=~
    bAdjustRootForFemaleRider=~
    bIsPlayingEmoteOnHoverboard=~
    bShouldApplyAdditive=~
    bPlayPivotOnGroundAndNotBoosting=~
    bIsOnLowerHill=~
    bLowerBodyIdleToLoopTransition=~
    bInterruptHoverboardFullybody=~
    bEnterFullBodyHoverboardState=~
    bDefaultToJumpStartTransition=~
    bLocomotionPoseToJumpTransition=~
    bBoostingToBoostingJumpTransition=~
    bJumpToLocomotionPoseTransition=~
    bJumpApexToJumpFallTransition=~
    bIdleToMovementStartTransition=~
    bIdleToMovementLoopTransition=~
    bMovementLoopToMovementStopTransition=~
    bMovementLoopToPivotTransition=~
    bMovementLoopToIdleTransition=~
    bIdleAdditiveToCollisionNTransition=~
    bSplitBodyToHoverboardBRTransition=~
    bHoverboardBRToSplitBodyTransition=~
    bHoverboardBRMovementToJumpChargeTransition=~
    bIdlesToJackalVehicleTransition=~
    bPlayAdditiveLeans=~
    bPlayBalloonLeans=~
    bPlayJumpTrickVertical=~
    bPlayJumpTrick=~
    bPlayMovingFast=~
    bPlayHipAdjustmentAdditive=~
    bPlayDriveSouth=~
    bPlayHeadAimOffset=~
    bPlaySlopeAimOffset=~
    JumpCombatAdditiveWeight=~
    MeleeTwistIdle=~
    MeleeTwistLocomotionLoop=~
    ImpactDisplacement=~
    SteerYaw=~
    EmoteHoverboardPosition=~
    EmoteHoverboardRotation=~
    bShouldAttachFeetToHoverboard=~
    FootLeftLocationOffset=~
    FootLeftRotationOffset=~
    FootRightLocationOffset=~
    FootRightRotationOffset=~
    StoppedThreshold=~
    MovingForwardThreshold=~
    MovingFowardFastThreshold=~
    DefaultToJumpStartTransitionThreshold=~
    JumpTrickAngularVelocityThreshold=~
    JumpDistanceFromGroundThreshold=~
    VelocityStartThreshold=~
    MovingFastThreshold=~
    RelaxedSpeedThreshold=~
    JumpChargeStartThreshold=~
    JumpChargeMidThreshold=~
    JumpChargeFullThreshold=~
    RotatingAngularVelocityThreshold=~
    IdleToLoopTransitionSpeedThreshold=~
    IdleToMovementLoopTransitionThreshold=~
    MeleeTwistIdleMultiplier=~
    MeleeTwistLocomotionLoopMultiplier=~
    LeanYawForMaxPivotPlayRate=~
    SteerAlphaForMaxPivotPlayRate=~
   )
   n2 = AimOffset Player 'JackalVehicle_Slope_Pitch_AO'!(X=@n1.SlopePitchDegreeAngle, BasePose=Use cached pose 'InputPoseNoRoot'(), Alpha:float=1) -> (Pose) {blendSpace=&JackalVehicle_Slope_Pitch_AO, alphaBoolBlend {blendOption=Linear}}
   n3 = AimOffset Player 'JackalVehicle_Slope_Roll_AO'!(X=@n1.SlopeRollDegreeAngle, BasePose=@n2, Alpha:float=1) -> (Pose) {blendSpace=&JackalVehicle_Slope_Roll_AO, alphaBoolBlend {blendOption=Linear}}
   Output Pose(Result=Blend Poses by bool(BlendPose_0=@n3, BlendPose_1=Use cached pose 'InputPose'(), BlendTime_0=0, BlendTime_1=0.15, bActiveValue=@n1.bPlaySlopeAimOffset) {blendType=Linear}) {layerGroup=Vehicle Override}
   Save cached pose 'InputPose'(Pose=Input Pose() {name=InPoseSplitBody}) {cacheName=InputPose}
   Save cached pose 'InputPoseNoRoot'(Pose=Input Pose() {name=InPoseUpperAndLowerBodyRemoveRoot}) {cacheName=InputPoseNoRoot}
  ]}
]
variables [ShouldTriggerGrappleMontage {type=boolean}]
