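# Base Animation Blueprint for the drivable Jackal hoverboard, animating the board on its own vehicle skeleton.
# A FullBody state machine switches combat and sprint modes, nesting movement, jump, trick and boost states fed by cached poses.
# An additive idle layers on top, then a bIsBothLegFK blend brings in the grappling-hook machine.
# The output adds sprint or combat yaw leans, falling back to idle without a driver or attached feet.
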AnimBlueprint /Game/Athena/DrivableVehicles/Hoverboard/JackalVehicle_Hoverboard_Base_AnimBP
target_skeleton=&/SK_Jackal_Skeleton_Athena
graphs [
 AnimGraph {nodes [
   n1 = Get CombatLeanYaw() -> (CombatLeanYaw:float)
   n2 = Get LeanYaw() -> (LeanYaw:float)
   n3 = Blend Poses by bool(
    BlendPose_0=Apply Additive!(
     Base=Use cached pose 'LocomotionPose'()
     Additive=Blendspace Player 'JackalVehicle_Leans_BS_Hoverboard'!(X=@n2) -> (Pose) {blendSpace=/Game/Animation/Game/MainPlayer/Vehicles/JackalVehicle/Parts/JackalVehicle_Leans_BS_Hoverboard}
     bAlphaBoolEnabled:bool=Get bPlayAdditiveLeans() -> (bPlayAdditiveLeans:bool)
    ) -> (Pose) {alphaInputType=Bool, alphaBoolBlend {blendInTime=0.2, blendOption=Linear, blendOutTime=0.2}}
    BlendPose_1=Apply Additive(
     Base=Use cached pose 'LocomotionPose'()
     Additive=Blendspace Player 'JackalVehicle_Leans_BS_Hoverboard'(X=@n1) {blendSpace=&JackalVehicle_Leans_BS_Hoverboard}
     bAlphaBoolEnabled=Get bPlayAdditiveLeans()
    ) {alphaInputType=Bool, alphaBoolBlend {blendInTime=0.15, blendOption=Linear, blendOutTime=0.15}}
    BlendTime_0=0
    BlendTime_1=0
    bActiveValue=Get bIsSprinting()
   ) {blendType=Linear}
   n4 = Sequence Player 'JackalVehicle_Idle_Hoverboard'() {sequence=/Game/Animation/Game/MainPlayer/Vehicles/JackalVehicle/Parts/JackalVehicle_Idle_Hoverboard}
   n5 = Get bHasDriver() -> (bHasDriver:bool)
   Output Pose(Result=Blend Poses by bool(
     BlendPose_0=Blend Poses by bool(BlendPose_0=@n3, BlendPose_1=@n4, BlendTime_0=0, BlendTime_1=0, bActiveValue=@n5) {blendType=Linear}
     BlendPose_1=Sequence Player 'JackalVehicle_Idle_Hoverboard'() {sequence=&JackalVehicle_Idle_Hoverboard}
     bActiveValue=Get AreFeetAttachedToBoard() -> (AreFeetAttachedToBoard:bool)
    ) {blendType=Linear}) {layerGroup=None}
   AnimGraphNode_StateMachine_1 = FullBody() {is=title AnimGraphNode_StateMachine}
   n6 = Blend Poses by bool(
    BlendPose_0=Sequence Player 'JackalVehicle_Idle_Hoverboard'() {sequence=&JackalVehicle_Idle_Hoverboard}
    BlendPose_1=@AnimGraphNode_StateMachine_1
    bActiveValue=Get bIsPlayingEmoteOnHoverboard() -> (bIsPlayingEmoteOnHoverboard:bool)
   ) {blendType=Linear}
   n7 = "Sequence Player (additive) 'JackalVehicle_Idle_CMM_Hoverboard'"() {sequence=/Game/Animation/Game/MainPlayer/Vehicles/JackalVehicle/Parts/JackalVehicle_Idle_CMM_Hoverboard}
   AnimGraphNode_StateMachine_2 = GrapplingHook() {is=title AnimGraphNode_StateMachine}
   Save cached pose 'LocomotionPose'(Pose=Two Way Blend!(
     A=Apply Additive!(Base=@n6, Additive=@n7, Alpha:float) -> (Pose) {alphaBoolBlend {blendOption=Linear}}
     B=@AnimGraphNode_StateMachine_2
     bAlphaBoolEnabled:bool=Get bIsBothLegFK() -> (bIsBothLegFK:bool)
    ) -> (Pose) {blendNode {alphaInputType=Bool, bResetChildOnActivation=true, alphaBoolBlend {blendInTime=0.15, blendOutTime=0.15}}}) {cacheName=LocomotionPose}
   AnimGraphNode_StateMachine_3 = Locomotion() {is=title AnimGraphNode_StateMachine}
   Save cached pose 'SprintingPose'(Pose=@AnimGraphNode_StateMachine_3) {cacheName=SprintingPose}
   n8 = Blendspace Player 'JackalVehicle_Leans_InPlace_BS_Hoverboard'!(X=Get LeanYaw()) -> (Pose) {blendSpace=/Game/Animation/Game/MainPlayer/Vehicles/JackalVehicle/Parts/JackalVehicle_Leans_InPlace_BS_Hoverboard}
   n9 = Sequence Player 'JackalVehicle_Idle_Hoverboard'() {sequence=&JackalVehicle_Idle_Hoverboard}
   Save cached pose 'IdlePose'(Pose=Blend Poses by bool(
     BlendPose_0=Sequence Player 'JackalVehicle_Idle_Hoverboard'() {sequence=&JackalVehicle_Idle_Hoverboard}
     BlendPose_1=Apply Additive(Base=@n9, Additive=@n8) {alphaBoolBlend {blendOption=Linear}}
     BlendTime_0=0.25
     BlendTime_1=0.25
     bActiveValue=Get bIsMoving() -> (bIsMoving:bool)
    ) {blendType=Linear}) {cacheName=IdlePose}
   AnimGraphNode_StateMachine_0 = CombatHoverboardMovement() {is=title AnimGraphNode_StateMachine}
   Save cached pose 'CombatPose'(Pose=@AnimGraphNode_StateMachine_0) {cacheName=CombatPose}
  ]}
 CombatHoverboardMovement {
  entry=Idle
  path=AnimGraph.AnimGraphNode_StateMachine_0.CombatHoverboardMovement
  ids=id
  nodes [
   AnimStateNode_0 = Idle()
   AnimStateNode_13 = Start_MAVG() {is=title AnimStateNode}
   AnimStateNode_14 = Loop_MAVG() {is=title AnimStateNode}
   AnimStateNode_15 = Stop_MAVG() {is=title AnimStateNode}
   AnimStateTransitionNode_10 = AnimStateTransitionNode()
   AnimStateTransitionNode_44 = AnimStateTransitionNode() {bAutomaticRuleBasedOnSequencePlayerInState=true, crossfadeDuration=0.4}
   AnimStateTransitionNode_45 = AnimStateTransitionNode() {priorityOrder=2}
   AnimStateTransitionNode_9 = AnimStateTransitionNode() {bAutomaticRuleBasedOnSequencePlayerInState=true, crossfadeDuration=0.35}
  ]
  links [
   {
    dst=Start_MAVG
    rule=AnimGraph.AnimGraphNode_StateMachine_0.CombatHoverboardMovement.AnimStateTransitionNode_10.Transition
    src=Idle
    via=AnimStateTransitionNode_10
   }
   {
    dst=Stop_MAVG
    rule=AnimGraph.AnimGraphNode_StateMachine_0.CombatHoverboardMovement.AnimStateTransitionNode_45.AnimationTransitionGraph_1
    src=Loop_MAVG
    via=AnimStateTransitionNode_45
   }
   {
    dst=Loop_MAVG
    rule=AnimGraph.AnimGraphNode_StateMachine_0.CombatHoverboardMovement.AnimStateTransitionNode_44.AnimationTransitionGraph_1
    src=Start_MAVG
    via=AnimStateTransitionNode_44
   }
   {
    dst=Idle
    rule=AnimGraph.AnimGraphNode_StateMachine_0.CombatHoverboardMovement.AnimStateTransitionNode_9.Transition
    src=Stop_MAVG
    via=AnimStateTransitionNode_9
   }
  ]
 }
 Idle {
  path=AnimGraph.AnimGraphNode_StateMachine_0.CombatHoverboardMovement.AnimStateNode_0.Idle
  nodes [Output Animation Pose(Result=Use cached pose 'IdlePose'())]
 }
 Start_MAVG {
  path=AnimGraph.AnimGraphNode_StateMachine_0.CombatHoverboardMovement.AnimStateNode_13.Start_MAVG
  nodes [
   n1 = Sequence Player 'JackalVehicle_Drive_N_Combat_Start_Hoverboard'() {bLoopAnimation=false, sequence=/Game/Animation/Game/MainPlayer/Vehicles/JackalVehicle/Parts/JackalVehicle_Drive_N_Combat_Start_Hoverboard}
   n2 = Sequence Player 'JackalVehicle_Drive_S_Combat_Start_Hoverboard'() {bLoopAnimation=false, sequence=/Game/Animation/Game/MainPlayer/Vehicles/JackalVehicle/Parts/JackalVehicle_Drive_S_Combat_Start_Hoverboard}
   n3 = "Blend Poses (EFortCardinalDirection)"!(
    BlendPose_0:Pose Link Structure=@n1
    BlendPose_1:Pose Link Structure=@n2
    BlendTime_0:float=0.2
    BlendTime_1:float=0.2
    ActiveEnumValue:Byte=Get LocomotionCardinalDirection() -> (LocomotionCardinalDirection:EFortCardinalDirection Enum=North)
   ) -> (Pose:Pose Link Structure) {
    is=title AnimGraphNode_BlendListByEnum
    blendType=Linear
    blendPose [{linkId=-1, sourceLinkId=-1}, {linkId=-1, sourceLinkId=-1}]
    blendTime [0.1,0.1]
   }
   Output Animation Pose(Result=@n3)
  ]
 }
 Loop_MAVG {
  path=AnimGraph.AnimGraphNode_StateMachine_0.CombatHoverboardMovement.AnimStateNode_14.Loop_MAVG
  nodes [Output Animation Pose(Result=Sequence Player 'JackalVehicle_Idle_Hoverboard'() {sequence=&JackalVehicle_Idle_Hoverboard})]
 }
 Stop_MAVG {
  path=AnimGraph.AnimGraphNode_StateMachine_0.CombatHoverboardMovement.AnimStateNode_15.Stop_MAVG
  nodes [Output Animation Pose(Result=Sequence Player 'JackalVehicle_Drive_N_Stop_Hoverboard'() {bLoopAnimation=false, sequence=/Game/Animation/Game/MainPlayer/Vehicles/JackalVehicle/Parts/JackalVehicle_Drive_N_Stop_Hoverboard})]
 }
 Transition {
  path=AnimGraph.AnimGraphNode_StateMachine_0.CombatHoverboardMovement.AnimStateTransitionNode_10.Transition
  nodes [Result(bCanEnterTransition=Get bIsMoving())]
 }
 AnimationTransitionGraph_1 {path=AnimGraph.AnimGraphNode_StateMachine_0.CombatHoverboardMovement.AnimStateTransitionNode_44.AnimationTransitionGraph_1, nodes [Result()]}
 AnimationTransitionGraph_1 {
  path=AnimGraph.AnimGraphNode_StateMachine_0.CombatHoverboardMovement.AnimStateTransitionNode_45.AnimationTransitionGraph_1
  nodes [Result(bCanEnterTransition=NOT Boolean(A=Get bIsMoving()))]
 }
 Transition {path=AnimGraph.AnimGraphNode_StateMachine_0.CombatHoverboardMovement.AnimStateTransitionNode_9.Transition, nodes [Result()]}
 FullBody {
  entry=CombatMode
  path=AnimGraph.AnimGraphNode_StateMachine_1.FullBody
  ids=id
  nodes [
   AnimStateNode_0 = CombatMode() {is=title AnimStateNode}
   AnimStateNode_1 = SprintMode() {is=title AnimStateNode}
   AnimStateTransitionNode_7 = AnimStateTransitionNode()
   AnimStateTransitionNode_8 = AnimStateTransitionNode()
  ]
  links [
   {
    dst=SprintMode
    rule=AnimGraph.AnimGraphNode_StateMachine_1.FullBody.AnimStateTransitionNode_7.Transition
    src=CombatMode
    via=AnimStateTransitionNode_7
   }
   {
    dst=CombatMode
    rule=AnimGraph.AnimGraphNode_StateMachine_1.FullBody.AnimStateTransitionNode_8.Transition
    src=SprintMode
    via=AnimStateTransitionNode_8
   }
  ]
 }
 CombatMode {
  path=AnimGraph.AnimGraphNode_StateMachine_1.FullBody.AnimStateNode_0.CombatMode
  nodes [
   AnimGraphNode_StateMachine_0 = CombatHoverboard() {is=title AnimGraphNode_StateMachine}
   Output Animation Pose(Result=@AnimGraphNode_StateMachine_0)
  ]
 }
 CombatHoverboard {
  entry=Default
  path=AnimGraph.AnimGraphNode_StateMachine_1.FullBody.AnimStateNode_0.CombatMode.AnimGraphNode_StateMachine_0.CombatHoverboard
  ids=id
  nodes [
   AnimStateNode_0 = JumpCombatStart() {is=title AnimStateNode}
   AnimStateNode_1 = Default() {is=title AnimStateNode}
   AnimStateNode_10 = JumpCombatLand() {is=title AnimStateNode}
   AnimStateNode_6 = JumpCombatLoop() {is=title AnimStateNode}
   AnimStateNode_9 = JumpCombatFall() {is=title AnimStateNode}
   AnimStateTransitionNode_0 = AnimStateTransitionNode() {bAutomaticRuleBasedOnSequencePlayerInState=true}
   AnimStateTransitionNode_1 = AnimStateTransitionNode() {crossfadeDuration=0.25}
   AnimStateTransitionNode_19 = AnimStateTransitionNode() {crossfadeDuration=0.08}
   AnimStateTransitionNode_20 = AnimStateTransitionNode() {bAutomaticRuleBasedOnSequencePlayerInState=true}
   AnimStateTransitionNode_21 = AnimStateTransitionNode()
   AnimStateTransitionNode_22 = AnimStateTransitionNode() {crossfadeDuration=0.04}
   AnimStateTransitionNode_42 = AnimStateTransitionNode() {crossfadeDuration=0.04}
   AnimStateTransitionNode_43 = AnimStateTransitionNode()
  ]
  links [
   {
    dst=JumpCombatStart
    rule=AnimGraph.AnimGraphNode_StateMachine_1.FullBody.AnimStateNode_0.CombatMode.AnimGraphNode_StateMachine_0.CombatHoverboard.AnimStateTransitionNode_1.Transition
    src=Default
    via=AnimStateTransitionNode_1
   }
   {
    dst=JumpCombatLand
    rule=AnimGraph.AnimGraphNode_StateMachine_1.FullBody.AnimStateNode_0.CombatMode.AnimGraphNode_StateMachine_0.CombatHoverboard.AnimStateTransitionNode_22.AnimationTransitionGraph_0
    src=JumpCombatFall
    via=AnimStateTransitionNode_22
   }
   {
    dst=JumpCombatLand
    rule=AnimGraph.AnimGraphNode_StateMachine_1.FullBody.AnimStateNode_0.CombatMode.AnimGraphNode_StateMachine_0.CombatHoverboard.AnimStateTransitionNode_42.AnimationTransitionGraph_0
    src=JumpCombatFall
    via=AnimStateTransitionNode_42
   }
   {
    dst=JumpCombatLand
    rule=AnimGraph.AnimGraphNode_StateMachine_1.FullBody.AnimStateNode_0.CombatMode.AnimGraphNode_StateMachine_0.CombatHoverboard.AnimStateTransitionNode_43.AnimationTransitionGraph_0
    src=JumpCombatFall
    via=AnimStateTransitionNode_43
   }
   {
    dst=Default
    rule=AnimGraph.AnimGraphNode_StateMachine_1.FullBody.AnimStateNode_0.CombatMode.AnimGraphNode_StateMachine_0.CombatHoverboard.AnimStateTransitionNode_0.Transition
    src=JumpCombatLand
    via=AnimStateTransitionNode_0
   }
   {
    dst=Default
    rule=AnimGraph.AnimGraphNode_StateMachine_1.FullBody.AnimStateNode_0.CombatMode.AnimGraphNode_StateMachine_0.CombatHoverboard.AnimStateTransitionNode_19.Transition
    src=JumpCombatLand
    via=AnimStateTransitionNode_19
   }
   {
    dst=JumpCombatFall
    rule=AnimGraph.AnimGraphNode_StateMachine_1.FullBody.AnimStateNode_0.CombatMode.AnimGraphNode_StateMachine_0.CombatHoverboard.AnimStateTransitionNode_21.AnimationTransitionGraph_0
    src=JumpCombatLoop
    via=AnimStateTransitionNode_21
   }
   {
    dst=JumpCombatLoop
    rule=AnimGraph.AnimGraphNode_StateMachine_1.FullBody.AnimStateNode_0.CombatMode.AnimGraphNode_StateMachine_0.CombatHoverboard.AnimStateTransitionNode_20.AnimationTransitionGraph_0
    src=JumpCombatStart
    via=AnimStateTransitionNode_20
   }
  ]
 }
 JumpCombatStart {
  path=AnimGraph.AnimGraphNode_StateMachine_1.FullBody.AnimStateNode_0.CombatMode.AnimGraphNode_StateMachine_0.CombatHoverboard.AnimStateNode_0.JumpCombatStart
  nodes [
   n1 = Sequence Player 'JackalVehicle_JumpCombat_Start_Hoverboard'() {bLoopAnimation=false, sequence=/Game/Animation/Game/MainPlayer/Vehicles/JackalVehicle/Parts/JackalVehicle_JumpCombat_Start_Hoverboard}
   Output Animation Pose(Result=@n1)
  ]
 }
 Default {
  path=AnimGraph.AnimGraphNode_StateMachine_1.FullBody.AnimStateNode_0.CombatMode.AnimGraphNode_StateMachine_0.CombatHoverboard.AnimStateNode_1.Default
  nodes [Output Animation Pose(Result=Use cached pose 'CombatPose'())]
 }
 JumpCombatLand {
  path=AnimGraph.AnimGraphNode_StateMachine_1.FullBody.AnimStateNode_0.CombatMode.AnimGraphNode_StateMachine_0.CombatHoverboard.AnimStateNode_10.JumpCombatLand
  nodes [
   n1 = Sequence Player 'JackalVehicle_JumpCombat_Land_Hoverboard'() {bLoopAnimation=false, sequence=/Game/Animation/Game/MainPlayer/Vehicles/JackalVehicle/Parts/JackalVehicle_JumpCombat_Land_Hoverboard}
   Output Animation Pose(Result=@n1)
  ]
 }
 JumpCombatLoop {
  path=AnimGraph.AnimGraphNode_StateMachine_1.FullBody.AnimStateNode_0.CombatMode.AnimGraphNode_StateMachine_0.CombatHoverboard.AnimStateNode_6.JumpCombatLoop
  nodes [Output Animation Pose(Result=Sequence Player 'JackalVehicle_JumpCombat_Loop_Hoverboard'() {sequence=/Game/Animation/Game/MainPlayer/Vehicles/JackalVehicle/Parts/JackalVehicle_JumpCombat_Loop_Hoverboard})]
 }
 JumpCombatFall {
  path=AnimGraph.AnimGraphNode_StateMachine_1.FullBody.AnimStateNode_0.CombatMode.AnimGraphNode_StateMachine_0.CombatHoverboard.AnimStateNode_9.JumpCombatFall
  nodes [
   n1 = Sequence Player 'JackalVehicle_JumpCombat_Fall_Hoverboard'() {bLoopAnimation=false, sequence=/Game/Animation/Game/MainPlayer/Vehicles/JackalVehicle/Parts/JackalVehicle_JumpCombat_Fall_Hoverboard}
   Output Animation Pose(Result=@n1)
  ]
 }
 Transition {
  path=AnimGraph.AnimGraphNode_StateMachine_1.FullBody.AnimStateNode_0.CombatMode.AnimGraphNode_StateMachine_0.CombatHoverboard.AnimStateTransitionNode_0.Transition
  nodes [Result()]
 }
 Transition {
  path=AnimGraph.AnimGraphNode_StateMachine_1.FullBody.AnimStateNode_0.CombatMode.AnimGraphNode_StateMachine_0.CombatHoverboard.AnimStateTransitionNode_1.Transition
  nodes [Result(bCanEnterTransition=Get bDefaultToJumpCombatStartTransition() -> (bDefaultToJumpCombatStartTransition:bool))]
 }
 Transition {
  path=AnimGraph.AnimGraphNode_StateMachine_1.FullBody.AnimStateNode_0.CombatMode.AnimGraphNode_StateMachine_0.CombatHoverboard.AnimStateTransitionNode_19.Transition
  nodes [Result(bCanEnterTransition=Get bIsVehicleJumping() -> (bIsVehicleJumping:bool))]
 }
 AnimationTransitionGraph_0 {
  path=AnimGraph.AnimGraphNode_StateMachine_1.FullBody.AnimStateNode_0.CombatMode.AnimGraphNode_StateMachine_0.CombatHoverboard.AnimStateTransitionNode_20.AnimationTransitionGraph_0
  nodes [Result()]
 }
 AnimationTransitionGraph_0 {
  path=AnimGraph.AnimGraphNode_StateMachine_1.FullBody.AnimStateNode_0.CombatMode.AnimGraphNode_StateMachine_0.CombatHoverboard.AnimStateTransitionNode_21.AnimationTransitionGraph_0
  nodes [Result(bCanEnterTransition=NOT Boolean(A=Get bInAir() -> (bInAir:bool)))]
 }
 AnimationTransitionGraph_0 {
  path=AnimGraph.AnimGraphNode_StateMachine_1.FullBody.AnimStateNode_0.CombatMode.AnimGraphNode_StateMachine_0.CombatHoverboard.AnimStateTransitionNode_22.AnimationTransitionGraph_0
  nodes [Result(bCanEnterTransition=NOT Boolean(A=Get bInAir()))]
 }
 AnimationTransitionGraph_0 {
  path=AnimGraph.AnimGraphNode_StateMachine_1.FullBody.AnimStateNode_0.CombatMode.AnimGraphNode_StateMachine_0.CombatHoverboard.AnimStateTransitionNode_42.AnimationTransitionGraph_0
  nodes [Result(bCanEnterTransition=Get bIsVehicleJumping())]
 }
 AnimationTransitionGraph_0 {
  path=AnimGraph.AnimGraphNode_StateMachine_1.FullBody.AnimStateNode_0.CombatMode.AnimGraphNode_StateMachine_0.CombatHoverboard.AnimStateTransitionNode_43.AnimationTransitionGraph_0
  nodes [Result(bCanEnterTransition=NOT Boolean(A=Get bInAir()))]
 }
 SprintMode {
  path=AnimGraph.AnimGraphNode_StateMachine_1.FullBody.AnimStateNode_1.SprintMode
  nodes [
   AnimGraphNode_StateMachine_0 = SprintingHoverboardMovement() {is=title AnimGraphNode_StateMachine}
   Output Animation Pose(Result=@AnimGraphNode_StateMachine_0)
  ]
 }
 SprintingHoverboardMovement {
  entry=LocomotionPose
  path=AnimGraph.AnimGraphNode_StateMachine_1.FullBody.AnimStateNode_1.SprintMode.AnimGraphNode_StateMachine_0.SprintingHoverboardMovement
  ids=id
  nodes [
   AnimStateNode_0 = LocomotionPose() {is=title AnimStateNode}
   AnimStateNode_1 = Jump() {is=title AnimStateNode}
   AnimStateNode_2 = Boosting() {is=title AnimStateNode}
   AnimStateTransitionNode_11 = AnimStateTransitionNode() {crossfadeDuration=0.15}
   AnimStateTransitionNode_12 = AnimStateTransitionNode() {crossfadeDuration=0.6}
   AnimStateTransitionNode_13 = AnimStateTransitionNode() {crossfadeDuration=0.1}
   AnimStateTransitionNode_14 = AnimStateTransitionNode() {crossfadeDuration=0.45}
   AnimStateTransitionNode_15 = AnimStateTransitionNode() {crossfadeDuration=0}
  ]
  links [
   {
    dst=LocomotionPose
    rule=AnimGraph.AnimGraphNode_StateMachine_1.FullBody.AnimStateNode_1.SprintMode.AnimGraphNode_StateMachine_0.SprintingHoverboardMovement.AnimStateTransitionNode_14.Transition
    src=Boosting
    via=AnimStateTransitionNode_14
   }
   {
    dst=Boosting
    rule=AnimGraph.AnimGraphNode_StateMachine_1.FullBody.AnimStateNode_1.SprintMode.AnimGraphNode_StateMachine_0.SprintingHoverboardMovement.AnimStateTransitionNode_15.Transition
    src=Jump
    via=AnimStateTransitionNode_15
   }
   {
    dst=LocomotionPose
    rule=AnimGraph.AnimGraphNode_StateMachine_1.FullBody.AnimStateNode_1.SprintMode.AnimGraphNode_StateMachine_0.SprintingHoverboardMovement.AnimStateTransitionNode_12.Transition
    src=Jump
    via=AnimStateTransitionNode_12
   }
   {
    dst=Boosting
    rule=AnimGraph.AnimGraphNode_StateMachine_1.FullBody.AnimStateNode_1.SprintMode.AnimGraphNode_StateMachine_0.SprintingHoverboardMovement.AnimStateTransitionNode_13.Transition
    src=LocomotionPose
    via=AnimStateTransitionNode_13
   }
   {
    dst=Jump
    rule=AnimGraph.AnimGraphNode_StateMachine_1.FullBody.AnimStateNode_1.SprintMode.AnimGraphNode_StateMachine_0.SprintingHoverboardMovement.AnimStateTransitionNode_11.Transition
    src=LocomotionPose
    via=AnimStateTransitionNode_11
   }
  ]
 }
 LocomotionPose {
  path=AnimGraph.AnimGraphNode_StateMachine_1.FullBody.AnimStateNode_1.SprintMode.AnimGraphNode_StateMachine_0.SprintingHoverboardMovement.AnimStateNode_0.LocomotionPose
  nodes [Output Animation Pose(Result=Use cached pose 'SprintingPose'())]
 }
 Jump {
  path=AnimGraph.AnimGraphNode_StateMachine_1.FullBody.AnimStateNode_1.SprintMode.AnimGraphNode_StateMachine_0.SprintingHoverboardMovement.AnimStateNode_1.Jump
  nodes [
   AnimGraphNode_StateMachine_0 = SpritingJump() {is=title AnimGraphNode_StateMachine}
   Output Animation Pose(Result=@AnimGraphNode_StateMachine_0)
  ]
 }
 SpritingJump {
  entry=Default
  path=AnimGraph.AnimGraphNode_StateMachine_1.FullBody.AnimStateNode_1.SprintMode.AnimGraphNode_StateMachine_0.SprintingHoverboardMovement.AnimStateNode_1.Jump.AnimGraphNode_StateMachine_0.SpritingJump
  ids=id
  nodes [
   AnimStateNode_0 = Default()
   AnimStateNode_1 = JumpApex() {is=title AnimStateNode}
   AnimStateNode_11 = JumpStart() {is=title AnimStateNode}
   AnimStateNode_2 = JumpFall() {is=title AnimStateNode}
   AnimStateNode_7 = JumpLand() {is=title AnimStateNode}
   AnimStateTransitionNode_0 = AnimStateTransitionNode() {crossfadeDuration=0.1}
   AnimStateTransitionNode_11 = AnimStateTransitionNode() {bAutomaticRuleBasedOnSequencePlayerInState=true}
   AnimStateTransitionNode_13 = AnimStateTransitionNode() {crossfadeDuration=0.35}
   AnimStateTransitionNode_16 = AnimStateTransitionNode()
   AnimStateTransitionNode_17 = AnimStateTransitionNode() {bAutomaticRuleBasedOnSequencePlayerInState=true}
   AnimStateTransitionNode_19 = AnimStateTransitionNode() {crossfadeDuration=0.067}
   AnimStateTransitionNode_3 = AnimStateTransitionNode() {crossfadeDuration=0.35}
  ]
  links [
   {
    dst=JumpStart
    rule=AnimGraph.AnimGraphNode_StateMachine_1.FullBody.AnimStateNode_1.SprintMode.AnimGraphNode_StateMachine_0.SprintingHoverboardMovement.AnimStateNode_1.Jump.AnimGraphNode_StateMachine_0.SpritingJump.AnimStateTransitionNode_16.Transition
    src=Default
    via=AnimStateTransitionNode_16
   }
   {
    dst=JumpFall
    rule=AnimGraph.AnimGraphNode_StateMachine_1.FullBody.AnimStateNode_1.SprintMode.AnimGraphNode_StateMachine_0.SprintingHoverboardMovement.AnimStateNode_1.Jump.AnimGraphNode_StateMachine_0.SpritingJump.AnimStateTransitionNode_13.AnimationTransitionGraph_0
    src=JumpApex
    via=AnimStateTransitionNode_13
   }
   {
    dst=JumpFall
    rule=AnimGraph.AnimGraphNode_StateMachine_1.FullBody.AnimStateNode_1.SprintMode.AnimGraphNode_StateMachine_0.SprintingHoverboardMovement.AnimStateNode_1.Jump.AnimGraphNode_StateMachine_0.SpritingJump.AnimStateTransitionNode_3.AnimationTransitionGraph_0
    src=JumpApex
    via=AnimStateTransitionNode_3
   }
   {
    dst=JumpLand
    rule=AnimGraph.AnimGraphNode_StateMachine_1.FullBody.AnimStateNode_1.SprintMode.AnimGraphNode_StateMachine_0.SprintingHoverboardMovement.AnimStateNode_1.Jump.AnimGraphNode_StateMachine_0.SpritingJump.AnimStateTransitionNode_0.AnimationTransitionGraph_0
    src=JumpFall
    via=AnimStateTransitionNode_0
   }
   {
    dst=Default
    rule=AnimGraph.AnimGraphNode_StateMachine_1.FullBody.AnimStateNode_1.SprintMode.AnimGraphNode_StateMachine_0.SprintingHoverboardMovement.AnimStateNode_1.Jump.AnimGraphNode_StateMachine_0.SpritingJump.AnimStateTransitionNode_17.Transition
    src=JumpLand
    via=AnimStateTransitionNode_17
   }
   {
    dst=JumpStart
    rule=AnimGraph.AnimGraphNode_StateMachine_1.FullBody.AnimStateNode_1.SprintMode.AnimGraphNode_StateMachine_0.SprintingHoverboardMovement.AnimStateNode_1.Jump.AnimGraphNode_StateMachine_0.SpritingJump.AnimStateTransitionNode_19.AnimationTransitionGraph_0
    src=JumpLand
    via=AnimStateTransitionNode_19
   }
   {
    dst=JumpApex
    rule=AnimGraph.AnimGraphNode_StateMachine_1.FullBody.AnimStateNode_1.SprintMode.AnimGraphNode_StateMachine_0.SprintingHoverboardMovement.AnimStateNode_1.Jump.AnimGraphNode_StateMachine_0.SpritingJump.AnimStateTransitionNode_11.AnimationTransitionGraph_0
    src=JumpStart
    via=AnimStateTransitionNode_11
   }
  ]
 }
 Default {
  path=AnimGraph.AnimGraphNode_StateMachine_1.FullBody.AnimStateNode_1.SprintMode.AnimGraphNode_StateMachine_0.SprintingHoverboardMovement.AnimStateNode_1.Jump.AnimGraphNode_StateMachine_0.SpritingJump.AnimStateNode_0.Default
  nodes [Output Animation Pose(Result=Use cached pose 'SprintingPose'())]
 }
 JumpApex {
  path=AnimGraph.AnimGraphNode_StateMachine_1.FullBody.AnimStateNode_1.SprintMode.AnimGraphNode_StateMachine_0.SprintingHoverboardMovement.AnimStateNode_1.Jump.AnimGraphNode_StateMachine_0.SpritingJump.AnimStateNode_1.JumpApex
  nodes [
   n1 = Sequence Player 'JackalVehicle_JumpNormal_Loop_Hoverboard'() {sequence=/Game/Animation/Game/MainPlayer/Vehicles/JackalVehicle/Parts/JackalVehicle_JumpNormal_Loop_Hoverboard}
   n2 = Sequence Player 'JackalVehicle_Idle_Hoverboard'() {sequence=&JackalVehicle_Idle_Hoverboard}
   n3 = Get bPlayJumpTrick() -> (bPlayJumpTrick:bool)
   Output Animation Pose(Result=Blend Poses by bool(
     BlendPose_0=Sequence Player 'JackalVehicle_JumpTrick_Vertical_Hoverboard'() {sequence=/Game/Animation/Game/MainPlayer/Vehicles/JackalVehicle/Parts/JackalVehicle_JumpTrick_Vertical_Hoverboard}
     BlendPose_1=Blend Poses by bool(BlendPose_0=@n2, BlendPose_1=@n1, BlendTime_0=0.3, BlendTime_1=0.3, bActiveValue=@n3) {blendType=Linear}
     BlendTime_0=0.3
     BlendTime_1=0.3
     bActiveValue=Get bPlayJumpTrickVertical() -> (bPlayJumpTrickVertical:bool)
    ) {blendType=Linear})
  ]
 }
 JumpStart {
  path=AnimGraph.AnimGraphNode_StateMachine_1.FullBody.AnimStateNode_1.SprintMode.AnimGraphNode_StateMachine_0.SprintingHoverboardMovement.AnimStateNode_1.Jump.AnimGraphNode_StateMachine_0.SpritingJump.AnimStateNode_11.JumpStart
  nodes [
   n1 = Sequence Player 'JackalVehicle_JumpNormal_Start_Hoverboard'() {bLoopAnimation=false, sequence=/Game/Animation/Game/MainPlayer/Vehicles/JackalVehicle/Parts/JackalVehicle_JumpNormal_Start_Hoverboard}
   n2 = Sequence Player 'JackalVehicle_Idle_Hoverboard'() {sequence=&JackalVehicle_Idle_Hoverboard}
   n3 = Get bPlayJumpTrick()
   n4 = Sequence Player 'JackalVehicle_JumpTrick_Vertical_Start_Hoverboard'() {
    bLoopAnimation=false
    sequence=/Game/Animation/Game/MainPlayer/Vehicles/JackalVehicle/Parts/JackalVehicle_JumpTrick_Vertical_Start_Hoverboard
   }
   Output Animation Pose(Result=Blend Poses by bool(
     BlendPose_0=@n4
     BlendPose_1=Blend Poses by bool(BlendPose_0=@n2, BlendPose_1=@n1, BlendTime_0=0.25, BlendTime_1=0.25, bActiveValue=@n3) {blendType=Linear}
     BlendTime_0=0.2
     BlendTime_1=0.2
     bActiveValue=Get bPlayJumpTrickVertical()
    ) {blendType=Linear})
  ]
 }
 JumpFall {
  path=AnimGraph.AnimGraphNode_StateMachine_1.FullBody.AnimStateNode_1.SprintMode.AnimGraphNode_StateMachine_0.SprintingHoverboardMovement.AnimStateNode_1.Jump.AnimGraphNode_StateMachine_0.SpritingJump.AnimStateNode_2.JumpFall
  nodes [
   n1 = Sequence Player 'JackalVehicle_JumpNormal_Fall_Hoverboard'() {
    bLoopAnimation=false
    playRate=0.3
    sequence=/Game/Animation/Game/MainPlayer/Vehicles/JackalVehicle/Parts/JackalVehicle_JumpNormal_Fall_Hoverboard
   }
   n2 = Sequence Player 'JackalVehicle_Idle_Hoverboard'() {sequence=&JackalVehicle_Idle_Hoverboard}
   n3 = Get bPlayJumpTrick()
   Output Animation Pose(Result=Blend Poses by bool(
     BlendPose_0=Sequence Player 'JackalVehicle_JumpTrick_Vertical_Hoverboard'() {sequence=&JackalVehicle_JumpTrick_Vertical_Hoverboard}
     BlendPose_1=Blend Poses by bool(BlendPose_0=@n2, BlendPose_1=@n1, BlendTime_0=0.3, BlendTime_1=0.3, bActiveValue=@n3) {blendType=Linear}
     BlendTime_0=0.3
     BlendTime_1=0.3
     bActiveValue=Get bPlayJumpTrickVertical()
    ) {blendType=Linear})
  ]
 }
 JumpLand {
  path=AnimGraph.AnimGraphNode_StateMachine_1.FullBody.AnimStateNode_1.SprintMode.AnimGraphNode_StateMachine_0.SprintingHoverboardMovement.AnimStateNode_1.Jump.AnimGraphNode_StateMachine_0.SpritingJump.AnimStateNode_7.JumpLand
  nodes [Output Animation Pose(Result=Sequence Player 'JackalVehicle_JumpNormal_Land_Hoverboard'() {sequence=/Game/Animation/Game/MainPlayer/Vehicles/JackalVehicle/Parts/JackalVehicle_JumpNormal_Land_Hoverboard})]
 }
 AnimationTransitionGraph_0 {
  path=AnimGraph.AnimGraphNode_StateMachine_1.FullBody.AnimStateNode_1.SprintMode.AnimGraphNode_StateMachine_0.SprintingHoverboardMovement.AnimStateNode_1.Jump.AnimGraphNode_StateMachine_0.SpritingJump.AnimStateTransitionNode_0.AnimationTransitionGraph_0
  nodes [Result(bCanEnterTransition=NOT Boolean(A=Get bInAir()))]
 }
 AnimationTransitionGraph_0 {
  path=AnimGraph.AnimGraphNode_StateMachine_1.FullBody.AnimStateNode_1.SprintMode.AnimGraphNode_StateMachine_0.SprintingHoverboardMovement.AnimStateNode_1.Jump.AnimGraphNode_StateMachine_0.SpritingJump.AnimStateTransitionNode_11.AnimationTransitionGraph_0
  nodes [Result()]
 }
 AnimationTransitionGraph_0 {
  path=AnimGraph.AnimGraphNode_StateMachine_1.FullBody.AnimStateNode_1.SprintMode.AnimGraphNode_StateMachine_0.SprintingHoverboardMovement.AnimStateNode_1.Jump.AnimGraphNode_StateMachine_0.SpritingJump.AnimStateTransitionNode_13.AnimationTransitionGraph_0
  nodes [Result(bCanEnterTransition=NOT Boolean(A=Get bInAir()))]
 }
 Transition {
  path=AnimGraph.AnimGraphNode_StateMachine_1.FullBody.AnimStateNode_1.SprintMode.AnimGraphNode_StateMachine_0.SprintingHoverboardMovement.AnimStateNode_1.Jump.AnimGraphNode_StateMachine_0.SpritingJump.AnimStateTransitionNode_16.Transition
  nodes [Result(bCanEnterTransition=Get bDefaultToJumpStartTransition() -> (bDefaultToJumpStartTransition:bool))]
 }
 Transition {
  path=AnimGraph.AnimGraphNode_StateMachine_1.FullBody.AnimStateNode_1.SprintMode.AnimGraphNode_StateMachine_0.SprintingHoverboardMovement.AnimStateNode_1.Jump.AnimGraphNode_StateMachine_0.SpritingJump.AnimStateTransitionNode_17.Transition
  nodes [Result()]
 }
 AnimationTransitionGraph_0 {
  path=AnimGraph.AnimGraphNode_StateMachine_1.FullBody.AnimStateNode_1.SprintMode.AnimGraphNode_StateMachine_0.SprintingHoverboardMovement.AnimStateNode_1.Jump.AnimGraphNode_StateMachine_0.SpritingJump.AnimStateTransitionNode_19.AnimationTransitionGraph_0
  nodes [Result(bCanEnterTransition=Get bIsVehicleJumping())]
 }
 AnimationTransitionGraph_0 {
  path=AnimGraph.AnimGraphNode_StateMachine_1.FullBody.AnimStateNode_1.SprintMode.AnimGraphNode_StateMachine_0.SprintingHoverboardMovement.AnimStateNode_1.Jump.AnimGraphNode_StateMachine_0.SpritingJump.AnimStateTransitionNode_3.AnimationTransitionGraph_0
  nodes [Result(bCanEnterTransition=Get bJumpApexToJumpFallTransition() -> (bJumpApexToJumpFallTransition:bool))]
 }
 Boosting {
  path=AnimGraph.AnimGraphNode_StateMachine_1.FullBody.AnimStateNode_1.SprintMode.AnimGraphNode_StateMachine_0.SprintingHoverboardMovement.AnimStateNode_2.Boosting
  nodes [Output Animation Pose(Result=Sequence Player 'JackalVehicle_Idle_Hoverboard'() {sequence=&JackalVehicle_Idle_Hoverboard})]
 }
 Transition {
  path=AnimGraph.AnimGraphNode_StateMachine_1.FullBody.AnimStateNode_1.SprintMode.AnimGraphNode_StateMachine_0.SprintingHoverboardMovement.AnimStateTransitionNode_11.Transition
  nodes [Result(bCanEnterTransition=Get bLocomotionPoseToJumpTransitionAndNotBoosting() -> (bLocomotionPoseToJumpTransitionAndNotBoosting:bool))]
 }
 Transition {
  path=AnimGraph.AnimGraphNode_StateMachine_1.FullBody.AnimStateNode_1.SprintMode.AnimGraphNode_StateMachine_0.SprintingHoverboardMovement.AnimStateTransitionNode_12.Transition
  nodes [Result(bCanEnterTransition=Get bJumpToLocomotionPoseTransition() -> (bJumpToLocomotionPoseTransition:bool))]
 }
 Transition {
  path=AnimGraph.AnimGraphNode_StateMachine_1.FullBody.AnimStateNode_1.SprintMode.AnimGraphNode_StateMachine_0.SprintingHoverboardMovement.AnimStateTransitionNode_13.Transition
  nodes [Result(bCanEnterTransition=Get bIsBoosting() -> (bIsBoosting:bool))]
 }
 Transition {
  path=AnimGraph.AnimGraphNode_StateMachine_1.FullBody.AnimStateNode_1.SprintMode.AnimGraphNode_StateMachine_0.SprintingHoverboardMovement.AnimStateTransitionNode_14.Transition
  nodes [Result(bCanEnterTransition=NOT Boolean(A=Get bIsBoosting()))]
 }
 Transition {
  path=AnimGraph.AnimGraphNode_StateMachine_1.FullBody.AnimStateNode_1.SprintMode.AnimGraphNode_StateMachine_0.SprintingHoverboardMovement.AnimStateTransitionNode_15.Transition
  nodes [Result(bCanEnterTransition=Get bIsBoosting())]
 }
 Transition {
  path=AnimGraph.AnimGraphNode_StateMachine_1.FullBody.AnimStateTransitionNode_7.Transition
  nodes [Result(bCanEnterTransition=Get bIsSprinting())]
 }
 Transition {
  path=AnimGraph.AnimGraphNode_StateMachine_1.FullBody.AnimStateTransitionNode_8.Transition
  nodes [Result(bCanEnterTransition=NOT Boolean(A=Get bIsSprinting()))]
 }
 GrapplingHook {
  entry=Idle
  path=AnimGraph.AnimGraphNode_StateMachine_2.GrapplingHook
  ids=id
  nodes [
   AnimStateNode_0 = Idle()
   AnimStateNode_1 = Idle_PlaceHolder() {is=title AnimStateNode}
   AnimStateNode_2 = HoverboardGrapplingJumpRetrigerred() {is=title AnimStateNode, bAlwaysResetOnEntry=true}
   AnimStateTransitionNode_3 = AnimStateTransitionNode() {crossfadeDuration=0.15}
   AnimStateTransitionNode_4 = AnimStateTransitionNode() {crossfadeDuration=0.15}
   AnimStateTransitionNode_5 = AnimStateTransitionNode() {crossfadeDuration=0.15}
  ]
  links [
   {
    dst=Idle_PlaceHolder
    rule=AnimGraph.AnimGraphNode_StateMachine_2.GrapplingHook.AnimStateTransitionNode_3.Transition
    src=HoverboardGrapplingJumpRetrigerred
    via=AnimStateTransitionNode_3
   }
   {
    dst=HoverboardGrapplingJumpRetrigerred
    rule=AnimGraph.AnimGraphNode_StateMachine_2.GrapplingHook.AnimStateTransitionNode_5.Transition
    src=Idle
    via=AnimStateTransitionNode_5
   }
   {
    dst=Idle
    rule=AnimGraph.AnimGraphNode_StateMachine_2.GrapplingHook.AnimStateTransitionNode_4.Transition
    src=Idle_PlaceHolder
    via=AnimStateTransitionNode_4
   }
  ]
 }
 Idle {
  path=AnimGraph.AnimGraphNode_StateMachine_2.GrapplingHook.AnimStateNode_0.Idle
  nodes [
   Sequence Player 'GrapplingHook_Travel_Hoverboard_Vehicle'() {
    bIgnoreForRelevancyTest=true
    bLoopAnimation=false
    sequence=/Game/Animation/Game/MainPlayer/Vehicles/JackalVehicle/Parts/GrapplingHook_Travel_Hoverboard_Vehicle
   }
   Output Animation Pose(Result=Sequence Player 'JackalVehicle_Idle_Hoverboard'() {sequence=&JackalVehicle_Idle_Hoverboard})
  ]
 }
 Idle_PlaceHolder {
  path=AnimGraph.AnimGraphNode_StateMachine_2.GrapplingHook.AnimStateNode_1.Idle_PlaceHolder
  nodes [Output Animation Pose(Result=Sequence Player 'JackalVehicle_Idle_Hoverboard'() {sequence=&JackalVehicle_Idle_Hoverboard})]
 }
 HoverboardGrapplingJumpRetrigerred {
  path=AnimGraph.AnimGraphNode_StateMachine_2.GrapplingHook.AnimStateNode_2.HoverboardGrapplingJumpRetrigerred
  nodes [
   n1 = Sequence Player 'GrapplingHook_Travel_Hoverboard_Vehicle'() {
    bIgnoreForRelevancyTest=true
    bLoopAnimation=false
    sequence=&GrapplingHook_Travel_Hoverboard_Vehicle
   }
   Output Animation Pose(Result=@n1)
  ]
 }
 Transition {
  path=AnimGraph.AnimGraphNode_StateMachine_2.GrapplingHook.AnimStateTransitionNode_3.Transition
  nodes [Result(bCanEnterTransition=NOT Boolean(A=Get bIsBothLegFK()))]
 }
 Transition {
  path=AnimGraph.AnimGraphNode_StateMachine_2.GrapplingHook.AnimStateTransitionNode_4.Transition
  nodes [Result(bCanEnterTransition=Get ShouldTrigerGrappleMontage() -> (ShouldTrigerGrappleMontage:bool))]
 }
 Transition {
  path=AnimGraph.AnimGraphNode_StateMachine_2.GrapplingHook.AnimStateTransitionNode_5.Transition
  nodes [Result(bCanEnterTransition=Get ShouldTrigerGrappleMontage())]
 }
 Locomotion {
  entry=Idle
  path=AnimGraph.AnimGraphNode_StateMachine_3.Locomotion
  ids=id
  nodes [
   AnimStateNode_12 = Pivot() {is=title AnimStateNode}
   AnimStateNode_3 = MovementStart() {is=title AnimStateNode}
   AnimStateNode_4 = MovementLoop() {is=title AnimStateNode}
   AnimStateNode_5 = MovementStop() {is=title AnimStateNode}
   AnimStateNode_8 = Idle()
   AnimStateTransitionNode_1 = AnimStateTransitionNode()
   AnimStateTransitionNode_10 = AnimStateTransitionNode() {crossfadeDuration=0.45}
   AnimStateTransitionNode_2 = AnimStateTransitionNode() {crossfadeDuration=0.45}
   AnimStateTransitionNode_27 = AnimStateTransitionNode()
   AnimStateTransitionNode_30 = AnimStateTransitionNode()
   AnimStateTransitionNode_31 = AnimStateTransitionNode() {bAutomaticRuleBasedOnSequencePlayerInState=true, crossfadeDuration=0.4}
   AnimStateTransitionNode_4 = AnimStateTransitionNode()
   AnimStateTransitionNode_5 = AnimStateTransitionNode() {bAutomaticRuleBasedOnSequencePlayerInState=true, crossfadeDuration=0.4, priorityOrder=2}
   AnimStateTransitionNode_6 = AnimStateTransitionNode() {crossfadeDuration=0.15}
   AnimStateTransitionNode_7 = AnimStateTransitionNode() {bAutomaticRuleBasedOnSequencePlayerInState=true, crossfadeDuration=0.35}
  ]
  links [
   {
    dst=MovementLoop
    rule=AnimGraph.AnimGraphNode_StateMachine_3.Locomotion.AnimStateTransitionNode_1.Transition
    src=Idle
    via=AnimStateTransitionNode_1
   }
   {
    dst=MovementStart
    rule=AnimGraph.AnimGraphNode_StateMachine_3.Locomotion.AnimStateTransitionNode_4.AnimationTransitionGraph_1
    src=Idle
    via=AnimStateTransitionNode_4
   }
   {
    dst=Idle
    rule=AnimGraph.AnimGraphNode_StateMachine_3.Locomotion.AnimStateTransitionNode_27.Transition
    src=MovementLoop
    via=AnimStateTransitionNode_27
   }
   {
    dst=Idle
    rule=AnimGraph.AnimGraphNode_StateMachine_3.Locomotion.AnimStateTransitionNode_6.AnimationTransitionGraph_1
    src=MovementLoop
    via=AnimStateTransitionNode_6
   }
   {
    dst=MovementStop
    rule=AnimGraph.AnimGraphNode_StateMachine_3.Locomotion.AnimStateTransitionNode_10.AnimationTransitionGraph_1
    src=MovementLoop
    via=AnimStateTransitionNode_10
   }
   {
    dst=MovementStop
    rule=AnimGraph.AnimGraphNode_StateMachine_3.Locomotion.AnimStateTransitionNode_2.AnimationTransitionGraph_1
    src=MovementLoop
    via=AnimStateTransitionNode_2
   }
   {
    dst=Pivot
    rule=AnimGraph.AnimGraphNode_StateMachine_3.Locomotion.AnimStateTransitionNode_30.Transition
    src=MovementLoop
    via=AnimStateTransitionNode_30
   }
   {
    dst=MovementLoop
    rule=AnimGraph.AnimGraphNode_StateMachine_3.Locomotion.AnimStateTransitionNode_5.AnimationTransitionGraph_1
    src=MovementStart
    via=AnimStateTransitionNode_5
   }
   {
    dst=Idle
    rule=AnimGraph.AnimGraphNode_StateMachine_3.Locomotion.AnimStateTransitionNode_7.AnimationTransitionGraph_1
    src=MovementStop
    via=AnimStateTransitionNode_7
   }
   {
    dst=MovementLoop
    rule=AnimGraph.AnimGraphNode_StateMachine_3.Locomotion.AnimStateTransitionNode_31.Transition
    src=Pivot
    via=AnimStateTransitionNode_31
   }
  ]
 }
 Pivot {
  path=AnimGraph.AnimGraphNode_StateMachine_3.Locomotion.AnimStateNode_12.Pivot
  nodes [
   n1 = "Blend Poses (EFortCardinalDirection)"(
    BlendPose_0=Sequence Player 'JackalVehicle_Pivot_StoN_Hoverboard'() {bLoopAnimation=false, sequence=/Game/Animation/Game/MainPlayer/Vehicles/JackalVehicle/Parts/JackalVehicle_Pivot_StoN_Hoverboard}
    BlendPose_1=Sequence Player 'JackalVehicle_Pivot_NtoS_Hoverboard'() {bLoopAnimation=false, sequence=/Game/Animation/Game/MainPlayer/Vehicles/JackalVehicle/Parts/JackalVehicle_Pivot_NtoS_Hoverboard}
    BlendTime_0=0.1
    BlendTime_1=0.1
    ActiveEnumValue=Get PrePivot() -> (PrePivot:EFortCardinalDirection Enum=North)
   ) {blendType=Linear, blendPose [{linkId=-1, sourceLinkId=-1}, {linkId=-1, sourceLinkId=-1}], blendTime [0.1,0.1]}
   Output Animation Pose(Result=@n1)
  ]
 }
 MovementStart {
  path=AnimGraph.AnimGraphNode_StateMachine_3.Locomotion.AnimStateNode_3.MovementStart
  nodes [
   n1 = Sequence Player 'JackalVehicle_Drive_N_Fast_Start_Hoverboard'() {bLoopAnimation=false, sequence=/Game/Animation/Game/MainPlayer/Vehicles/JackalVehicle/Parts/JackalVehicle_Drive_N_Fast_Start_Hoverboard}
   n2 = "Blend Poses (EFortCardinalDirection)"(
    BlendPose_0=@n1
    BlendPose_1=Sequence Player 'JackalVehicle_Drive_S_Start_Hoverboard'() {bLoopAnimation=false, sequence=/Game/Animation/Game/MainPlayer/Vehicles/JackalVehicle/Parts/JackalVehicle_Drive_S_Start_Hoverboard}
    BlendTime_0=0.2
    BlendTime_1=0.2
    ActiveEnumValue=Get LocomotionCardinalDirection() -> (LocomotionCardinalDirection=North)
   ) {blendType=Linear, blendPose [{linkId=-1, sourceLinkId=-1}, {linkId=-1, sourceLinkId=-1}], blendTime [0.1,0.1]}
   Output Animation Pose(Result=@n2)
  ]
 }
 MovementLoop {
  path=AnimGraph.AnimGraphNode_StateMachine_3.Locomotion.AnimStateNode_4.MovementLoop
  nodes [
   n1 = Sequence Player 'JackalVehicle_Pivot_EtoW_Hoverboard'() {bLoopAnimation=false, sequence=/Game/Animation/Game/MainPlayer/Vehicles/JackalVehicle/Parts/JackalVehicle_Pivot_EtoW_Hoverboard}
   n2 = Sequence Player 'JackalVehicle_Drive_W_Fast_CMM_Hoverboard'() {sequence=/Game/Animation/Game/MainPlayer/Vehicles/JackalVehicle/Parts/JackalVehicle_Drive_W_Fast_CMM_Hoverboard}
   n3 = Get bPlayMovingFast() -> (bPlayMovingFast:bool)
   n4 = Sequence Player 'JackalVehicle_Pivot_WtoE_Hoverboard'() {bLoopAnimation=false, sequence=/Game/Animation/Game/MainPlayer/Vehicles/JackalVehicle/Parts/JackalVehicle_Pivot_WtoE_Hoverboard}
   n5 = Sequence Player 'JackalVehicle_Drive_E_Fast_CMM_Hoverboard'() {sequence=/Game/Animation/Game/MainPlayer/Vehicles/JackalVehicle/Parts/JackalVehicle_Drive_E_Fast_CMM_Hoverboard}
   n6 = Get bPlayMovingFast()
   n7 = "Blend Poses (EFortCardinalDirection)"!(
    BlendPose_0=Sequence Player 'JackalVehicle_Drive_N_Fast_CMM_Hoverboard'() {sequence=/Game/Animation/Game/MainPlayer/Vehicles/JackalVehicle/Parts/JackalVehicle_Drive_N_Fast_CMM_Hoverboard}
    BlendPose_1=Blend Poses by bool(
     BlendPose_0=Sequence Player 'JackalVehicle_Drive_S_CMM_Hoverboard'() {sequence=/Game/Animation/Game/MainPlayer/Vehicles/JackalVehicle/Parts/JackalVehicle_Drive_S_CMM_Hoverboard}
     BlendPose_1=Blend Poses by bool(BlendPose_0=@n5, BlendPose_1=@n4, BlendTime_0=0.4, BlendTime_1=0.3, bActiveValue=@n6) {blendType=Linear}
     BlendTime_0=0.2
     BlendTime_1=0.2
     bActiveValue=Get bPlayDriveSouth() -> (bPlayDriveSouth:bool)
    ) {blendType=Linear}
    BlendPose_2:Pose Link Structure=Sequence Player 'JackalVehicle_Drive_S_CMM_Hoverboard'() {sequence=&JackalVehicle_Drive_S_CMM_Hoverboard}
    BlendPose_3:Pose Link Structure=Blend Poses by bool(
     BlendPose_0=Sequence Player 'JackalVehicle_Drive_S_CMM_Hoverboard'() {sequence=&JackalVehicle_Drive_S_CMM_Hoverboard}
     BlendPose_1=Blend Poses by bool(BlendPose_0=@n2, BlendPose_1=@n1, BlendTime_0=0.4, BlendTime_1=0.3, bActiveValue=@n3) {blendType=Linear}
     BlendTime_0=0.2
     BlendTime_1=0.2
     bActiveValue=Get bPlayDriveSouth()
    ) {blendType=Linear}
    BlendTime_0=0.3
    BlendTime_1=0.25
    BlendTime_2:float=0.15
    BlendTime_3:float=0.25
    ActiveEnumValue=Get LocomotionCardinalDirection() -> (LocomotionCardinalDirection=North)
   ) -> (Pose) {
    blendType=Linear
    childUpateMode=ResetChildOnActivate
    blendPose [{linkId=-1, sourceLinkId=-1}, {linkId=-1, sourceLinkId=-1}, {linkId=-1, sourceLinkId=-1}, {linkId=-1, sourceLinkId=-1}]
    blendTime [0.1,0.1,0.1,0.1]
   }
   Output Animation Pose(Result=@n7)
  ]
 }
 MovementStop {
  path=AnimGraph.AnimGraphNode_StateMachine_3.Locomotion.AnimStateNode_5.MovementStop
  nodes [
   n1 = Sequence Player 'JackalVehicle_Drive_N_Stop_Hoverboard'() {
    bLoopAnimation=false
    groupName=Stop
    method=SyncGroup
    sequence=&JackalVehicle_Drive_N_Stop_Hoverboard
    startPosition=0.3
   }
   Output Animation Pose(Result=@n1)
  ]
 }
 Idle {
  path=AnimGraph.AnimGraphNode_StateMachine_3.Locomotion.AnimStateNode_8.Idle
  nodes [
   n1 = Get LeanYaw()
   Output Animation Pose(Result=Apply Additive(
     Base=Sequence Player 'JackalVehicle_Idle_Hoverboard'() {sequence=&JackalVehicle_Idle_Hoverboard}
     Additive=Blendspace Player 'JackalVehicle_Leans_InPlace_BS_Hoverboard'(X=@n1) {blendSpace=&JackalVehicle_Leans_InPlace_BS_Hoverboard}
    ) {alphaBoolBlend {blendOption=Linear}})
  ]
 }
 Transition {
  path=AnimGraph.AnimGraphNode_StateMachine_3.Locomotion.AnimStateTransitionNode_1.Transition
  nodes [Result(bCanEnterTransition=Get bIdleToMovementLoopTransition() -> (bIdleToMovementLoopTransition:bool))]
 }
 AnimationTransitionGraph_1 {
  path=AnimGraph.AnimGraphNode_StateMachine_3.Locomotion.AnimStateTransitionNode_10.AnimationTransitionGraph_1
  nodes [Result(bCanEnterTransition=Get bMovementLoopToMovementStopTransition() -> (bMovementLoopToMovementStopTransition:bool))]
 }
 AnimationTransitionGraph_1 {
  path=AnimGraph.AnimGraphNode_StateMachine_3.Locomotion.AnimStateTransitionNode_2.AnimationTransitionGraph_1
  nodes [Result(bCanEnterTransition=Get bMovementLoopToMovementStopTransition())]
 }
 Transition {
  path=AnimGraph.AnimGraphNode_StateMachine_3.Locomotion.AnimStateTransitionNode_27.Transition
  nodes [Result(bCanEnterTransition=Get bMovementLoopToIdleTransition() -> (bMovementLoopToIdleTransition:bool))]
 }
 Transition {
  path=AnimGraph.AnimGraphNode_StateMachine_3.Locomotion.AnimStateTransitionNode_30.Transition
  nodes [Result(bCanEnterTransition=Get bMovementLoopToPivotTransition() -> (bMovementLoopToPivotTransition:bool))]
 }
 Transition {path=AnimGraph.AnimGraphNode_StateMachine_3.Locomotion.AnimStateTransitionNode_31.Transition, nodes [Result()]}
 AnimationTransitionGraph_1 {
  path=AnimGraph.AnimGraphNode_StateMachine_3.Locomotion.AnimStateTransitionNode_4.AnimationTransitionGraph_1
  nodes [Result(bCanEnterTransition=Get bIdleToMovementStartTransition() -> (bIdleToMovementStartTransition:bool))]
 }
 AnimationTransitionGraph_1 {path=AnimGraph.AnimGraphNode_StateMachine_3.Locomotion.AnimStateTransitionNode_5.AnimationTransitionGraph_1, nodes [Result()]}
 AnimationTransitionGraph_1 {
  path=AnimGraph.AnimGraphNode_StateMachine_3.Locomotion.AnimStateTransitionNode_6.AnimationTransitionGraph_1
  nodes [Result(bCanEnterTransition=Get bMovementLoopToIdleTransition())]
 }
 AnimationTransitionGraph_1 {path=AnimGraph.AnimGraphNode_StateMachine_3.Locomotion.AnimStateTransitionNode_7.AnimationTransitionGraph_1, nodes [Result()]}
 EventGraph {nodes [
   Event BlueprintUpdateAnimation()
   n1 = GetDriver!(self:Fort Vehicle Interface Interface=Get JackalVehicle() -> (JackalVehicle:Fort Athena Jackal Vehicle Object Reference)) -> (ReturnValue:Fort Player Pawn Object Reference)
   > Is Valid(InputObject=@n1)
   > n2 = Cast To JackalVehicleLayerAnimBP!(
    execute
    Object=GetLinkedAnimLayerInstanceByClass!(
     self:Skeletal Mesh Component Object Reference=Get Mesh(self=@n1)
     InClass:Anim Instance Class Reference=/Game/Characters/Player/Male/Male_Avg_Base/LayerAnimBP/JackalVehicleLayerAnimBP.JackalVehicleLayerAnimBP_C
    ) -> (ReturnValue:Anim Instance Object Reference)
   ) -> (then, CastFailed, AsJackal Vehicle Layer Anim BP:Jackal Vehicle Layer Anim BP Object Reference) {is=title K2Node_DynamicCast}
   > Set ShouldTrigerGrappleMontage(ShouldTrigerGrappleMontage:bool=Get ShouldTriggerGrappleMontage!(self:Jackal Vehicle Layer Anim BP Object Reference=@n2) -> (ShouldTriggerGrappleMontage:bool)) -> (Output_Get:bool)
   n3 = GetAnimationInterface!(self:Fort Player Pawn Object Reference=@n1) -> (ReturnValue:Fort Player Animation Interface Object Reference)
   > Set AreFeetAttachedToBoard(AreFeetAttachedToBoard:bool=Break Fort Anim Input Jackal Vehicle!(FortAnimInput_JackalVehicle:"Fort Anim Input Jackal Vehicle Structure (by ref)"=GetJackalVehicleInput!(self:Fort Player Animation Interface Object Reference=@n3) -> (ReturnValue:"Fort Anim Input Jackal Vehicle Structure (by ref)")) -> (
     bIsUsingJackalVehicle:bool=~
     bIsSteeringLeft:bool=~
     bIsSteeringRight:bool=~
     SteerAngleDeadZoneDegrees:float=~
     SteerAngle:float=~
     LeanYaw:float=~
     QuantizedSteerAngle:float=~
     SteerAlpha:float=~
     RunForwardAlpha:float=~
     SlopePitchDegreeAngle:float=~
     SlopeRollDegreeAngle:float=~
     DistanceFromGround:float=~
     PivotPlayRate:float=~
     bIsReversing:bool=~
     bIsBraking:bool=~
     bIsMoving:bool=~
     bIsMovingForward:bool=~
     bIsSprinting:bool=~
     bInAir:bool=~
     bIsFalling:bool=~
     bIsJumping:bool=~
     bIsRelaxed:bool=~
     bIsBoosting:bool=~
     bHasReachedJumpChargeStartThreshold:bool=~
     bHasReachedJumpChargeMidThreshold:bool=~
     bHasReachedJumpChargeFullThreshold:bool=~
     bAdjustRootForFemaleRider:bool=~
     bIsPlayingEmoteOnHoverboard:bool=~
     bShouldApplyAdditive:bool=~
     bPlayPivotOnGroundAndNotBoosting:bool=~
     bIsOnLowerHill:bool=~
     bLowerBodyIdleToLoopTransition:bool=~
     bInterruptHoverboardFullybody:bool=~
     bEnterFullBodyHoverboardState:bool=~
     bDefaultToJumpStartTransition:bool=~
     bLocomotionPoseToJumpTransition:bool=~
     bBoostingToBoostingJumpTransition:bool=~
     bJumpToLocomotionPoseTransition:bool=~
     bJumpApexToJumpFallTransition:bool=~
     bIdleToMovementStartTransition:bool=~
     bIdleToMovementLoopTransition:bool=~
     bMovementLoopToMovementStopTransition:bool=~
     bMovementLoopToPivotTransition:bool=~
     bMovementLoopToIdleTransition:bool=~
     bIdleAdditiveToCollisionNTransition:bool=~
     bSplitBodyToHoverboardBRTransition:bool=~
     bHoverboardBRToSplitBodyTransition:bool=~
     bHoverboardBRMovementToJumpChargeTransition:bool=~
     bIdlesToJackalVehicleTransition:bool=~
     bPlayAdditiveLeans:bool=~
     bPlayBalloonLeans:bool=~
     bPlayJumpTrickVertical:bool=~
     bPlayJumpTrick:bool=~
     bPlayMovingFast:bool=~
     bPlayHipAdjustmentAdditive:bool=~
     bPlayDriveSouth:bool=~
     bPlayHeadAimOffset:bool=~
     bPlaySlopeAimOffset:bool=~
     JumpCombatAdditiveWeight:float=~
     MeleeTwistIdle:float=~
     MeleeTwistLocomotionLoop:float=~
     VelocityImpact:Fort Anim Input Velocity Impact Structure
     ImpactDisplacement:Vector=~
     SteerYaw:float=~
     EmoteHoverboardPosition:Vector=~
     EmoteHoverboardRotation:Rotator=~
     bShouldAttachFeetToHoverboard:bool=~
     FootLeftLocationOffset:Vector=~
     FootLeftRotationOffset:Rotator=~
     FootRightLocationOffset:Vector=~
     FootRightRotationOffset:Rotator=~
     StoppedThreshold:float=~
     MovingForwardThreshold:float=~
     MovingFowardFastThreshold:float=~
     DefaultToJumpStartTransitionThreshold:float=~
     JumpTrickAngularVelocityThreshold:float=~
     JumpDistanceFromGroundThreshold:float=~
     VelocityStartThreshold:float=~
     MovingFastThreshold:float=~
     RelaxedSpeedThreshold:float=~
     JumpChargeStartThreshold:float=~
     JumpChargeMidThreshold:float=~
     JumpChargeFullThreshold:float=~
     RotatingAngularVelocityThreshold:float=~
     IdleToLoopTransitionSpeedThreshold:float=~
     IdleToMovementLoopTransitionThreshold:float=~
     MeleeTwistIdleMultiplier:float=~
     MeleeTwistLocomotionLoopMultiplier:float=~
     LeanYawForMaxPivotPlayRate:float=~
     SteerAlphaForMaxPivotPlayRate:float=~
     PrePivotCardinalDirection:EFortCardinalDirection Enum
    ) {is=title K2Node_BreakStruct}.bShouldAttachFeetToHoverboard) -> (Output_Get:bool)
  ]}
]
variables [
 AreFeetAttachedToBoard {type=boolean}
 PrePivot {type=EFortCardinalDirection, default=North}
 ShouldTrigerGrappleMontage {type=boolean}
]
