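# Physics asset for the Jackal hoverboard skeletal mesh, a drivable vehicle in Athena.
# It holds a single body on the root bone with one capsule named DriverCollision, raised above the root to stand in for the rider's collision volume.
# No constraints are defined, so there is no articulated simulation between bones.
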PhysicsAsset /Game/Athena/DrivableVehicles/Hoverboard/SK_Jackal_Athena_Physics
bodies {root [{shapeType=Capsule, length=88.5, radius=37.25, shapeName=DriverCollision, center=<-1.0082 0.0002 95.2608>}]}
constraints []
